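# Niagara system for the FlameBride character's ambient flame, smoke and ember effect, attached to skeletal mesh bones or sockets.
# FireTrail and SmokeTrail burst persistent meshes at per-partition placements, feeding smoothed positions to particle parameters and switching layouts on Change_OffPhysics.
# Owner speed gates drifting Dirt and FirePiece sprites, and movement-spawned Run_Fire and Run_Piece sprites.
# Front-end preview, a clamped system float and time of day adjust brightness.
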NiagaraSystem /BRCosmetics/Characters/Player/Female/Medium/Bodies/F_MED_FlameBride/FX/NS_Character_FlameBride_Idle
dependencies {
 usedDataInterfaces [
  &NiagaraDataInterfaceArrayFloat3
  &NiagaraDataInterfaceArrayInt32
  &NiagaraDataInterfaceArrayFloat
  &NiagaraDataInterfaceColorCurve
  &NiagaraDataInterfaceCurve
  &NiagaraDataInterfaceSpriteRendererInfo
 ]
 usedDynamicInputs [&ClampFloat]
 usedModules [
  &self:SystemSpawnScript.SystemScriptSource.SystemScriptGraph.NiagaraNodeAssignment_0.SetVariables_#1
  &SystemState
  &SampleSkeletalMesh
  &self:SystemSpawnScript.SystemScriptSource.SystemScriptGraph.NiagaraNodeAssignment_1.SetVariables_#2
 ]
}
emitters {
 Dirt {
  emitterUpdateScript [
   &EmitterState(
    Enable Distance Culling=true
    Inactive Response=::NewEnumerator1
    Life Cycle Mode=::NewEnumerator1
    Loop Duration=1
    MaxDistance=2000
    Module.bUseMaxDistance=true
    Module.bUseMinDistance=false
    Scalability Mode=::NewEnumerator1
    UseLoopDelay=false
   )
   &SpawnRate(
    SpawnRate=&Multiply_Float(
     A=&MakeCustomFloatFromBool(Bool=$User.bIsFrontEndPreview, True Float=10, False Float=20)
     B=&MakeCustomFloatFromBool(
      Bool=&SetBoolByFloatComparison(A=&VectorLength(Vector=$Engine.Owner.Velocity), B=15)
      True Float=0
      False Float=1
     )
    )
    Use Spawn Probability=false
   )
  ]
  particleSpawnScript [
   &PartitionParticles(Number Of Partitions=15)
   &InitializeParticle(
    Lifetime Min=0.75
    Lifetime Mode=::NewEnumerator1
    Sprite Rotation Mode=::NewEnumerator2
    Sprite Size Max=<2 2>
    Sprite Size Min=<0.5 0.5>
    Sprite Size Mode=::NewEnumerator2
    UsePositionOffset=false
    Write Material Random=true
   )
   &ShapeLocation(
    MaskSpawnGroup=false
    Offset Mode=::NewEnumerator0
    Override Seed=false
    Rotation Axis=<1 0 0.6>
    Rotation Mode=::NewEnumerator1
    Sphere Radius=10
    UseOwnerScale=true
   )
   &AddVelocity(Velocity=<0 0 10>)
   &AddVelocity(Module.Use Falloff Distance=false, Override Seed=false, Velocity Mode=::NewEnumerator1, Velocity Speed=5) {module=AddVelocity001}
  ]
  particleUpdateScript [
   &ParticleState()
   &SpriteRotationRate(Rotation Rate=&RandomRangeFloat(Minimum=-20, Maximum=10))
   &SkeletalMeshLocation(
    Bone / Socket Index=&RandomRangeInt(Minimum=0, Maximum=15)
    Bone Sampling Mode=::NewEnumerator4
    MaskSpawnGroup=false
    Orientation Sampling=::NewEnumerator1
    Position Sampling=::NewEnumerator1
    Sampled Position Offset=&AddVectorToPosition(Position=Linked() {linkedVariable {name=Particles.ShapeLocation.ShapeVector, type=Vector3f}}, Vector=<0 0 0>)
    Skeletal Mesh=$System.Skeletal Mesh
    Source Movement Scale:Vector3f=<0.1 0.1 0.1>
   )
   &CurlNoiseForce(Bias Noise=false, Noise Strength=15, Pan Noise=false)
   &Drag(Use Linear Drag=true, Use Rotational Drag=true)
   &InheritSourceMovement(Applied Position Scale=<1 1 0>, ApplyForce=false, ApplyPosition=true, ApplyPositionFalloff=false) {enabled=false}
   &SolveForcesAndVelocity(Clamp Velocity=false, Limit Acceleration=false)
   &DynamicMaterialParameters(
    Index 0 Param 1=&FloatFromCurve(
     FloatCurve=DataInterface() {propertyValues {Curve {keys [{arriveTangent=-4.7818, interpMode=RCIM_Cubic, leaveTangent=-4.7818, tangentMode=RCTM_User, time=0.6, value=1}, <at 1 0.2 RCIM_Cubic>]}}}
     Scale Curve=&RandomRangeFloat(Minimum=0.5, Maximum=1)
    )
    Index 0 Param 2=&RandomRangeFloat(Minimum=0, Maximum=0.5)
    Index 0 Param 3=&RandomRangeFloat(Minimum=0, Maximum=0.5)
    Index 0 Param 4=&RandomRangeFloat(Minimum=0.3, Maximum=0.3)
    Module.Index0Float0Write=true
    Module.Index0Float1Write=true
    Module.Index0Float2Write=true
    Module.Index0Float3Write=true
   )
   &ScaleColor(
    Scale Alpha=&FloatFromCurve(FloatCurve=DataInterface() {propertyValues {Curve {keys [<at 0 0>,<at 0.2 1>,<at 0.8 1>,<at 1 0>]}}}, Scale Curve=1.3)
    Scale RGB=&MakeVectorFromLinearColorRGB(Color=<rgba 0.0625 0.0589 0.0573 1>)
    ScaleA=true
    ScaleRGB=true
   )
  ]
  rendererClasses [&NiagaraSpriteRendererProperties]
  renderers [{rendererClass=&NiagaraSpriteRendererProperties}]
 }
 FirePiece {
  emitterUpdateScript [
   &EmitterState(
    Enable Distance Culling=true
    Inactive Response=::NewEnumerator1
    Life Cycle Mode=::NewEnumerator1
    Loop Duration=1
    MaxDistance=2000
    Module.bUseMaxDistance=true
    Module.bUseMinDistance=false
    Scalability Mode=::NewEnumerator1
    UseLoopDelay=false
   )
   &SpawnRate(
    SpawnRate=&Multiply_Float(
     A=&MakeCustomFloatFromBool(Bool=$User.bIsFrontEndPreview, True Float=20, False Float=5)
     B=&MakeCustomFloatFromBool(
      Bool=&SetBoolByFloatComparison(A=&VectorLength(Vector=$Engine.Owner.Velocity), B=15)
      True Float=0
      False Float=1
     )
    )
    Use Spawn Probability=false
   )
  ]
  particleSpawnScript [
   &PartitionParticles(Number Of Partitions=15)
   &InitializeParticle(
    Lifetime Max=1
    Lifetime Min=0.4
    Lifetime Mode=::NewEnumerator1
    Sprite Size Max=<0.8 1.4>
    Sprite Size Min=<0.4 0.8>
    Sprite Size Mode=::NewEnumerator2
    UsePositionOffset=false
    Write Material Random=true
   )
   &ShapeLocation(MaskSpawnGroup=false, Offset Mode=::NewEnumerator0, Override Seed=false, Sphere Radius=10, UseOwnerScale=true)
   &AddVelocity(Velocity=<0 0 12>)
   &AddVelocity(Module.Use Falloff Distance=false, Override Seed=false, Velocity Mode=::NewEnumerator1, Velocity Speed=15) {module=AddVelocity001}
  ]
  particleUpdateScript [
   &ParticleState()
   &SkeletalMeshLocation(
    Bone / Socket Index=&RandomRangeInt(Minimum=0, Maximum=15)
    Bone Sampling Mode=::NewEnumerator4
    MaskSpawnGroup=false
    Orientation Sampling=::NewEnumerator1
    Position Sampling=::NewEnumerator1
    Sampled Position Offset=&AddVectorToPosition(Position=Linked() {linkedVariable {name=Particles.ShapeLocation.ShapeVector, type=Vector3f}}, Vector=<0 0 5>)
    Skeletal Mesh=$System.Skeletal Mesh
    Source Movement Scale:Vector3f=<0.1 0.1 0.1>
   )
   &CurlNoiseForce(Bias Noise=false, Noise Strength=-20, Pan Noise=false)
   &Drag(Use Linear Drag=true, Use Rotational Drag=true)
   &SolveForcesAndVelocity(Clamp Velocity=false, Limit Acceleration=false)
   &ScaleColor(
    Scale Alpha=&FloatFromCurve(FloatCurve=DataInterface() {propertyValues {Curve {keys [<at 0 0>,<at 0.2 1>,<at 0.8 1>,<at 1 0>]}}})
    Scale RGB=&MakeVectorFromLinearColorRGB(Color=&ColorFromCurve(ColorCurve=DataInterface() {propertyValues {
        AlphaCurve {keys []}
        BlueCurve {keys [<at 0.4801 0 RCIM_Cubic>,<at 0.6719 0 RCIM_Cubic>]}
        GreenCurve {keys [<at 0.4801 0.8 RCIM_Cubic>,<at 0.6719 0.15 RCIM_Cubic>]}
        RedCurve {keys [<at 0.4801 3 RCIM_Cubic>,<at 0.6719 1 RCIM_Cubic>]}
       }}))
    ScaleA=true
    ScaleRGB=true
   )
   /SetVariables(Particles.Glow03_Size=&Multiply_Vector2D_ByFloat(
     Float=&FloatFromCurve(
      FloatCurve=DataInterface() {propertyValues {Curve {keys [
          <at 0 0 RCIM_Cubic>
          <at 0.2 1.3 RCIM_Cubic>
          {arriveTangent=-0.1423, interpMode=RCIM_Cubic, leaveTangent=-0.1423, tangentMode=RCTM_User, time=1, value=0}
         ]}}}
      Scale Curve=5
     )
     Vector2D=<0.5 1>
    )) {isSetParameters=true}
   /SetVariables(Particles.Glow03_Color=&MakeLinearColorFromVectorAndFloat(
     "Vector (RGB)"=&MakeVectorFromLinearColorRGB(Color=&MultiplyLinearColor(
       Linear Color=<rgba 1 0.2 0.02 0.2>
       Scale Factor=&MakeVector4_FromFloat(Float=&Multiply_Float(A=1, B=&FNTimeOfDayScalar(Night Scale=0.5, Morning Scale=1, Evening Scale=1)))
      ))
     "Float (Alpha)"=&FloatFromCurve(
      FloatCurve=DataInterface() {propertyValues {Curve {keys [
          <at 0 0 RCIM_Cubic>
          {arriveTangent=0.1196, interpMode=RCIM_Cubic, leaveTangent=0.1196, tangentMode=RCTM_User, time=0.2, value=1}
          {arriveTangent=0.0387, interpMode=RCIM_Cubic, leaveTangent=0.0387, tangentMode=RCTM_User, time=0.5, value=1}
          {arriveTangent=-0.1423, interpMode=RCIM_Cubic, leaveTangent=-0.1423, tangentMode=RCTM_User, time=1, value=0}
         ]}}}
      Scale Curve=0.5
     )
    )) {isSetParameters=true}
  ]
  rendererClasses [&NiagaraSpriteRendererProperties]
  renderers [
   {rendererClass=&NiagaraSpriteRendererProperties}
   {rendererClass=&NiagaraSpriteRendererProperties, rendererIndex=1}
  ]
 }
 FireTrail {
  emitterUpdateScript [
   &EmitterState(
    Inactive Response=::NewEnumerator1
    Life Cycle Mode=::NewEnumerator1
    Loop Behavior=::NewEnumerator1
    Loop Duration Mode=::NewEnumerator1
    UseLoopDelay=false
   )
   &SpawnBurst_Instantaneous(
    Spawn Count=&MakeCustomIntFromBool(Bool=$User.Change_OffPhysics, True INT=7, False INT=10)
    Use Spawn Probability=false
    UseLoopCountLimit=false
   )
  ]
  particleSpawnScript [
   &PartitionParticles(Number Of Partitions=&MakeCustomIntFromBool(Bool=$User.Change_OffPhysics, True INT=7, False INT=10))
   &InitializeParticle(
    Lifetime=5
    Mesh Scale=&MakeCustomVectorFromBool(
     Bool=$User.Change_OffPhysics
     True Vector=&SelectVectorFromArray(
      Direct Array Index=$Particles.PartitionParticles.Partition
      Vector Selection Array=DataInterface() {propertyValues {FloatData [<1 1 0.8>, <1 1 0.8>, <1 1 1>, <1 1 0.7>, <1 1 0.8>, <1 1 0.9>, <1 1 0.8>]}}
     )
     False Vector=&Multiply_Vector(
      A=&SelectVectorFromArray(
       Direct Array Index=$Particles.PartitionParticles.Partition
       Vector Selection Array=DataInterface() {propertyValues {FloatData [<1 1 0.8>, <1 1 1>, <1 1 1.2>, <1 1 0.7>, <1 1 0.8>, <1 1 0.9>, <1 1 0.8>, <1 1 0.7>, <1 1 0.8>, <1 1 0.8>]}}
      )
      B=<1 1 1.1>
     )
    )
    Mesh Scale Mode=::NewEnumerator4
    UsePositionOffset=false
    Write Material Random=true
   )
   &InitialMeshOrientation(
    Mesh Orientation Mode=::NewEnumerator2
    Rotation=&MakeCustomVectorFromBool(
     Bool=$User.Change_OffPhysics
     True Vector=<0 0.25 0>
     False Vector=&SelectVectorFromArray(
      Direct Array Index=$Particles.PartitionParticles.Partition
      Vector Selection Array=DataInterface() {propertyValues {FloatData [<0.01 0.02 0>, <0 0 0>, <0 0 0>, <0 0 0>, <0 0 0>, <0 0 0>, <0 0 0>, <0 0 0>, <0 0 0>, <0 -0.01 0>]}}
     )
    )
    Use Rotation Vector=true
   )
   &CameraOffset(Camera Offset Amount=&MakeCustomFloatFromBool(Bool=$User.Change_OffPhysics, True Float=5, False Float=2))
  ]
  particleUpdateScript [
   &ParticleState(Kill Particles When Lifetime Has Elapsed=false, Let Infinitely Lived Particles Die When Emitter Deactivates=true)
   &SkeletalMeshLocation(
    Bone / Socket Index=&MakeCustomIntFromBool(
     Bool=$User.Change_OffPhysics
     True INT=&SelectIntFromArray(
      Direct Array Index=$Particles.PartitionParticles.Partition
      Int Selection Array=DataInterface() {propertyValues {IntData [15,15,15,16,16,16,16]}}
     )
     False INT=&SelectIntFromArray(
      Direct Array Index=$Particles.PartitionParticles.Partition
      Int Selection Array=DataInterface() {propertyValues {IntData [1,2,3,4,5,8,9,10,11,14]}}
     )
    )
    Bone Sampling Mode=::NewEnumerator4
    MaskSpawnGroup=false
    Sampled Position Offset=&MakeCustomVectorFromBool(
     Bool=$User.Change_OffPhysics
     True Vector=&MakeCustomVectorFromBool(
      Bool=$User.bIsFrontEndPreview
      True Vector=&SelectVectorFromArray(
       Direct Array Index=$Particles.PartitionParticles.Partition
       Vector Selection Array=DataInterface() {propertyValues {FloatData [<-60 -15 0>, <-40 -15 6>, <-40 -12 16>, <5 5 10>, <5 5 0>, <-8 10 0>, <-5 0 -9>]}}
      )
      False Vector=&SelectVectorFromArray(
       Direct Array Index=$Particles.PartitionParticles.Partition
       Vector Selection Array=DataInterface() {propertyValues {FloatData [<-60 -15 0>, <-40 -15 6>, <-40 -12 16>, <5 5 10>, <5 5 0>, <-8 15 0>, <-5 0 -12>]}}
      )
     )
     False Vector=&SelectVectorFromArray(
      Direct Array Index=$Particles.PartitionParticles.Partition
      Vector Selection Array=DataInterface() {propertyValues {FloatData [<0 0 2>, <2 0 -5>, <0 0 -5>, <0 0 0>, <0 0 -3>, <2 0 -3>, <1 0 -1>, <0 -2 1>, <1 0 0>, <2 -1 -1>]}}
     )
    )
    Skeletal Mesh=$System.Skeletal Mesh
   )
   &InterpolateOverTime(Rate Of Change=8, Target Vector Value:NiagaraPosition=$Particles.Position, Variable Type=::NewEnumerator2)
   /SetVariables(Particles.InterpolateDY:Vector4f=&MakeFloat4FromVectorAndFloat(
     "Vector (RGB)"=$StackContext.InterpolateOverTime.InterpolateOverTime_V3_.Moving Average
     "Float (Alpha)"=&MakeCustomFloatFromBool(Bool=$User.Change_OffPhysics, True Float=0, False Float=0.3)
    )) {isSetParameters=true, module=SetVariables_#3}
   &DynamicMaterialParameters(
    Index 0 Param 1=&RandomRangeFloat(Minimum=0.5, Maximum=1)
    Index 0 Param 2=&RandomRangeFloat(Minimum=0.8, Maximum=1)
    Index 0 Param 3=&RandomRangeFloat(Minimum=0.5, Maximum=1)
    Index 0 Param 4=&RandomRangeFloat(Minimum=0.5, Maximum=1)
    Index 3 Param 1=&MakeCustomFloatFromBool(
     Bool=$User.Change_OffPhysics
     True Float=&SelectFloatFromArray(
      Direct Array Index=$Particles.PartitionParticles.Partition
      Float Selection Array=DataInterface() {propertyValues {FloatData [6,2,6,5,3,2,2]}}
     )
     False Float=&SelectFloatFromArray(
      Direct Array Index=$Particles.PartitionParticles.Partition
      Float Selection Array=DataInterface() {propertyValues {FloatData [8,1,8,5,3,2,2,5,6,5]}}
     )
    )
    Index 3 Param 2=&MakeCustomFloatFromBool(
     Bool=$User.Change_OffPhysics
     True Float=&SelectFloatFromArray(
      Direct Array Index=$Particles.PartitionParticles.Partition
      Float Selection Array=DataInterface() {propertyValues {FloatData [8,10,8,8,7,7,5]}}
     )
     False Float=&SelectFloatFromArray(
      Direct Array Index=$Particles.PartitionParticles.Partition
      Float Selection Array=DataInterface() {propertyValues {FloatData [8,12,10,8,7,7,10,5,3,6]}}
     )
    )
    Index 3 Param 3=&RandomRangeFloat(Minimum=4, Maximum=6)
    Index 3 Param 4=&RandomRangeFloat(Minimum=0.5, Maximum=1)
    Module.Index0Float0Write=true
    Module.Index0Float1Write=true
    Module.Index0Float2Write=true
    Module.Index0Float3Write=true
    Module.Index3Float0Write=true
    Module.Index3Float1Write=true
    Module.Index3Float2Write=true
    Module.Index3Float3Write=true
    Write Parameter Index 3=::NewEnumerator1
   )
   &ScaleColor(
    Scale Alpha=&MakeCustomFloatFromBool(
     Bool=$User.Change_OffPhysics
     True Float=&RandomRangeFloat(Minimum=0.4, Maximum=0.6)
     False Float=&RandomRangeFloat(Minimum=0.5, Maximum=1)
    )
    ScaleA=true
    ScaleRGB=true
   )
   &ScaleColor(
    Scale Alpha=&MakeCustomFloatFromBool(Bool=$User.bIsFrontEndPreview, True Float=1.2, False Float=&Lerp_Float(A=0.5, B=1.5, Alpha=$System.ClampedFloat))
    Scale RGB=&MakeVectorFromLinearColorRGB(Color=&MakeCustomLinearColorFromBool(
      Bool=$User.bIsFrontEndPreview
      True Color=<rgba 1.7 1.7 1.7 1>
      False Color=&LerpLinearColors(StartColor=<rgba 1 1 1 1>, EndColor=<rgba 2.5 2.5 2.5 1>, LerpFactor=$System.ClampedFloat)
     ))
    ScaleA=true
    ScaleRGB=true
   ) {module=ScaleColor002}
   &ScaleMeshSize(Scale Factor=&VectorFromFloat(Value=&MakeCustomFloatFromBool(Bool=$User.Change_OffPhysics, True Float=1, False Float=1)))
  ]
  rendererClasses [&NiagaraMeshRendererProperties]
  renderers [{rendererClass=&NiagaraMeshRendererProperties, rendererIndex=0}]
 }
 Run_Fire {
  emitterUpdateScript [
   &EmitterState(
    Enable Distance Culling=true
    Inactive Response=::NewEnumerator1
    Life Cycle Mode=::NewEnumerator1
    Loop Duration=1
    MaxDistance=2000
    Module.bUseMaxDistance=true
    Module.bUseMinDistance=false
    Scalability Mode=::NewEnumerator1
    UseLoopDelay=false
   )
   &SpawnPerUnit(
    Max Movement Threshold=1500
    Module.Use Max Movement Threshold=true
    Module.Use Movement Tolerance=false
    Spawn Spacing:NiagaraFloat=&Multiply_Float(
     A=7
     B=&Multiply_Float(
      A=&MakeCustomFloatFromBool(
       Bool=&SetBoolByFloatComparison(A=&VectorLength(Vector=$Engine.Owner.Velocity), B=100)
       True Float=1
       False Float=0
      )
      B=&MakeCustomFloatFromBool(
       Bool=&SetBoolByFloatComparison(A=&VectorLength(Vector=$Engine.Owner.Velocity), B=1200)
       True Float=0
       False Float=1
      )
     )
    )
    Use Spawn Probability=false
   )
  ]
  particleSpawnScript [
   &PartitionParticles(Number Of Partitions=11)
   &InitializeParticle(
    Lifetime Max=0.4
    Lifetime Min=0.3
    Lifetime Mode=::NewEnumerator1
    Sprite Rotation Mode=::NewEnumerator2
    Sprite Size Mode=::NewEnumerator1
    Uniform Sprite Size Max=15
    Uniform Sprite Size Min=6
    UsePositionOffset=false
    Write Material Random=true
   )
   &ShapeLocation(
    MaskSpawnGroup=false
    Offset Mode=::NewEnumerator0
    Override Seed=false
    Rotation Axis=<1 1 1>
    Rotation Mode=::NewEnumerator1
    Sphere Radius=5
    UseOwnerScale=true
   )
   &SkeletalMeshLocation(
    Bone / Socket Index=&SelectIntFromArray(
     Direct Array Index=$Particles.PartitionParticles.Partition
     Int Selection Array=DataInterface() {propertyValues {IntData [14,0,1,2,3,4,7,8,9,10,11]}}
    )
    Bone Sampling Mode=::NewEnumerator4
    MaskSpawnGroup=false
    Orientation Sampling=::NewEnumerator1
    Sampled Position Offset=&AddVectorToPosition(Position=<0 0 12>, Vector=$StackContext.ShapeLocation.ShapeVector)
    Skeletal Mesh=$System.Skeletal Mesh
   )
   &AddVelocity(Velocity=<0 -5 0>) {enabled=false}
   &AddVelocity(Module.Use Falloff Distance=false, Override Seed=false, Velocity Mode=::NewEnumerator1) {enabled=false, module=AddVelocity001}
  ]
  particleUpdateScript [
   &ParticleState()
   &ScaleSpriteSize(
    Uniform Curve Scale=1.4
    Uniform Curve Sprite Scale=DataInterface() {propertyValues {Curve {keys [
        {arriveTangent=3.1184, interpMode=RCIM_Cubic, leaveTangent=3.1184, tangentMode=RCTM_User, time=0, value=0}
        {interpMode=RCIM_Cubic, tangentMode=RCTM_User, time=1, value=1}
       ]}}}
   )
   &SpriteRotationRate(Rotation Rate=&RandomRangeFloat(Minimum=-100, Maximum=100))
   &CurlNoiseForce(Bias Noise=false, Noise Strength=100, Pan Noise=false)
   &Drag(Use Linear Drag=true, Use Rotational Drag=true) {enabled=false}
   &InheritSourceMovement(Applied Position Scale=<0.8 0.8 0.8>, ApplyForce=false, ApplyPosition=true, ApplyPositionFalloff=false)
   &SubUVAnimation(
    End Frame Range Override:NiagaraInt32=&RandomRangeInt(Minimum=12, Maximum=16)
    "Sprite Renderer "=DataInterface() {propertyValues {SpriteRenderer=&self:Run_Fire_0.NiagaraSpriteRendererProperties_1}}
    Start Frame Range Override=7
    UseEndFrame=true
    UseStartFrame=true
   )
   &ScaleColor(
    Scale Alpha=&FloatFromCurve(
     FloatCurve=DataInterface() {propertyValues {Curve {keys [
         <at 0 0 RCIM_Cubic>
         {arriveTangent=-0.0472, interpMode=RCIM_Cubic, leaveTangent=-0.0472, tangentMode=RCTM_User, time=0.2, value=1}
         {arriveTangent=-1.5121, interpMode=RCIM_Cubic, leaveTangent=-1.5121, tangentMode=RCTM_User, time=0.8, value=0.75}
         <at 1 0 RCIM_Cubic>
        ]}}}
     Scale Curve=&MakeCustomFloatFromBool(Bool=$User.bIsFrontEndPreview, True Float=0.8, False Float=&RandomRangeFloat(Minimum=0.5, Maximum=1))
    )
    Scale RGB=&MakeCustomVectorFromBool(
     Bool=$User.bIsFrontEndPreview
     True Vector=&MakeVectorFromLinearColorRGB(Color=&ColorFromCurve(ColorCurve=DataInterface() {propertyValues {
         AlphaCurve {keys []}
         BlueCurve {keys [<at 0.0717 0.0133>, {arriveTangent=-0.0287, interpMode=RCIM_Cubic, leaveTangent=-0.0287, time=0.4137, value=0.01}, <at 0.536 0>]}
         GreenCurve {keys [<at 0.0717 0.22>, {arriveTangent=-0.4631, interpMode=RCIM_Cubic, leaveTangent=-0.4631, time=0.4137, value=0.075}, <at 0.536 0.005>]}
         RedCurve {keys [<at 0.0717 1>, {arriveTangent=-2.0644, interpMode=RCIM_Cubic, leaveTangent=-2.0644, time=0.4137, value=0.5}, <at 0.536 0.0417>]}
        }}))
     False Vector=&MakeVectorFromLinearColorRGB(Color=&ColorFromCurve(ColorCurve=DataInterface() {propertyValues {
         AlphaCurve {keys []}
         BlueCurve {keys [<at 0.0717 0.0243>, {interpMode=RCIM_Cubic, leaveTangent=-0.0481, time=0.2309, value=0}, <at 0.536 0>]}
         GreenCurve {keys [<at 0.0717 0.3>, {interpMode=RCIM_Cubic, leaveTangent=-0.5825, time=0.2309, value=0.12}, <at 0.536 0.005>]}
         RedCurve {keys [<at 0.0717 1.5>, {interpMode=RCIM_Cubic, leaveTangent=-2.8796, time=0.2309, value=1}, <at 0.536 0.0417>]}
        }}))
    )
    ScaleA=true
    ScaleRGB=true
   )
   &ScaleColor(
    Scale Alpha=&FNTimeOfDayScalar(Night Scale=0.6, Morning Scale=0.7, Evening Scale=0.7)
    Scale RGB=&MakeVectorFromLinearColorRGB(Color=&FNTimeOfDayColor(Night Color=<rgba 1 0.9948 0.2694 1>, Morning Color=<rgba 1 1 1 1>, Day Color=<rgba 1 1 1 1>, Evening Color=<rgba 1 1 1 1>))
    ScaleA=true
    ScaleRGB=true
   ) {module=ScaleColor001}
   &SolveForcesAndVelocity(Clamp Velocity=false, Limit Acceleration=false)
  ]
  rendererClasses [&NiagaraSpriteRendererProperties]
  renderers [{rendererClass=&NiagaraSpriteRendererProperties}]
 }
 Run_Piece {
  emitterUpdateScript [
   &EmitterState(
    Enable Distance Culling=true
    Inactive Response=::NewEnumerator1
    Life Cycle Mode=::NewEnumerator1
    Loop Duration=1
    MaxDistance=2000
    Module.bUseMaxDistance=true
    Module.bUseMinDistance=false
    Scalability Mode=::NewEnumerator1
    UseLoopDelay=false
   )
   &SpawnPerUnit(
    Max Movement Threshold=1500
    Module.Use Max Movement Threshold=true
    Module.Use Movement Tolerance=false
    Spawn Spacing:NiagaraFloat=&Multiply_Float(
     A=12
     B=&Multiply_Float(
      A=&MakeCustomFloatFromBool(
       Bool=&SetBoolByFloatComparison(A=&VectorLength(Vector=$Engine.Owner.Velocity), B=100)
       True Float=1
       False Float=0
      )
      B=&MakeCustomFloatFromBool(
       Bool=&SetBoolByFloatComparison(A=&VectorLength(Vector=$Engine.Owner.Velocity), B=1200)
       True Float=0
       False Float=1
      )
     )
    )
    Use Spawn Probability=false
   )
  ]
  particleSpawnScript [
   &PartitionParticles(Number Of Partitions=11)
   &InitializeParticle(
    Lifetime Max=0.42
    Lifetime Min=0.3
    Lifetime Mode=::NewEnumerator1
    Sprite Size Max=<1 1.6>
    Sprite Size Min=<0.5 1>
    Sprite Size Mode=::NewEnumerator2
    UsePositionOffset=false
    Write Material Random=true
   )
   &ShapeLocation(
    MaskSpawnGroup=false
    Offset Mode=::NewEnumerator0
    Override Seed=false
    Rotation Axis=<1 1 1>
    Rotation Mode=::NewEnumerator1
    Sphere Radius=5
    UseOwnerScale=true
   )
   &SkeletalMeshLocation(
    Bone / Socket Index=&SelectIntFromArray(
     Direct Array Index=$Particles.PartitionParticles.Partition
     Int Selection Array=DataInterface() {propertyValues {IntData [14,0,1,2,3,4,7,8,9,10,11]}}
    )
    Bone Sampling Mode=::NewEnumerator4
    MaskSpawnGroup=false
    Orientation Sampling=::NewEnumerator1
    Sampled Position Offset=&AddVectorToPosition(Position=<0 0 12>, Vector=$StackContext.ShapeLocation.ShapeVector)
    Skeletal Mesh=$System.Skeletal Mesh
   )
   &AddVelocity(Velocity=<0 -5 0>) {enabled=false}
   &AddVelocity(Module.Use Falloff Distance=false, Override Seed=false, Velocity Mode=::NewEnumerator1, Velocity Speed=100) {enabled=false, module=AddVelocity001}
  ]
  particleUpdateScript [
   &ParticleState()
   &CurlNoiseForce(Bias Noise=false, Noise Strength=150, Pan Noise=false)
   &Drag(Use Linear Drag=true, Use Rotational Drag=true) {enabled=false}
   &InheritSourceMovement(Applied Position Scale=<0.8 0.8 0.8>, ApplyForce=false, ApplyPosition=true, ApplyPositionFalloff=false)
   &SolveForcesAndVelocity(Clamp Velocity=false, Limit Acceleration=false)
   &ScaleColor(
    Scale Alpha=&FloatFromCurve(FloatCurve=DataInterface() {propertyValues {Curve {keys [<at 0 0>,<at 0.2 1>,<at 0.7 1>,<at 1 0>]}}})
    Scale RGB=&MakeVectorFromLinearColorRGB(Color=&ColorFromCurve(ColorCurve=DataInterface() {propertyValues {
        AlphaCurve {keys []}
        BlueCurve {keys [<at 0.1559 0 RCIM_Cubic>,<at 0.5281 0 RCIM_Cubic>]}
        GreenCurve {keys [<at 0.1559 0.8 RCIM_Cubic>,<at 0.5281 0.15 RCIM_Cubic>]}
        RedCurve {keys [<at 0.1559 3 RCIM_Cubic>,<at 0.5281 1 RCIM_Cubic>]}
       }}))
    ScaleA=true
    ScaleRGB=true
   )
   /SetVariables(Particles.Glow03_Size=&Multiply_Vector2D_ByFloat(
     Float=&FloatFromCurve(
      FloatCurve=DataInterface() {propertyValues {Curve {keys [
          <at 0 0 RCIM_Cubic>
          <at 0.2 1.3 RCIM_Cubic>
          {arriveTangent=-0.1423, interpMode=RCIM_Cubic, leaveTangent=-0.1423, tangentMode=RCTM_User, time=1, value=0}
         ]}}}
      Scale Curve=5
     )
     Vector2D=<0.5 1>
    )) {isSetParameters=true}
   /SetVariables(Particles.Glow03_Color=&MakeLinearColorFromVectorAndFloat(
     "Vector (RGB)"=&MakeVectorFromLinearColorRGB(Color=&MultiplyLinearColor(
       Linear Color=<rgba 1 0.2 0.02 0.2>
       Scale Factor=&MakeVector4_FromFloat(Float=&Multiply_Float(A=1, B=&FNTimeOfDayScalar(Night Scale=0.5, Morning Scale=1, Evening Scale=1)))
      ))
     "Float (Alpha)"=&FloatFromCurve(
      FloatCurve=DataInterface() {propertyValues {Curve {keys [
          <at 0 0 RCIM_Cubic>
          {arriveTangent=-0.0051, interpMode=RCIM_Cubic, leaveTangent=-0.0051, tangentMode=RCTM_User, time=0.2, value=1}
          {arriveTangent=0.0387, interpMode=RCIM_Cubic, leaveTangent=0.0387, tangentMode=RCTM_User, time=0.6, value=1}
          {arriveTangent=-0.1423, interpMode=RCIM_Cubic, leaveTangent=-0.1423, tangentMode=RCTM_User, time=1, value=0}
         ]}}}
      Scale Curve=0.5
     )
    )) {isSetParameters=true}
  ]
  rendererClasses [&NiagaraSpriteRendererProperties]
  renderers [
   {rendererClass=&NiagaraSpriteRendererProperties}
   {rendererClass=&NiagaraSpriteRendererProperties, rendererIndex=1}
  ]
 }
 SmokeTrail {
  emitterUpdateScript [
   &EmitterState(
    Inactive Response=::NewEnumerator1
    Life Cycle Mode=::NewEnumerator1
    Loop Behavior=::NewEnumerator1
    Loop Duration Mode=::NewEnumerator1
    UseLoopDelay=false
   )
   &SpawnBurst_Instantaneous(
    Spawn Count=&MakeCustomIntFromBool(Bool=$User.Change_OffPhysics, True INT=7, False INT=13)
    Use Spawn Probability=false
    UseLoopCountLimit=false
   )
  ]
  particleSpawnScript [
   &PartitionParticles(Number Of Partitions=&MakeCustomIntFromBool(Bool=$User.Change_OffPhysics, True INT=7, False INT=13))
   &InitializeParticle(
    Color Maximum=<rgba 3 2.3806 2.125 1>
    Color Minimum=<rgba 0.1562 0.1562 0.1562 0.8>
    Color Mode=::NewEnumerator2
    Lifetime=5
    Mesh Scale=&MakeCustomVectorFromBool(
     Bool=$User.Change_OffPhysics
     True Vector=&SelectVectorFromArray(
      Direct Array Index=$Particles.PartitionParticles.Partition
      Vector Selection Array=DataInterface() {propertyValues {FloatData [<1 1 1.2>, <1 1 2>, <1 1 1.2>, <1 1 0.8>, <1 1 1>, <1 1 1>, <1 1 1>]}}
     )
     False Vector=&SelectVectorFromArray(
      Direct Array Index=$Particles.PartitionParticles.Partition
      Vector Selection Array=DataInterface() {propertyValues {FloatData [<1 1 1.2>, <1 1 2>, <1 1 1.2>, <1 1 1>, <1 1 0.7>, <1 1 0.5>, <1 1 0.3>, <1 1 0.5>, <1 1 0.65>, <1 1 1>, <1 1 1>, <1 1 0.7>, <1 1 1>]}}
     )
    )
    Mesh Scale Mode=::NewEnumerator4
    UsePositionOffset=false
    Write Material Random=true
   )
   &InitialMeshOrientation(
    Mesh Orientation Mode=::NewEnumerator2
    Rotation=&MakeCustomVectorFromBool(
     Bool=$User.Change_OffPhysics
     True Vector=<0 0.25 0>
     False Vector=&SelectVectorFromArray(
      Direct Array Index=$Particles.PartitionParticles.Partition
      Vector Selection Array=DataInterface() {propertyValues {FloatData [
         <0.01 0.01 0>
         <0 0.018 0>
         <0 0.02 0>
         <0 0 0>
         <0 -0.01 0>
         <0 -0.012 0>
         <0 -0.01 0>
         <-0.01 0 0>
         <-0.01 0 0>
         <0 -0.01 0>
         <0 0.03 0>
         <-0.02 -0.02 0>
         <-0.01 -0.01 0>
        ]}}
     )
    )
    Use Rotation Vector=true
   )
   &CameraOffset(Camera Offset Amount=&MakeCustomFloatFromBool(Bool=$User.Change_OffPhysics, True Float=5, False Float=2))
  ]
  particleUpdateScript [
   &ParticleState(Kill Particles When Lifetime Has Elapsed=false, Let Infinitely Lived Particles Die When Emitter Deactivates=true)
   &SkeletalMeshLocation(
    Bone / Socket Index=&MakeCustomIntFromBool(
     Bool=$User.Change_OffPhysics
     True INT=&SelectIntFromArray(
      Direct Array Index=$Particles.PartitionParticles.Partition
      Int Selection Array=DataInterface() {propertyValues {IntData [15,15,15,16,16,16,16]}}
     )
     False INT=&SelectIntFromArray(
      Direct Array Index=$Particles.PartitionParticles.Partition
      Int Selection Array=DataInterface() {propertyValues {IntData [0,1,2,3,4,5,6,7,8,9,10,12,14]}}
     )
    )
    Bone Sampling Mode=::NewEnumerator4
    MaskSpawnGroup=false
    Sampled Position Offset=&MakeCustomVectorFromBool(
     Bool=$User.Change_OffPhysics
     True Vector=&MakeCustomVectorFromBool(
      Bool=$User.bIsFrontEndPreview
      True Vector=&SelectVectorFromArray(
       Direct Array Index=$Particles.PartitionParticles.Partition
       Vector Selection Array=DataInterface() {propertyValues {FloatData [<-60 -15 0>, <-40 -15 6>, <-40 -12 16>, <5 5 10>, <5 5 0>, <-8 10 0>, <-5 0 -9>]}}
      )
      False Vector=&SelectVectorFromArray(
       Direct Array Index=$Particles.PartitionParticles.Partition
       Vector Selection Array=DataInterface() {propertyValues {FloatData [<-60 -15 0>, <-40 -15 6>, <-40 -12 16>, <5 5 10>, <5 5 0>, <-8 15 0>, <-5 0 -12>]}}
      )
     )
     False Vector=&SelectVectorFromArray(
      Direct Array Index=$Particles.PartitionParticles.Partition
      Vector Selection Array=DataInterface() {propertyValues {FloatData [<-1 0 0>, <0 0 0>, <0 0 0>, <0 0 0>, <0 0 0>, <0 0 0>, <-1 0 0>, <0 0 -2>, <1 1 -2>, <2 2 0>, <0 0 0>, <0 0 0>, <3 0 0>]}}
     )
    )
    Skeletal Mesh=$System.Skeletal Mesh
   )
   &InterpolateOverTime(Rate Of Change=8, Target Vector Value:NiagaraPosition=$Particles.Position, Variable Type=::NewEnumerator2)
   /SetVariables(Particles.InterpolateDY:Vector4f=&MakeFloat4FromVectorAndFloat(
     "Vector (RGB)"=$StackContext.InterpolateOverTime.InterpolateOverTime_V3_.Moving Average
     "Float (Alpha)"=&MakeCustomFloatFromBool(Bool=$User.Change_OffPhysics, True Float=0, False Float=0.3)
    )) {isSetParameters=true, module=SetVariables_#3}
   &DynamicMaterialParameters(
    Index 0 Param 1=&RandomRangeFloat(Minimum=0.5, Maximum=1.2)
    Index 0 Param 2=&RandomRangeFloat(Minimum=0.3, Maximum=1)
    Index 0 Param 3=&RandomRangeFloat(Minimum=0.5, Maximum=1)
    Index 0 Param 4=&RandomRangeFloat(Minimum=0.5, Maximum=1)
    Index 3 Param 1=&MakeCustomFloatFromBool(
     Bool=$User.Change_OffPhysics
     True Float=10
     False Float=&SelectFloatFromArray(
      Direct Array Index=$Particles.PartitionParticles.Partition
      Float Selection Array=DataInterface() {propertyValues {FloatData [5,8,8,5,2,1,2,4,10,3,13,2,6]}}
     )
    )
    Index 3 Param 2=&MakeCustomFloatFromBool(
     Bool=$User.Change_OffPhysics
     True Float=15
     False Float=&SelectFloatFromArray(
      Direct Array Index=$Particles.PartitionParticles.Partition
      Float Selection Array=DataInterface() {propertyValues {FloatData [10,15,17,13,8,10,11,12,10,16,13,6,10]}}
     )
    )
    Index 3 Param 3=&RandomRangeFloat(Minimum=3, Maximum=5)
    Index 3 Param 4=&RandomRangeFloat(Minimum=0.5, Maximum=1)
    Module.Index0Float0Write=true
    Module.Index0Float1Write=true
    Module.Index0Float2Write=true
    Module.Index0Float3Write=true
    Module.Index3Float0Write=true
    Module.Index3Float1Write=true
    Module.Index3Float2Write=true
    Module.Index3Float3Write=true
    Write Parameter Index 3=::NewEnumerator1
   )
   /SetVariables(Particles.Dyn002:Vector4f=&MakeVector4(
     X=&MakeCustomFloatFromBool(
      Bool=$User.Change_OffPhysics
      True Float=10
      False Float=&SelectFloatFromArray(
       Direct Array Index=$Particles.PartitionParticles.Partition
       Float Selection Array=DataInterface() {propertyValues {FloatData [10,15,30,20,2,1,2,4,10,3,13,1,5]}}
      )
     )
     Y=&MakeCustomFloatFromBool(
      Bool=$User.Change_OffPhysics
      True Float=12
      False Float=&SelectFloatFromArray(
       Direct Array Index=$Particles.PartitionParticles.Partition
       Float Selection Array=DataInterface() {propertyValues {FloatData [10,15,17,13,10,10,11,15,5,20,15,10,20]}}
      )
     )
     Z=&RandomRangeFloat(Minimum=4, Maximum=7)
     W=&RandomRangeFloat(Minimum=0.5, Maximum=0.8)
    )) {isSetParameters=true}
   /SetVariables(Particles.Size002:Vector3f=&MakeCustomVectorFromBool(
     Bool=$User.Change_OffPhysics
     True Vector=&SelectVectorFromArray(
      Direct Array Index=$Particles.PartitionParticles.Partition
      Vector Selection Array=DataInterface() {propertyValues {FloatData [<1 1 1.5>, <1 1 1.5>, <1 1 1.2>, <1 1 1.2>, <1 1 1.5>, <1 1 1.3>, <1 1 1>]}}
     )
     False Vector=&SelectVectorFromArray(
      Direct Array Index=$Particles.PartitionParticles.Partition
      Vector Selection Array=DataInterface() {propertyValues {FloatData [<1 1 1.6>, <1 1 2>, <1 1 2.2>, <1 1 1.6>, <1 1 1.2>, <1 1 1.2>, <1 1 0.8>, <1 1 1>, <1 1 1>, <1 1 1.3>, <1 1 1.5>, <1 1 1.2>, <1 1 1.6>]}}
     )
    )) {isSetParameters=true}
   /SetVariables(Particles.Size002_02:Vector3f=&Multiply_Vector(
     A=$Particles.Size002
     B=&VectorFromFloat(Value=&MakeCustomFloatFromBool(Bool=$User.Change_OffPhysics, True Float=1, False Float=1))
    )) {isSetParameters=true}
   &ScaleColor(Scale Alpha=&RandomRangeFloat(Minimum=0.5, Maximum=0.6), ScaleA=true, ScaleRGB=true)
   /SetVariables(Particles.Color003:LinearColor=&MakeLinearColorFromVectorAndFloat(
     "Vector (RGB)"=&MakeVectorFromLinearColorRGB(Color=&RandomRangeLinearColor(Minimum=<rgba 0.1562 0.1562 0.1562 0.8>, Maximum=<rgba 1 0.7935 0.7083 1>))
     "Float (Alpha)"=&MakeCustomFloatFromBool(Bool=$User.bIsFrontEndPreview, True Float=0.7, False Float=&Lerp_Float(A=0.4, B=0.7, Alpha=$System.ClampedFloat))
    )) {isSetParameters=true}
   &ScaleMeshSize(Scale Factor=&VectorFromFloat(Value=&MakeCustomFloatFromBool(Bool=$User.Change_OffPhysics, True Float=1, False Float=1)))
  ]
  rendererClasses [&NiagaraMeshRendererProperties]
  renderers [
   {rendererClass=&NiagaraMeshRendererProperties, rendererIndex=0}
   {rendererClass=&NiagaraMeshRendererProperties, rendererIndex=1}
   {rendererClass=&NiagaraMeshRendererProperties, rendererIndex=2}
  ]
 }
}
summary {
 systemName=NS_Character_FlameBride_Idle
 emitters [
  {emitterName=SmokeTrail, rendererClasses [&NiagaraMeshRendererProperties]}
  {emitterName=FireTrail, rendererClasses [&NiagaraMeshRendererProperties]}
  {emitterName=FirePiece, rendererClasses [&NiagaraSpriteRendererProperties]}
  {emitterName=Dirt, rendererClasses [&NiagaraSpriteRendererProperties]}
  {emitterName=Run_Piece, rendererClasses [&NiagaraSpriteRendererProperties]}
  {emitterName=Run_Fire, rendererClasses [&NiagaraSpriteRendererProperties]}
 ]
 userVariables [
  User.bIsFrontEndPreview {type=NiagaraBool}
  User.Change_OffPhysics {type=NiagaraBool}
  User.FXAmount {type=NiagaraFloat, defaultValue=0, description=""}
  User.PlayerSkeletalMesh {
   type=NiagaraDataInterfaceSkeletalMesh
   defaultValue {
    struct=NiagaraExt_VariableValue_DataInterface
    value {dataInterface=&self:NiagaraDataInterfaceSkeletalMesh_0}
   }
  }
 ]
}
