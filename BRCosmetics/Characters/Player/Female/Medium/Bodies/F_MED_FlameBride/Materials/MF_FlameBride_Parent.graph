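# Material function for the Flame Bride character that layers a reactive burning effect onto incoming material attributes.
# Ember Time and Burn Time parameters, gated by a reactivity toggle, serve as effect drivers for thresholds on a packed burn mask, producing flame and singe erosion.
# Panned turbulence feeds a blackbody emissive with inverse eye adaptation and a separate Lumen contribution scale for ray tracing.
# Masks blend emissive, burn color, opacity and opacity mask.
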MaterialFunction /BRCosmetics/Characters/Player/Female/Medium/Bodies/F_MED_FlameBride/Materials/MF_FlameBride_Parent
graph {nodes [
  n1 = GetMaterialAttributes(None=FunctionInput() {inputName=Result, inputType=FunctionInput_MaterialAttributes}) {attributeGetTypes [
    #1
    #2
    #3
    #4
   ]}
  n2 = NamedRerouteUsage()
  n3 = Multiply(A=@n2, B=ScalarParameter() {parameterName=Lumen Contribution Factor})
  n4 = NamedRerouteUsage()
  n5 = NamedRerouteUsage()
  n6 = Subtract(A=NamedRerouteUsage(), B=NamedRerouteUsage())
  n7 = SetMaterialAttributes!(
   MaterialAttributes=@n1.MaterialAttributes
   EmissiveColor=LinearInterpolate(A=@n1.EmissiveColor, B=RayTracingQualitySwitch(Normal=@n2, RayTraced=@n3), Alpha=NamedRerouteUsage())
   BaseColor=LinearInterpolate(A=NamedRerouteUsage(), B=VectorParameter() {group=Flame Bride, parameterName=Burn Color}.RGB, Alpha=NamedRerouteUsage()) {constB=0}
   Opacity=LinearInterpolate(A=LinearInterpolate(A=@n4, Alpha=@n5) {constB=0}, Alpha=Saturate(None=@n6))
   OpacityMask=LinearInterpolate(A=NamedRerouteUsage(), Alpha=NamedRerouteUsage()) {constB=0}
  ) {attributeSetTypes [
    #1
    #2
    #3
    #4
   ]}
  FunctionOutput(None=@n7) {outputName=SetMaterialAttributes}
  NamedRerouteDeclaration(None=LinearInterpolate(
    B=ScalarParameter() {group=Flame Bride, parameterName=Ember Time}
    Alpha=ScalarParameter() {defaultValue=1, group=Flame Bride, parameterName=Enable Reactivity?}
   )) {name=Ember - Reactivity Toggle, nodeColor=<rgba 0.0547 0 1 1>}
  NamedRerouteDeclaration(None=LinearInterpolate(
    B=ScalarParameter() {group=Flame Bride, parameterName=Burn Time}
    Alpha=ScalarParameter() {defaultValue=1, group=Flame Bride, parameterName=Enable Reactivity?}
   )) {name=Burn - Reactivity Toggle, nodeColor=<rgba 0.0547 0 1 1>}
  n8 = NamedRerouteUsage()
  n9 = LinearInterpolate(
   A=ScalarParameter() {defaultValue=0.1, group=Flame Bride, parameterName=Flame Step Min}
   B=ScalarParameter() {defaultValue=-0.6, group=Flame Bride, parameterName=Flame Step Max}
   Alpha=@n8
  )
  n10 = Add(
   A=@n9
   B=LinearInterpolate(
    A=ScalarParameter() {defaultValue=1.1, group=Flame Bride, parameterName=Flame Smoothness Min}
    B=ScalarParameter() {defaultValue=1.2, group=Flame Bride, parameterName=Flame Smoothness Max}
    Alpha=@n8
   )
  )
  n11 = TextureSampleParameter2D!(UVs=TextureCoordinate(), Bias=ScalarParameter() {defaultValue=1, group=Flame Bride, parameterName=Bias}, Apply View MipBias) {
   group=Flame Bride
   mipValueMode=TMVM_MipBias
   parameterName=Burn Mask
   samplerType=SAMPLERTYPE_Masks
   texture=/BRCosmetics/Characters/Player/Female/Medium/Bodies/F_MED_FlameBride/Textures/T_FlameBride_FaceAcc_FX2
  }
  n12 = ScalarParameter() {defaultValue=0.2, group=Flame Bride, parameterName=Flame Removal}
  n13 = ScalarParameter() {group=Flame Bride, parameterName=Use Mask Removal?}
  n14 = LinearInterpolate(B=Multiply(A=@n11.R, B=@n12), Alpha=@n13) {constA=1, constAlpha=0}
  n15 = Saturate(None=Multiply(A=SmoothStep(Min=@n9, Max=@n10, Value=@n11.G), B=@n14))
  n16 = TextureCoordinate()
  n17 = ScalarParameter() {defaultValue=10, group=Flame Bride, parameterName=Tiling}
  n18 = TextureSample(UVs=Panner(Coordinate=Multiply(A=@n16, B=@n17), Speed=ScalarParameter() {defaultValue=0.1, group=Flame Bride, parameterName=Speed}) {speedX=0.1, speedY=0.1}) {texture=/Game/Global/Textures/Default/Caustics/T_Turbulence_Seamless}
  n19 = ScalarParameter() {defaultValue=0.2, group=Flame Bride, parameterName=Min Noise}
  n20 = Multiply(
   A=Power(Base=Clamp(None=@n18.RGB, Min=@n19) {minDefault=0.5}, Exp=ScalarParameter() {defaultValue=0.5, group=Flame Bride, parameterName=Exp})
   B=ScalarParameter() {defaultValue=2, group=Flame Bride, parameterName=NOise Intensity}
  )
  n21 = NamedRerouteDeclaration(None=Multiply(A=@n15, B=Add(A=@n15, B=@n20))) {name=Emissive Area Mask, nodeColor=<rgba 1 0 0.2578 1>}
  n22 = LinearInterpolate(
   A=ScalarParameter() {defaultValue=3000, group=Flame Bride, parameterName=Temperture Min}
   B=ScalarParameter() {defaultValue=3000, group=Flame Bride, parameterName=Temperture Max}
   Alpha=NamedRerouteUsage()
  )
  n23 = Multiply(A=BlackBody!(Temp=Multiply(A=@n21, B=@n22)), B=ScalarParameter() {defaultValue=1, group=Flame Bride, parameterName=Flame Boost})
  n24 = LinearInterpolate(
   A=ScalarParameter() {defaultValue=1, group=Flame Bride, parameterName=Velocity Intensity Min}
   B=ScalarParameter() {defaultValue=50, group=Flame Bride, parameterName=Velocity Intensity Max}
   Alpha=NamedRerouteUsage()
  )
  NamedRerouteDeclaration(None=Multiply(A=Multiply(A=@n23, B=@n24), B=EyeAdaptationInverse().EyeAdaptationInverse)) {name=Emissive Out, nodeColor=<rgba 0 0.0156 1 1>}
  n25 = LinearInterpolate(
   A=ScalarParameter() {defaultValue=0.8, group=Flame Bride, parameterName=Difference Min}
   B=ScalarParameter() {defaultValue=1.1, group=Flame Bride, parameterName=Difference Max}
   Alpha=NamedRerouteUsage()
  )
  n26 = SmoothStep(Min=@n25, Max=@n25, Value=@n11.G) {constMax=0}
  NamedRerouteDeclaration(None=Saturate(None=Multiply(A=@n14, B=@n26))) {name=Flame Erosion, nodeColor=<rgba 0.3359 1 0 1>}
  NamedRerouteDeclaration(None=@n11.B) {name=Singed Mask, nodeColor=<rgba 1 0 0.6562 1>}
  NamedRerouteDeclaration(None=@n1.OpacityMask) {name=Opacity Mask In, nodeColor=<rgba 0 0.9766 1 1>}
  NamedRerouteDeclaration(None=@n1.Opacity) {name=Opacity In, nodeColor=<rgba 1 0.7031 0 1>}
  NamedRerouteDeclaration(None=@n1.BaseColor) {name=Base Color In, nodeColor=<rgba 0.3828 0 1 1>}
  n27 = LinearInterpolate(B=Multiply(A=@n11.B, B=@n12), Alpha=@n13) {constA=1, constAlpha=0}
  n28 = LinearInterpolate(
   A=ScalarParameter() {defaultValue=0.5, group=Flame Bride, parameterName=Burn Difference Min}
   B=ScalarParameter() {defaultValue=1.5, group=Flame Bride, parameterName=Burn Difference Max}
   Alpha=NamedRerouteUsage()
  )
  n29 = SmoothStep(Min=@n28, Max=Add(A=@n28, B=ScalarParameter() {parameterName=Burn Smoothness}), Value=@n11.G)
  NamedRerouteDeclaration(None=Saturate(None=Multiply(A=@n27, B=@n29))) {name=Singe Erosion, nodeColor=<rgba 0 1 0.0625 1>}
 ]}
