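# Physics asset for the dress and belt of the female medium FlameBride player body.
# Four limited-swing dress chains hang from the pelvis or, at left front, the left thigh, with cross-constraints between neighboring lower segments.
# Belt ends and a three-segment strap hang from a belt base on spine_02.
# Leg, pelvis and spine shapes provide body collision, while a huge sunken root sphere approximates a floor.
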PhysicsAsset /BRCosmetics/Characters/Player/Female/Medium/Bodies/F_MED_FlameBride/Meshes/F_MED_FlameBride_Physics_dress
bodies {
 calf_l [{shapeType=Capsule, length=42.4715, radius=8, shapeName=calf_l_capsule, center=<-29.5 -4 0>, rotation=<rot 0 90 -90>}]
 calf_r [{shapeType=Capsule, length=42.4715, radius=8, shapeName=calf_l_capsule, center=<22 4 0>, rotation=<rot 0 90 90>}]
 dyn_belt_base [{shapeType=Capsule, length=2, radius=1, shapeName=dyn_belt_base_capsule}]
 dyn_belt_l [{shapeType=Capsule, length=2, radius=1, shapeName=dyn_belt_l_capsule, center=<5 0 0>, rotation=<rot 0 0 90>}]
 dyn_belt_l_base [{shapeType=Capsule, length=2, radius=1, shapeName=dyn_belt_l_base_capsule}]
 dyn_belt_r [{shapeType=Capsule, length=2, radius=1, shapeName=dyn_belt_r_capsule, center=<5 0 0>, rotation=<rot 0 0 90>}]
 dyn_belt_r_base [{shapeType=Capsule, length=2, radius=1, shapeName=dyn_belt_r_base_capsule}]
 dyn_belt_strap_01 [{shapeType=Capsule, length=5, radius=2.5, shapeName=dyn_belt_strap_01_capsule, center=<5 0 0>, rotation=<rot 90 0 0>}]
 dyn_belt_strap_02 [{shapeType=Capsule, length=2, radius=2, shapeName=dyn_belt_strap_02_capsule, center=<4.9879 0 0>, rotation=<rot 90 180 -90.0001>}]
 dyn_belt_strap_03 [{shapeType=Capsule, length=6.5, radius=4, shapeName=dyn_belt_strap_03_capsule, center=<8 0 0>, rotation=<rot 85 90.0011 -179.9994>}]
 dyn_belt_strap_base [
  {shapeType=Capsule, length=2, radius=2, shapeName=dyn_belt_strap_base_capsule}
  {shapeType=Capsule, length=5.5, radius=15, shapeName=dyn_belt_strap_base_capsule, center=<6.7924 1.4176 -68.5912>, rotation=<rot 0 0 90>}
 ]
 dyn_dress_01_l_bk [{shapeType=Capsule, length=16, radius=5, shapeName=dyn_dress_01_l_bk_capsule, center=<8 0 0>, rotation=<rot 0 0 90>}]
 dyn_dress_01_l_fr [{shapeType=Capsule, length=17.2317, radius=6, shapeName=dyn_dress_01_l_fr_capsule, center=<8 0 0>, rotation=<rot 0 0 90>}]
 dyn_dress_01_l_side [{shapeType=Capsule, length=16, radius=6, shapeName=dyn_dress_01_l_side_capsule, center=<8 0 0>, rotation=<rot 0 0 90>}]
 dyn_dress_01_r_bk [{shapeType=Capsule, length=16, radius=5, shapeName=dyn_dress_01_r_bk_capsule, center=<8 0 0>, rotation=<rot 0 0 90>}]
 dyn_dress_02_l_bk [{shapeType=Capsule, length=16, radius=5, shapeName=dyn_dress_02_l_bk_capsule, center=<8 0 0>, rotation=<rot 0 0 90>}]
 dyn_dress_02_l_fr [{shapeType=Capsule, length=16, radius=6, shapeName=dyn_dress_02_l_fr_capsule, center=<8 0 0>, rotation=<rot 0 0 90>}]
 dyn_dress_02_l_side [{shapeType=Capsule, length=16, radius=6, shapeName=dyn_dress_02_l_side_capsule, center=<8 0 0>, rotation=<rot 0 0 90>}]
 dyn_dress_02_r_bk [{shapeType=Capsule, length=16, radius=5, shapeName=dyn_dress_02_r_bk_capsule, center=<8 0 0>, rotation=<rot 0 0 90>}]
 dyn_dress_03_l_bk [{shapeType=Capsule, length=16, radius=5, shapeName=dyn_dress_03_l_bk_capsule, center=<8 0 0>, rotation=<rot 0 0 90>}]
 dyn_dress_03_l_fr [{shapeType=Capsule, length=16, radius=6, shapeName=dyn_dress_03_l_fr_capsule, center=<8 0 0>, rotation=<rot 0 0 90>}]
 dyn_dress_03_l_side [{shapeType=Capsule, length=16, radius=5, shapeName=dyn_dress_03_l_side_capsule, center=<8 0 0>, rotation=<rot 0 0 90>}]
 dyn_dress_03_r_bk [{shapeType=Capsule, length=16, radius=5, shapeName=dyn_dress_03_r_bk_capsule, center=<8 0 0>, rotation=<rot 0 0 90>}]
 dyn_dress_04_l_bk [{shapeType=Capsule, length=36.9022, radius=5, shapeName=dyn_dress_04_l_bk_capsule, center=<8 0 0>, rotation=<rot 0 0 90>}]
 dyn_dress_04_l_fr [{shapeType=Capsule, length=16, radius=6, shapeName=dyn_dress_04_l_fr_capsule, center=<8 0 0>, rotation=<rot 0 0 90>}]
 dyn_dress_04_l_side [{shapeType=Capsule, length=36.9022, radius=5, shapeName=dyn_dress_04_l_side_capsule, center=<8 0 0>, rotation=<rot 0 0 90>}]
 dyn_dress_04_r_bk [{shapeType=Capsule, length=36.9022, radius=5, shapeName=dyn_dress_04_r_bk_capsule, center=<8 0 0>, rotation=<rot 0 0 90>}]
 dyn_dress_l_bk_base [{shapeType=Capsule, length=16, radius=5, shapeName=dyn_dress_l_bk_base_capsule, center=<8 0 0>, rotation=<rot 0 0 90>}]
 dyn_dress_l_fr_base [{shapeType=Capsule, length=1, radius=1, shapeName=dyn_dress_l_fr_base_capsule, rotation=<rot 0 0 90>}]
 dyn_dress_l_side_base [{shapeType=Capsule, length=1, radius=1, shapeName=dyn_dress_l_side_base_capsule, rotation=<rot 0 0 90>}]
 dyn_dress_r_bk_base [{shapeType=Capsule, length=16, radius=5, shapeName=dyn_dress_r_bk_base_capsule, center=<8 0 0>, rotation=<rot 0 0 90>}]
 dyn_root_floor [{shapeType=Capsule, length=10, radius=20, shapeName=dyn_root_floor_capsule, rotation=<rot 0 -90 0>}]
 dyn_strap_collision_l [
  {shapeType=Capsule, length=0.505, radius=0.505, shapeName=dyn_strap_collision_l_capsule, rotation=<rot 0 -90 0>}
  {
   shapeType=Capsule
   length=20.5
   radius=6.5
   shapeName=deform_groin_base_l_capsule
   center=<-20.2039 3.4465 21.5246>
   rotation=<rot 85 -0.0001 -179.9999>
  }
 ]
 pelvis [
  {shapeType=Capsule, length=6, radius=9, center=<-1.7446 2.3079 0>}
  {shapeType=Capsule, length=8, radius=9, shapeName=pelvis_capsule, center=<-0.6396 0.1353 0>}
 ]
 root [{shapeType=Sphere, radius=400, shapeName=root_sphere, center=<0 0 -394>}, {shapeType=Sphere, radius=40, shapeName=root_sphere}]
 spine_01 [{shapeType=Capsule, length=6, radius=8, center=<0 1.4539 0>}]
 spine_02 [{shapeType=Capsule, length=15.1187, radius=7.5852, shapeName=spine_02_capsule}]
 thigh_l [
  {shapeType=Capsule, length=26.6739, radius=6.9677, center=<-22.1982 -0.2153 2.2717>, rotation=<rot 0 -87.6288 -97.9391>}
  {
   shapeType=Capsule
   length=26.6739
   radius=6.9677
   shapeName=thigh_l_capsule
   center=<-21.2575 -0.5904 8.6345>
   rotation=<rot 0 -87.6288 -97.9391>
  }
  {
   shapeType=Capsule
   length=3.1739
   radius=6.9677
   shapeName=thigh_l_capsule
   center=<-34.8132 -3.0822 -0.8959>
   rotation=<rot 0 -87.6288 -97.9391>
  }
 ]
 thigh_r [{shapeType=Capsule, length=25.6416, radius=8, center=<21.9942 0.2068 -0.8078>, rotation=<rot 0 -87.6286 82.0609>}]
}
constraints [
 {bone1Name=dyn_dress_04_r_bk, bone2Name=dyn_dress_03_r_bk, swing1LimitDegrees=30, swing2LimitDegrees=30, twistLimitDegrees=5}
 {
  bone1Name=dyn_dress_l_bk_base
  bone2Name=pelvis
  swing1LimitDegrees=45
  swing1Motion=Free
  swing2LimitDegrees=45
  swing2Motion=Free
  twistLimitDegrees=45
  twistMotion=Free
 }
 {bone1Name=dyn_dress_01_l_bk, bone2Name=dyn_dress_l_bk_base, swing1LimitDegrees=10, swing2LimitDegrees=30, twistLimitDegrees=5}
 {bone1Name=dyn_dress_02_l_bk, bone2Name=dyn_dress_01_l_bk, swing1LimitDegrees=10, swing2LimitDegrees=20, twistLimitDegrees=5}
 {bone1Name=dyn_dress_03_l_bk, bone2Name=dyn_dress_02_l_bk, swing1LimitDegrees=10, swing2LimitDegrees=15, twistLimitDegrees=5}
 {bone1Name=dyn_dress_04_l_bk, bone2Name=dyn_dress_03_l_bk, swing1LimitDegrees=30, swing2LimitDegrees=30, twistLimitDegrees=5}
 {
  bone1Name=dyn_dress_l_side_base
  bone2Name=pelvis
  swing1LimitDegrees=45
  swing1Motion=Free
  swing2LimitDegrees=45
  swing2Motion=Free
  twistLimitDegrees=45
  twistMotion=Free
 }
 {bone1Name=dyn_dress_01_l_side, bone2Name=dyn_dress_l_side_base, swing1LimitDegrees=10, swing2LimitDegrees=30, twistLimitDegrees=5}
 {bone1Name=dyn_dress_02_l_side, bone2Name=dyn_dress_01_l_side, swing1LimitDegrees=10, swing2LimitDegrees=20, twistLimitDegrees=5}
 {bone1Name=dyn_dress_03_l_side, bone2Name=dyn_dress_02_l_side, swing1LimitDegrees=10, swing2LimitDegrees=15, twistLimitDegrees=5}
 {bone1Name=dyn_dress_04_l_side, bone2Name=dyn_dress_03_l_side, swing1LimitDegrees=30, swing2LimitDegrees=30, twistLimitDegrees=5}
 {
  bone1Name=dyn_dress_r_bk_base
  bone2Name=pelvis
  swing1LimitDegrees=45
  swing1Motion=Free
  swing2LimitDegrees=45
  swing2Motion=Free
  twistLimitDegrees=45
  twistMotion=Free
 }
 {bone1Name=dyn_dress_01_r_bk, bone2Name=dyn_dress_r_bk_base, swing1LimitDegrees=10, swing2LimitDegrees=30, twistLimitDegrees=5}
 {bone1Name=dyn_dress_02_r_bk, bone2Name=dyn_dress_01_r_bk, swing1LimitDegrees=10, swing2LimitDegrees=30, twistLimitDegrees=5}
 {bone1Name=dyn_dress_03_r_bk, bone2Name=dyn_dress_02_r_bk, swing1LimitDegrees=10, swing2LimitDegrees=15, twistLimitDegrees=5}
 {bone1Name=dyn_dress_04_l_fr, bone2Name=dyn_dress_03_l_fr, swing1LimitDegrees=30, swing2LimitDegrees=30, twistLimitDegrees=5}
 {
  bone1Name=dyn_dress_l_fr_base
  bone2Name=thigh_l
  swing1LimitDegrees=45
  swing1Motion=Free
  swing2LimitDegrees=45
  swing2Motion=Free
  twistLimitDegrees=45
  twistMotion=Free
 }
 {bone1Name=dyn_dress_01_l_fr, bone2Name=dyn_dress_l_fr_base, swing1LimitDegrees=10, swing2LimitDegrees=30, twistLimitDegrees=5}
 {bone1Name=dyn_dress_02_l_fr, bone2Name=dyn_dress_01_l_fr, swing1LimitDegrees=10, swing2LimitDegrees=30, twistLimitDegrees=50}
 {bone1Name=dyn_dress_03_l_fr, bone2Name=dyn_dress_02_l_fr, swing1LimitDegrees=40, swing2LimitDegrees=30, twistLimitDegrees=50}
 {
  bone1Name=calf_l
  bone2Name=thigh_l
  swing1LimitDegrees=45
  swing1Motion=Free
  swing2LimitDegrees=45
  swing2Motion=Free
  twistLimitDegrees=45
  twistMotion=Free
 }
 {
  bone1Name=thigh_l
  bone2Name=pelvis
  swing1LimitDegrees=45
  swing1Motion=Free
  swing2LimitDegrees=45
  swing2Motion=Free
  twistLimitDegrees=45
  twistMotion=Free
 }
 {
  bone1Name=thigh_r
  bone2Name=pelvis
  swing1LimitDegrees=45
  swing1Motion=Free
  swing2LimitDegrees=45
  swing2Motion=Free
  twistLimitDegrees=45
  twistMotion=Free
 }
 {
  bone1Name=calf_r
  bone2Name=thigh_r
  swing1LimitDegrees=45
  swing1Motion=Free
  swing2LimitDegrees=45
  swing2Motion=Free
  twistLimitDegrees=45
  twistMotion=Free
 }
 {
  bone1Name=pelvis
  bone2Name=root
  swing1LimitDegrees=45
  swing1Motion=Free
  swing2LimitDegrees=45
  swing2Motion=Free
  twistLimitDegrees=45
  twistMotion=Free
 }
 {
  bone1Name=spine_01
  bone2Name=root
  swing1LimitDegrees=45
  swing1Motion=Free
  swing2LimitDegrees=45
  swing2Motion=Free
  twistLimitDegrees=45
  twistMotion=Free
 }
 {
  bone1Name=dyn_dress_04_l_fr
  bone2Name=dyn_dress_04_l_side
  swing1LimitDegrees=45
  swing1Motion=Free
  swing2LimitDegrees=45
  swing2Motion=Free
  twistLimitDegrees=45
  twistMotion=Free
 }
 {
  bone1Name=dyn_dress_04_l_bk
  bone2Name=dyn_dress_04_l_side
  swing1LimitDegrees=45
  swing1Motion=Free
  swing2LimitDegrees=45
  swing2Motion=Free
  twistLimitDegrees=45
  twistMotion=Free
 }
 {
  bone1Name=dyn_dress_04_r_bk
  bone2Name=dyn_dress_04_l_bk
  swing1LimitDegrees=45
  swing1Motion=Free
  swing2LimitDegrees=45
  swing2Motion=Free
  twistLimitDegrees=45
  twistMotion=Free
 }
 {
  bone1Name=dyn_belt_l_base
  bone2Name=dyn_belt_base
  swing1LimitDegrees=45
  swing1Motion=Free
  swing2LimitDegrees=45
  swing2Motion=Free
  twistLimitDegrees=45
  twistMotion=Free
 }
 {bone1Name=dyn_belt_l, bone2Name=dyn_belt_l_base, swing1LimitDegrees=30, swing2LimitDegrees=30, twistLimitDegrees=5}
 {
  bone1Name=dyn_belt_r_base
  bone2Name=dyn_belt_base
  swing1LimitDegrees=45
  swing1Motion=Free
  swing2LimitDegrees=45
  swing2Motion=Free
  twistLimitDegrees=45
  twistMotion=Free
 }
 {bone1Name=dyn_belt_r, bone2Name=dyn_belt_r_base, swing1LimitDegrees=30, swing2LimitDegrees=30, twistLimitDegrees=5}
 {
  bone1Name=dyn_belt_base
  bone2Name=spine_02
  swing1LimitDegrees=45
  swing1Motion=Free
  swing2LimitDegrees=45
  swing2Motion=Free
  twistLimitDegrees=45
  twistMotion=Free
 }
 {
  bone1Name=spine_02
  bone2Name=spine_01
  swing1LimitDegrees=45
  swing1Motion=Free
  swing2LimitDegrees=45
  swing2Motion=Free
  twistLimitDegrees=45
  twistMotion=Free
 }
 {
  bone1Name=dyn_root_floor
  bone2Name=pelvis
  swing1LimitDegrees=45
  swing1Motion=Free
  swing2LimitDegrees=45
  swing2Motion=Free
  twistLimitDegrees=45
  twistMotion=Free
 }
 {bone1Name=dyn_belt_strap_03, bone2Name=dyn_belt_strap_02, swing1LimitDegrees=15, swing2LimitDegrees=45, twistLimitDegrees=2}
 {
  bone1Name=dyn_belt_strap_base
  bone2Name=dyn_belt_base
  swing1LimitDegrees=45
  swing1Motion=Free
  swing2LimitDegrees=45
  swing2Motion=Free
  twistLimitDegrees=45
  twistMotion=Free
 }
 {bone1Name=dyn_belt_strap_01, bone2Name=dyn_belt_strap_base, swing1LimitDegrees=10, swing2LimitDegrees=45, twistLimitDegrees=2}
 {bone1Name=dyn_belt_strap_02, bone2Name=dyn_belt_strap_01, swing1LimitDegrees=15, swing2LimitDegrees=45, twistLimitDegrees=2}
 {
  bone1Name=dyn_strap_collision_l
  bone2Name=thigh_l
  swing1LimitDegrees=45
  swing1Motion=Free
  swing2LimitDegrees=45
  swing2Motion=Free
  twistLimitDegrees=45
  twistMotion=Free
 }
 {
  bone1Name=dyn_dress_03_l_fr
  bone2Name=dyn_dress_03_l_side
  swing1LimitDegrees=45
  swing1Motion=Locked
  swing2LimitDegrees=45
  swing2Motion=Locked
  twistLimitDegrees=45
  twistMotion=Locked
 }
]
