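# Post-processes the FlameBride body's pose, copied from its attached parent mesh, adding secondary motion to dress, lace and belt.
# A PreDyn Control Rig feeds RigidBody dress physics, weighted constraints anchoring lace bases, AnimDynamics lace and belt bones, then a PostDyn rig.
# That cached pose feeds alternative constraint chains placing skirt FX joints from leg or dress bones.
# Downstream, a bool blend on the FrontEnd pose offers adjusted front-left dress bones.
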AnimBlueprint /BRCosmetics/Characters/Player/Female/Medium/Bodies/F_MED_FlameBride/Meshes/F_MED_FlameBride_AnimBP
target_skeleton=&/F_MED_FlameBride_Skeleton
graphs [
 AnimGraph {nodes [
   n1 = "Constraint - dyn_dress_03_l_fr (P)"(ComponentPose=Local To Component(LocalPose=Use cached pose 'FrontEnd'()), Alpha=0.5, ConstraintWeights_0=1) {
    alphaBoolBlend {blendOption=Linear}
    boneToModify {boneName=dyn_dress_03_l_fr}
    constraintSetup [{targetBone {boneName=dyn_dress_02_l_fr}}]
    constraintWeights [1]
   }
   n2 = "Constraint - dyn_dress_04_l_fr (P)"(ComponentPose=@n1, Alpha=0.5, ConstraintWeights_0=1) {
    alphaBoolBlend {blendOption=Linear}
    boneToModify {boneName=dyn_dress_04_l_fr}
    constraintSetup [{targetBone {boneName=dyn_dress_03_l_fr}}]
    constraintWeights [1]
   }
   n3 = "Transform (Modify) Bone - Bone: dyn_dress_02_l_fr"(ComponentPose=@n2, Alpha=1, Translation="0,0,0", Rotation="8.874,-10.935,-32.422", Scale="1,1,1") {
    rotationMode=BMM_Additive
    rotationSpace=BCS_ParentBoneSpace
    translationMode=BMM_Additive
    translationSpace=BCS_BoneSpace
    alphaBoolBlend {blendOption=Linear}
    boneToModify {boneName=dyn_dress_02_l_fr}
    rotation=<rot 8.874 -10.935 -32.422>
   }
   n4 = Component To Local(ComponentPose=@n3)
   n5 = Use cached pose 'FrontEnd'()
   n6 = "Transform (Modify) Bone - Bone: extra_neck_02_scale"!(
    ComponentPose=Local To Component(LocalPose=Blend Poses by bool(BlendPose_0=@n4, BlendPose_1=@n5) {blendType=Linear})
    bAlphaBoolEnabled:bool=True
    Translation="0,0,0"
    Rotation="0,0,0"
    Scale="1,2,2"
   ) -> (Pose) {
    alphaInputType=Bool
    scaleMode=BMM_Additive
    scaleSpace=BCS_BoneSpace
    translationMode=BMM_Additive
    translationSpace=BCS_BoneSpace
    alphaBoolBlend {blendOption=Linear}
    boneToModify {boneName=extra_neck_02_scale}
    scale {y=2, z=2}
   }
   Output Pose(Result=Component To Local(ComponentPose=@n6)) {layerGroup=None}
   n7 = Control Rig!(Source=Copy Pose From Mesh() {bUseAttachedParent=true}, Alpha, crouch:bool=false, sliding:bool=false) -> (Pose) {
    alphaBoolBlend {blendOption=Linear}
    controlRigAssetReference {blueprintRigClass=&/F_MED_FlameBride_CtrlRig_PreDyn.F_MED_FlameBride_CtrlRig_PreDyn_C}
   }
   n8 = RigidBody!(
    ComponentPose=Local To Component(LocalPose=@n7)
    Alpha
    OverrideWorldGravity:Vector="0,0,-980"
    ExternalForce
    SimSpaceSettings:Sim Space Settings Structure=(WorldAlpha=0,VelocityScaleZ=1,DampingAlpha=1,MaxLinearVelocity=10000,MaxAngularVelocity=10000,MaxLinearAcceleration=10000,MaxAngularAcceleration=10000,ExternalLinearDragV=(X=0,Y=0,Z=0),ExternalLinearVelocity=(X=0,Y=0,Z=0),ExternalAngularVelocity=(X=0,Y=0,Z=0))
    bTransferBoneVelocities:bool=False
   ) -> (Pose) {
    bClampLinearTranslationLimitToRefPose=true
    bFreezeIncomingPoseOnStart=true
    bOverrideWorldGravity=true
    bTransferBoneVelocities=true
    overridePhysicsAsset=&/F_MED_FlameBride_Physics_dress
    simulationSpace=BaseBoneSpace
    alphaBoolBlend {blendOption=Linear}
    baseBoneRef {boneName=pelvis}
    componentAppliedLinearAccClamp=<3000 3000 3000>
    overrideWorldGravity {z=-980}
    simSpaceSettings {maxAngularAcceleration=3000, maxAngularVelocity=3000, maxLinearAcceleration=3000, maxLinearVelocity=3000, velocityScaleZ=0.1, worldAlpha=0.5}
   }
   n9 = "Constraint - dyn_lace_fr_r_base (PPPPPPPP)"!(
    ComponentPose=@n8
    Alpha=1
    ConstraintWeights_0=0.4068
    ConstraintWeights_1:float=0.3216
    ConstraintWeights_2:float=0.1677
    ConstraintWeights_3:float=0.0637
    ConstraintWeights_4:float=0.0171
    ConstraintWeights_5:float=0.0157
    ConstraintWeights_6:float=0.0071
    ConstraintWeights_7:float=0.0003
   ) -> (Pose) {
    alphaBoolBlend {blendOption=Linear}
    boneToModify {boneName=dyn_lace_fr_r_base}
    constraintSetup [
     {targetBone {boneName=deform_pec_r}}
     {targetBone {boneName=spine_04}}
     {targetBone {boneName=spine_05}}
     {targetBone {boneName=clavicle_r}}
     {targetBone {boneName=upperarm_twist_01_r}}
     {targetBone {boneName=dyn_lace_fr_r}}
     {targetBone {boneName=deform_upperarm_fr_r}}
     {targetBone {boneName=dyn_lace_fr_center}}
    ]
    constraintWeights [1,1,1,1,1,1,1,1]
   }
   n10 = "Constraint - dyn_lace_fr_center_base (PPP)"!(ComponentPose=@n9, Alpha=1, ConstraintWeights_0=0.6879, ConstraintWeights_1:float=0.3083, ConstraintWeights_2:float=0.0038) -> (Pose) {
    alphaBoolBlend {blendOption=Linear}
    boneToModify {boneName=dyn_lace_fr_center_base}
    constraintSetup [{targetBone {boneName=spine_04}}, {targetBone {boneName=spine_05}}, {targetBone {boneName=deform_pec_l}}]
    constraintWeights [1,1,1]
   }
   n11 = "Constraint - dyn_lace_fr_l_base (PPPPPP)"!(
    ComponentPose=@n10
    Alpha=1
    ConstraintWeights_0=0.4225
    ConstraintWeights_1:float=0.3396
    ConstraintWeights_2:float=0.1529
    ConstraintWeights_3:float=0.0623
    ConstraintWeights_4:float=0.0184
    ConstraintWeights_5:float=0.0043
   ) -> (Pose) {
    alphaBoolBlend {blendOption=Linear}
    boneToModify {boneName=dyn_lace_fr_l_base}
    constraintSetup [
     {targetBone {boneName=deform_pec_l}}
     {targetBone {boneName=spine_04}}
     {targetBone {boneName=spine_05}}
     {targetBone {boneName=clavicle_l}}
     {targetBone {boneName=upperarm_twist_01_l}}
     {targetBone {boneName=deform_upperarm_fr_l}}
    ]
    constraintWeights [1,1,1,1,1,1]
   }
   n12 = "Constraint - dyn_lace_bk_r_base (PPPPPPPPPPPP)"!(
    ComponentPose=@n11
    Alpha=1
    ConstraintWeights_0=0.4087
    ConstraintWeights_1:float=0.2319
    ConstraintWeights_2:float=0.1657
    ConstraintWeights_3:float=0.0716
    ConstraintWeights_4:float=0.0405
    ConstraintWeights_5:float=0.0306
    ConstraintWeights_6:float=0.0259
    ConstraintWeights_7:float=0.0095
    ConstraintWeights_8:float=0.0076
    ConstraintWeights_9:float=0.0075
    ConstraintWeights_10:float=0.0004
    ConstraintWeights_11:float=0.0001
   ) -> (Pose) {
    alphaBoolBlend {blendOption=Linear}
    boneToModify {boneName=dyn_lace_bk_r_base}
    constraintSetup [
     {targetBone {boneName=clavicle_r}}
     {targetBone {boneName=spine_05}}
     {targetBone {boneName=spine_04}}
     {targetBone {boneName=upperarm_twist_01_r}}
     {targetBone {boneName=dyn_lace_bk_r}}
     {targetBone {boneName=dyn_lace_bk_center}}
     {targetBone {boneName=deform_upperarm_bk_r}}
     {targetBone {boneName=spine_03}}
     {targetBone {boneName=spine_02}}
     {targetBone {boneName=clavicle_l}}
     {targetBone {boneName=deform_upperarm_up_r}}
     {targetBone {boneName=spine_01}}
    ]
    constraintWeights [1,1,1,1,1,1,1,1,1,1,1,1]
   }
   n13 = "Constraint - dyn_lace_bk_center_base (PPPPPPP)"!(
    ComponentPose=@n12
    Alpha=1
    ConstraintWeights_0=0.4455
    ConstraintWeights_1:float=0.2586
    ConstraintWeights_2:float=0.1299
    ConstraintWeights_3:float=0.1271
    ConstraintWeights_4:float=0.0199
    ConstraintWeights_5:float=0.0169
    ConstraintWeights_6:float=0.0021
   ) -> (Pose) {
    alphaBoolBlend {blendOption=Linear}
    boneToModify {boneName=dyn_lace_bk_center_base}
    constraintSetup [
     {targetBone {boneName=spine_05}}
     {targetBone {boneName=spine_04}}
     {targetBone {boneName=clavicle_r}}
     {targetBone {boneName=clavicle_l}}
     {targetBone {boneName=spine_03}}
     {targetBone {boneName=spine_02}}
     {targetBone {boneName=spine_01}}
    ]
    constraintWeights [1,1,1,1,1,1,1]
   }
   n14 = "Constraint - dyn_lace_bk_l_base (PPPPPPPPPPPP)"!(
    ComponentPose=@n13
    Alpha=1
    ConstraintWeights_0=0.41
    ConstraintWeights_1:float=0.2324
    ConstraintWeights_2:float=0.1662
    ConstraintWeights_3:float=0.0722
    ConstraintWeights_4:float=0.039
    ConstraintWeights_5:float=0.0286
    ConstraintWeights_6:float=0.0263
    ConstraintWeights_7:float=0.0095
    ConstraintWeights_8:float=0.0076
    ConstraintWeights_9:float=0.0075
    ConstraintWeights_10:float=0.0004
    ConstraintWeights_11:float=0.0001
   ) -> (Pose) {
    alphaBoolBlend {blendOption=Linear}
    boneToModify {boneName=dyn_lace_bk_l_base}
    constraintSetup [
     {targetBone {boneName=clavicle_l}}
     {targetBone {boneName=spine_05}}
     {targetBone {boneName=spine_04}}
     {targetBone {boneName=upperarm_twist_01_l}}
     {targetBone {boneName=dyn_lace_bk_l}}
     {targetBone {boneName=dyn_lace_bk_center}}
     {targetBone {boneName=deform_upperarm_bk_l}}
     {targetBone {boneName=spine_03}}
     {targetBone {boneName=spine_02}}
     {targetBone {boneName=clavicle_r}}
     {targetBone {boneName=deform_upperarm_up_l}}
     {targetBone {boneName=spine_01}}
    ]
    constraintWeights [1,1,1,1,1,1,1,1,1,1,1,1]
   }
   n15 = "Constraint - dyn_lace_bk_side_r_base (PPPPPPP)"!(
    ComponentPose=@n14
    Alpha=1
    ConstraintWeights_0=0.589
    ConstraintWeights_1:float=0.2348
    ConstraintWeights_2:float=0.0736
    ConstraintWeights_3:float=0.0644
    ConstraintWeights_4:float=0.0209
    ConstraintWeights_5:float=0.0089
    ConstraintWeights_6:float=0.0083
   ) -> (Pose) {
    alphaBoolBlend {blendOption=Linear}
    boneToModify {boneName=dyn_lace_bk_side_r_base}
    constraintSetup [
     {targetBone {boneName=upperarm_twist_01_r}}
     {targetBone {boneName=deform_upperarm_bk_r}}
     {targetBone {boneName=clavicle_r}}
     {targetBone {boneName=dyn_lace_bk_side_r}}
     {targetBone {boneName=deform_upperarm_up_r}}
     {targetBone {boneName=spine_05}}
     {targetBone {boneName=spine_04}}
    ]
    constraintWeights [1,1,1,1,1,1,1]
   }
   n16 = "Constraint - dyn_lace_side_r_base (PPPPP)"!(
    ComponentPose=@n15
    Alpha=1
    ConstraintWeights_0=0.8496
    ConstraintWeights_1:float=0.0803
    ConstraintWeights_2:float=0.0286
    ConstraintWeights_3:float=0.0208
    ConstraintWeights_4:float=0.0207
   ) -> (Pose) {
    alphaBoolBlend {blendOption=Linear}
    boneToModify {boneName=dyn_lace_side_r_base}
    constraintSetup [
     {targetBone {boneName=upperarm_twist_01_r}}
     {targetBone {boneName=deform_upperarm_up_r}}
     {targetBone {boneName=clavicle_r}}
     {targetBone {boneName=deform_upperarm_fr_r}}
     {targetBone {boneName=deform_upperarm_bk_r}}
    ]
    constraintWeights [1,1,1,1,1]
   }
   n17 = "Constraint - dyn_lace_fr_side_r_base (PPPPPPPP)"!(
    ComponentPose=@n16
    Alpha=1
    ConstraintWeights_0=0.4575
    ConstraintWeights_1:float=0.1836
    ConstraintWeights_2:float=0.1748
    ConstraintWeights_3:float=0.0807
    ConstraintWeights_4:float=0.0697
    ConstraintWeights_5:float=0.0148
    ConstraintWeights_6:float=0.0125
    ConstraintWeights_7:float=0.0066
   ) -> (Pose) {
    alphaBoolBlend {blendOption=Linear}
    boneToModify {boneName=dyn_lace_fr_side_r_base}
    constraintSetup [
     {targetBone {boneName=upperarm_twist_01_r}}
     {targetBone {boneName=deform_upperarm_fr_r}}
     {targetBone {boneName=clavicle_r}}
     {targetBone {boneName=spine_05}}
     {targetBone {boneName=spine_04}}
     {targetBone {boneName=deform_upperarm_up_r}}
     {targetBone {boneName=deform_pec_r}}
     {targetBone {boneName=dyn_lace_fr_side_r}}
    ]
    constraintWeights [1,1,1,1,1,1,1,1]
   }
   n18 = "Constraint - dyn_lace_fr_side_l_base (PPPPPPPPP)"!(
    ComponentPose=@n17
    Alpha=1
    ConstraintWeights_0=0.3544
    ConstraintWeights_1:float=0.1935
    ConstraintWeights_2:float=0.1786
    ConstraintWeights_3:float=0.1019
    ConstraintWeights_4:float=0.0989
    ConstraintWeights_5:float=0.0391
    ConstraintWeights_6:float=0.02
    ConstraintWeights_7:float=0.0135
    ConstraintWeights_8:float=0.0001
   ) -> (Pose) {
    alphaBoolBlend {blendOption=Linear}
    boneToModify {boneName=dyn_lace_fr_side_l_base}
    constraintSetup [
     {targetBone {boneName=upperarm_twist_01_l}}
     {targetBone {boneName=clavicle_l}}
     {targetBone {boneName=deform_upperarm_fr_l}}
     {targetBone {boneName=spine_04}}
     {targetBone {boneName=spine_05}}
     {targetBone {boneName=deform_pec_l}}
     {targetBone {boneName=dyn_lace_fr_side_l}}
     {targetBone {boneName=deform_upperarm_up_l}}
     {targetBone {boneName=dyn_lace_fr_l}}
    ]
    constraintWeights [1,1,1,1,1,1,1,1,1]
   }
   n19 = "Constraint - dyn_lace_side_l_base (PPPPP)"!(
    ComponentPose=@n18
    Alpha=1
    ConstraintWeights_0=0.8496
    ConstraintWeights_1:float=0.0803
    ConstraintWeights_2:float=0.0286
    ConstraintWeights_3:float=0.0208
    ConstraintWeights_4:float=0.0207
   ) -> (Pose) {
    alphaBoolBlend {blendOption=Linear}
    boneToModify {boneName=dyn_lace_side_l_base}
    constraintSetup [
     {targetBone {boneName=upperarm_twist_01_l}}
     {targetBone {boneName=deform_upperarm_up_l}}
     {targetBone {boneName=clavicle_l}}
     {targetBone {boneName=deform_upperarm_fr_l}}
     {targetBone {boneName=deform_upperarm_bk_l}}
    ]
    constraintWeights [1,1,1,1,1]
   }
   n20 = "Constraint - dyn_lace_bk_side_l_base (PPPPPPPP)"!(
    ComponentPose=@n19
    Alpha=1
    ConstraintWeights_0=0.6052
    ConstraintWeights_1:float=0.2408
    ConstraintWeights_2:float=0.0738
    ConstraintWeights_3:float=0.0433
    ConstraintWeights_4:float=0.0177
    ConstraintWeights_5:float=0.0096
    ConstraintWeights_6:float=0.0094
    ConstraintWeights_7:float=0.0001
   ) -> (Pose) {
    alphaBoolBlend {blendOption=Linear}
    boneToModify {boneName=dyn_lace_bk_side_l_base}
    constraintSetup [
     {targetBone {boneName=upperarm_twist_01_l}}
     {targetBone {boneName=deform_upperarm_bk_l}}
     {targetBone {boneName=clavicle_l}}
     {targetBone {boneName=dyn_lace_bk_side_l}}
     {targetBone {boneName=deform_upperarm_up_l}}
     {targetBone {boneName=spine_05}}
     {targetBone {boneName=spine_04}}
     {targetBone {boneName=dyn_lace_side_l}}
    ]
    constraintWeights [1,1,1,1,1,1,1,1]
   }
   n21 = Anim Dynamics - Bone: dyn_lace_fr_side_l!(ComponentPose=@n20, Alpha=1, GravityOverride:Vector="0,0,-980", ExternalForce="0,0,0") -> (Pose) {
    angularDampingOverride=1
    angularSpringConstant=200000
    bOverrideAngularDamping=true
    bUseGravityOverride=true
    bUsePlanarLimit=false
    gravityScale=0.35
    linearSpringConstant=20
    simulationSpace=BoneRelative
    alphaBoolBlend {blendOption=Linear}
    boundBone {boneName=dyn_lace_fr_side_l}
    componentAppliedLinearAccClamp=<1000 1000 1000>
    componentLinearAccScale=<1 1 1>
    componentLinearVelScale=<1 1 1>
    gravityOverride {z=-980}
    physicsBodyDefinitions [{
      boundBone {boneName=dyn_lace_fr_side_l}
      boxExtents=<60 20 20>
      constraintSetup {angularLimitsMax=<60 15 15>, angularLimitsMin=<-60 -15 -15>, angularTarget=<1 0 0>}
      localJointOffset=<15 0 0>
     }]
    relativeSpaceBone {boneName=dyn_lace_fr_side_l_base}
   }
   n22 = Anim Dynamics - Bone: dyn_lace_side_l!(ComponentPose=@n21, Alpha=1, GravityOverride:Vector="0,0,-980", ExternalForce="0,0,0") -> (Pose) {
    angularDampingOverride=1
    angularSpringConstant=200000
    bOverrideAngularDamping=true
    bUseGravityOverride=true
    bUsePlanarLimit=false
    gravityScale=0.35
    linearSpringConstant=20
    simulationSpace=BoneRelative
    alphaBoolBlend {blendOption=Linear}
    boundBone {boneName=dyn_lace_side_l}
    componentAppliedLinearAccClamp=<3000 3000 1000>
    componentLinearAccScale=<1 1 1>
    componentLinearVelScale=<1 1 3>
    gravityOverride {z=-980}
    physicsBodyDefinitions [{
      boundBone {boneName=dyn_lace_side_l}
      boxExtents=<60 20 20>
      constraintSetup {angularLimitsMax=<60 30 30>, angularLimitsMin=<-60 -30 -30>, angularTarget=<1 0 0>}
      localJointOffset=<15 0 0>
     }]
    relativeSpaceBone {boneName=dyn_lace_side_l_base}
   }
   n23 = Anim Dynamics - Bone: dyn_lace_bk_side_l!(ComponentPose=@n22, Alpha=1, GravityOverride:Vector="0,0,-980", ExternalForce="0,0,0") -> (Pose) {
    angularDampingOverride=1
    angularSpringConstant=200000
    bOverrideAngularDamping=true
    bUseGravityOverride=true
    bUsePlanarLimit=false
    gravityScale=0.35
    linearSpringConstant=20
    simulationSpace=BoneRelative
    alphaBoolBlend {blendOption=Linear}
    boundBone {boneName=dyn_lace_bk_side_l}
    componentAppliedLinearAccClamp=<3000 3000 1000>
    componentLinearAccScale=<1 1 1>
    componentLinearVelScale=<1 1 3>
    gravityOverride {z=-980}
    physicsBodyDefinitions [{
      boundBone {boneName=dyn_lace_bk_side_l}
      boxExtents=<60 20 20>
      constraintSetup {angularLimitsMax=<60 30 30>, angularLimitsMin=<-60 -30 -30>, angularTarget=<1 0 0>}
      localJointOffset=<15 0 0>
     }]
    relativeSpaceBone {boneName=dyn_lace_bk_side_l_base}
   }
   n24 = Anim Dynamics - Bone: dyn_lace_fr_side_r!(ComponentPose=@n23, Alpha=1, GravityOverride:Vector="0,0,-980", ExternalForce="0,0,0") -> (Pose) {
    angularDampingOverride=1
    angularSpringConstant=200000
    bOverrideAngularDamping=true
    bUseGravityOverride=true
    bUsePlanarLimit=false
    gravityScale=0.35
    linearSpringConstant=20
    simulationSpace=BoneRelative
    alphaBoolBlend {blendOption=Linear}
    boundBone {boneName=dyn_lace_fr_side_r}
    componentAppliedLinearAccClamp=<3000 3000 1000>
    componentLinearAccScale=<1 1 1>
    componentLinearVelScale=<1 1 3>
    gravityOverride {z=-980}
    physicsBodyDefinitions [{
      boundBone {boneName=dyn_lace_fr_side_r}
      boxExtents=<60 20 20>
      constraintSetup {angularLimitsMax=<60 30 30>, angularLimitsMin=<-60 -30 -30>, angularTarget=<1 0 0>}
      localJointOffset=<15 0 0>
     }]
    relativeSpaceBone {boneName=dyn_lace_fr_side_r_base}
   }
   n25 = Anim Dynamics - Bone: dyn_lace_side_r!(ComponentPose=@n24, Alpha=1, GravityOverride:Vector="0,0,-980", ExternalForce="0,0,0") -> (Pose) {
    angularDampingOverride=1
    angularSpringConstant=200000
    bOverrideAngularDamping=true
    bUseGravityOverride=true
    bUsePlanarLimit=false
    gravityScale=0.35
    linearSpringConstant=20
    simulationSpace=BoneRelative
    alphaBoolBlend {blendOption=Linear}
    boundBone {boneName=dyn_lace_side_r}
    componentAppliedLinearAccClamp=<3000 3000 1000>
    componentLinearAccScale=<1 1 1>
    componentLinearVelScale=<1 1 3>
    gravityOverride {z=-980}
    physicsBodyDefinitions [{
      boundBone {boneName=dyn_lace_side_r}
      boxExtents=<60 20 20>
      constraintSetup {angularLimitsMax=<60 30 30>, angularLimitsMin=<-60 -30 -30>, angularTarget=<1 0 0>}
      localJointOffset=<15 0 0>
     }]
    relativeSpaceBone {boneName=dyn_lace_side_r_base}
   }
   n26 = Anim Dynamics - Bone: dyn_lace_bk_side_r!(ComponentPose=@n25, Alpha=1, GravityOverride:Vector="0,0,-980", ExternalForce="0,0,0") -> (Pose) {
    angularDampingOverride=1
    angularSpringConstant=200000
    bOverrideAngularDamping=true
    bUseGravityOverride=true
    bUsePlanarLimit=false
    gravityScale=0.35
    linearSpringConstant=20
    simulationSpace=BoneRelative
    alphaBoolBlend {blendOption=Linear}
    boundBone {boneName=dyn_lace_bk_side_r}
    componentAppliedLinearAccClamp=<3000 3000 1000>
    componentLinearAccScale=<1 1 1>
    componentLinearVelScale=<1 1 3>
    gravityOverride {z=-980}
    physicsBodyDefinitions [{
      boundBone {boneName=dyn_lace_bk_side_r}
      boxExtents=<60 20 20>
      constraintSetup {angularLimitsMax=<60 30 30>, angularLimitsMin=<-60 -30 -30>, angularTarget=<1 0 0>}
      localJointOffset=<15 0 0>
     }]
    relativeSpaceBone {boneName=dyn_lace_bk_side_r_base}
   }
   n27 = Anim Dynamics - Bone: dyn_lace_bk_l!(ComponentPose=@n26, Alpha=1, GravityOverride:Vector="0,0,-980", ExternalForce="0,0,0") -> (Pose) {
    angularDampingOverride=1
    angularSpringConstant=200000
    bOverrideAngularDamping=true
    bUseGravityOverride=true
    bUsePlanarLimit=false
    gravityScale=0.35
    linearSpringConstant=20
    simulationSpace=BoneRelative
    alphaBoolBlend {blendOption=Linear}
    boundBone {boneName=dyn_lace_bk_l}
    componentAppliedLinearAccClamp=<3000 3000 1000>
    componentLinearAccScale=<1 1 1>
    componentLinearVelScale=<1 1 3>
    gravityOverride {z=-980}
    physicsBodyDefinitions [{
      boundBone {boneName=dyn_lace_bk_l}
      boxExtents=<60 20 20>
      constraintSetup {angularLimitsMax=<60 30 30>, angularLimitsMin=<-60 -30 -30>, angularTarget=<1 0 0>}
      localJointOffset=<15 0 0>
     }]
    relativeSpaceBone {boneName=dyn_lace_bk_l_base}
   }
   n28 = Anim Dynamics - Bone: dyn_lace_bk_center!(ComponentPose=@n27, Alpha=1, GravityOverride:Vector="0,0,-980", ExternalForce="0,0,0") -> (Pose) {
    angularDampingOverride=1
    angularSpringConstant=200000
    bOverrideAngularDamping=true
    bUseGravityOverride=true
    bUsePlanarLimit=false
    gravityScale=0.35
    linearSpringConstant=20
    simulationSpace=BoneRelative
    alphaBoolBlend {blendOption=Linear}
    boundBone {boneName=dyn_lace_bk_center}
    componentAppliedLinearAccClamp=<3000 3000 1000>
    componentLinearAccScale=<1 1 1>
    componentLinearVelScale=<1 1 3>
    gravityOverride {z=-980}
    physicsBodyDefinitions [{
      boundBone {boneName=dyn_lace_bk_center}
      boxExtents=<60 20 20>
      constraintSetup {angularLimitsMax=<60 30 30>, angularLimitsMin=<-60 -30 -30>, angularTarget=<1 0 0>}
      localJointOffset=<15 0 0>
     }]
    relativeSpaceBone {boneName=dyn_lace_bk_center_base}
   }
   n29 = Anim Dynamics - Bone: dyn_lace_bk_r!(ComponentPose=@n28, Alpha=1, GravityOverride:Vector="0,0,-980", ExternalForce="0,0,0") -> (Pose) {
    angularDampingOverride=1
    angularSpringConstant=200000
    bOverrideAngularDamping=true
    bUseGravityOverride=true
    bUsePlanarLimit=false
    gravityScale=0.35
    linearSpringConstant=20
    simulationSpace=BoneRelative
    alphaBoolBlend {blendOption=Linear}
    boundBone {boneName=dyn_lace_bk_r}
    componentAppliedLinearAccClamp=<3000 3000 1000>
    componentLinearAccScale=<1 1 1>
    componentLinearVelScale=<1 1 3>
    gravityOverride {z=-980}
    physicsBodyDefinitions [{
      boundBone {boneName=dyn_lace_bk_r}
      boxExtents=<60 20 20>
      constraintSetup {angularLimitsMax=<60 30 30>, angularLimitsMin=<-60 -30 -30>, angularTarget=<1 0 0>}
      localJointOffset=<15 0 0>
     }]
    relativeSpaceBone {boneName=dyn_lace_bk_r_base}
   }
   n30 = Anim Dynamics - Bone: dyn_lace_fr_l!(ComponentPose=@n29, Alpha=1, GravityOverride:Vector="0,0,-980", ExternalForce="0,0,0") -> (Pose) {
    angularDampingOverride=1
    angularSpringConstant=200000
    bOverrideAngularDamping=true
    bUseGravityOverride=true
    bUsePlanarLimit=false
    gravityScale=0.35
    linearSpringConstant=20
    simulationSpace=BoneRelative
    alphaBoolBlend {blendOption=Linear}
    boundBone {boneName=dyn_lace_fr_l}
    componentAppliedLinearAccClamp=<3000 3000 1000>
    componentLinearAccScale=<1 1 1>
    componentLinearVelScale=<1 1 3>
    gravityOverride {z=-980}
    physicsBodyDefinitions [{
      boundBone {boneName=dyn_lace_fr_l}
      boxExtents=<60 20 20>
      constraintSetup {angularLimitsMax=<60 30 30>, angularLimitsMin=<-60 -30 -30>, angularTarget=<1 0 0>}
      localJointOffset=<15 0 0>
     }]
    relativeSpaceBone {boneName=dyn_lace_fr_l_base}
   }
   n31 = Anim Dynamics - Bone: dyn_lace_fr_center!(ComponentPose=@n30, Alpha=1, GravityOverride:Vector="0,0,-980", ExternalForce="0,0,0") -> (Pose) {
    angularDampingOverride=1
    angularSpringConstant=200000
    bOverrideAngularDamping=true
    bUseGravityOverride=true
    bUsePlanarLimit=false
    gravityScale=0.35
    linearSpringConstant=20
    simulationSpace=BoneRelative
    alphaBoolBlend {blendOption=Linear}
    boundBone {boneName=dyn_lace_fr_center}
    componentAppliedLinearAccClamp=<3000 3000 1000>
    componentLinearAccScale=<1 1 1>
    componentLinearVelScale=<1 1 3>
    gravityOverride {z=-980}
    physicsBodyDefinitions [{
      boundBone {boneName=dyn_lace_fr_center}
      boxExtents=<60 20 20>
      constraintSetup {angularLimitsMax=<60 30 30>, angularLimitsMin=<-60 -30 -30>, angularTarget=<1 0 0>}
      localJointOffset=<15 0 0>
     }]
    relativeSpaceBone {boneName=dyn_lace_fr_center_base}
   }
   n32 = Anim Dynamics - Bone: dyn_lace_fr_r!(ComponentPose=@n31, Alpha=1, GravityOverride:Vector="0,0,-980", ExternalForce="0,0,-980") -> (Pose) {
    angularDampingOverride=1
    angularSpringConstant=200000
    bOverrideAngularDamping=true
    bUseGravityOverride=true
    bUsePlanarLimit=false
    gravityScale=0.35
    linearSpringConstant=20
    simulationSpace=BoneRelative
    alphaBoolBlend {blendOption=Linear}
    boundBone {boneName=dyn_lace_fr_r}
    componentAppliedLinearAccClamp=<3000 3000 1000>
    componentLinearAccScale=<1 1 1>
    componentLinearVelScale=<1 1 3>
    gravityOverride {z=-980}
    physicsBodyDefinitions [{
      boundBone {boneName=dyn_lace_fr_r}
      boxExtents=<60 20 20>
      constraintSetup {angularLimitsMax=<60 30 30>, angularLimitsMin=<-60 -30 -30>, angularTarget=<1 0 0>}
      localJointOffset=<15 0 0>
     }]
    relativeSpaceBone {boneName=dyn_lace_fr_r_base}
   }
   n33 = "Constraint - dyn_belt_strap_03 (T)"(ComponentPose=@n32, Alpha=0.3, ConstraintWeights_0=1) {
    alphaBoolBlend {blendOption=Linear}
    boneToModify {boneName=dyn_belt_strap_03}
    constraintSetup [{transformType=Translation, targetBone {boneName=dyn_dress_02_l_fr}}]
    constraintWeights [1]
   }
   n34 = "Constraint - dyn_belt_strap_03 (R)"(ComponentPose=@n33, Alpha=1, ConstraintWeights_0=1) {
    alphaBoolBlend {blendOption=Linear}
    boneToModify {boneName=dyn_belt_strap_03}
    constraintSetup [{transformType=Rotation, targetBone {boneName=dyn_dress_02_l_fr}}]
    constraintWeights [1]
   }
   n35 = Anim Dynamics - Bone: dyn_belt_l(ComponentPose=@n34, Alpha=1, ExternalForce="0,0,0") {
    angularDampingOverride=1
    angularSpringConstant=20000
    bAngularSpring=true
    bOverrideAngularDamping=true
    bUseGravityOverride=true
    gravityScale=0.35
    simulationSpace=RootRelative
    alphaBoolBlend {blendOption=Linear}
    boundBone {boneName=dyn_belt_l}
    componentLinearAccScale=<1 1 1>
    componentLinearVelScale=<1 1 1>
    gravityOverride {z=-980}
    physicsBodyDefinitions [{
      boundBone {boneName=dyn_belt_l}
      boxExtents=<20 30 30>
      constraintSetup {angularLimitsMax=<15 15 15>, angularLimitsMin=<-15 -15 -15>, angularTarget=<1 0 0>}
      localJointOffset=<1 0 0>
     }]
   }
   n36 = Anim Dynamics - Bone: dyn_belt_r(ComponentPose=@n35, Alpha=1, ExternalForce="0,0,0") {
    angularDampingOverride=1
    angularSpringConstant=20000
    bAngularSpring=true
    bOverrideAngularDamping=true
    bUseGravityOverride=true
    gravityScale=0.35
    simulationSpace=RootRelative
    alphaBoolBlend {blendOption=Linear}
    boundBone {boneName=dyn_belt_r}
    componentLinearAccScale=<1 1 1>
    componentLinearVelScale=<1 1 1>
    gravityOverride {z=-980}
    physicsBodyDefinitions [{
      boundBone {boneName=dyn_belt_r}
      boxExtents=<20 30 30>
      constraintSetup {angularLimitsMax=<15 15 15>, angularLimitsMin=<-15 -15 -15>, angularTarget=<1 0 0>}
      localJointOffset=<4 0 0>
     }]
   }
   n37 = Control Rig!(Source=Component To Local(ComponentPose=@n36), Alpha) -> (Pose) {
    alphaBoolBlend {blendOption=Linear}
    controlRigAssetReference {blueprintRigClass=&/F_MED_FlameBride_CtrlRig_PostDyn.F_MED_FlameBride_CtrlRig_PostDyn_C}
   }
   Save cached pose 'FX_joints'(Pose=@n37) {cacheName=FX_joints}
   n38 = "Constraint - fx_skirt_01 (PPP)"!(
    ComponentPose=Local To Component(LocalPose=Use cached pose 'FX_joints'())
    Alpha=1
    ConstraintWeights_0=0.8149
    ConstraintWeights_1:float=0.1634
    ConstraintWeights_2:float=0.0217
   ) -> (Pose) {
    alphaBoolBlend {blendOption=Linear}
    boneToModify {boneName=fx_skirt_01}
    constraintSetup [{targetBone {boneName=calf_twist_02_l}}, {targetBone {boneName=calf_l}}, {targetBone {boneName=calf_twist_01_l}}]
    constraintWeights [1,1,1]
   }
   n39 = "Constraint - fx_skirt_02 (PPPPPPPP)"!(
    ComponentPose=@n38
    Alpha=1
    ConstraintWeights_0=0.4056
    ConstraintWeights_1:float=0.1757
    ConstraintWeights_2:float=0.1413
    ConstraintWeights_3:float=0.1068
    ConstraintWeights_4:float=0.0967
    ConstraintWeights_5:float=0.0353
    ConstraintWeights_6:float=0.0311
    ConstraintWeights_7:float=0.0075
   ) -> (Pose) {
    alphaBoolBlend {blendOption=Linear}
    boneToModify {boneName=fx_skirt_02}
    constraintSetup [
     {targetBone {boneName=calf_twist_02_l}}
     {targetBone {boneName=calf_twist_02_r}}
     {targetBone {boneName=calf_l}}
     {targetBone {boneName=calf_r}}
     {targetBone {boneName=calf_twist_01_l}}
     {targetBone {boneName=thigh_r}}
     {targetBone {boneName=thigh_l}}
     {targetBone {boneName=thigh_twist_01_r}}
    ]
    constraintWeights [1,1,1,1,1,1,1,1]
   }
   n40 = "Constraint - fx_skirt_03 (PPPPPPPP)"!(
    ComponentPose=@n39
    Alpha=1
    ConstraintWeights_0=0.2283
    ConstraintWeights_1:float=0.205
    ConstraintWeights_2:float=0.1742
    ConstraintWeights_3:float=0.1325
    ConstraintWeights_4:float=0.1147
    ConstraintWeights_5:float=0.085
    ConstraintWeights_6:float=0.0363
    ConstraintWeights_7:float=0.0238
   ) -> (Pose) {
    alphaBoolBlend {blendOption=Linear}
    boneToModify {boneName=fx_skirt_03}
    constraintSetup [
     {targetBone {boneName=calf_twist_02_l}}
     {targetBone {boneName=calf_twist_02_r}}
     {targetBone {boneName=calf_r}}
     {targetBone {boneName=thigh_r}}
     {targetBone {boneName=calf_l}}
     {targetBone {boneName=thigh_l}}
     {targetBone {boneName=thigh_twist_01_r}}
     {targetBone {boneName=thigh_twist_01_l}}
    ]
    constraintWeights [1,1,1,1,1,1,1,1]
   }
   n41 = "Constraint - fx_skirt_04 (PPPPPPPP)"!(
    ComponentPose=@n40
    Alpha=1
    ConstraintWeights_0=0.3697
    ConstraintWeights_1:float=0.1568
    ConstraintWeights_2:float=0.1405
    ConstraintWeights_3:float=0.1002
    ConstraintWeights_4:float=0.0827
    ConstraintWeights_5:float=0.053
    ConstraintWeights_6:float=0.0516
    ConstraintWeights_7:float=0.0455
   ) -> (Pose) {
    alphaBoolBlend {blendOption=Linear}
    boneToModify {boneName=fx_skirt_04}
    constraintSetup [
     {targetBone {boneName=thigh_r}}
     {targetBone {boneName=thigh_twist_01_r}}
     {targetBone {boneName=calf_r}}
     {targetBone {boneName=thigh_l}}
     {targetBone {boneName=calf_twist_02_r}}
     {targetBone {boneName=calf_twist_02_l}}
     {targetBone {boneName=calf_l}}
     {targetBone {boneName=thigh_twist_01_l}}
    ]
    constraintWeights [1,1,1,1,1,1,1,1]
   }
   n42 = "Constraint - fx_skirt_05 (PPPPPPPP)"!(
    ComponentPose=@n41
    Alpha=1
    ConstraintWeights_0=0.5431
    ConstraintWeights_1:float=0.2493
    ConstraintWeights_2:float=0.093
    ConstraintWeights_3:float=0.0543
    ConstraintWeights_4:float=0.0246
    ConstraintWeights_5:float=0.0228
    ConstraintWeights_6:float=0.0072
    ConstraintWeights_7:float=0.0057
   ) -> (Pose) {
    alphaBoolBlend {blendOption=Linear}
    boneToModify {boneName=fx_skirt_05}
    constraintSetup [
     {targetBone {boneName=thigh_r}}
     {targetBone {boneName=calf_r}}
     {targetBone {boneName=thigh_twist_01_r}}
     {targetBone {boneName=deform_knee_in_r}}
     {targetBone {boneName=thigh_l}}
     {targetBone {boneName=calf_twist_02_r}}
     {targetBone {boneName=calf_l}}
     {targetBone {boneName=thigh_twist_01_l}}
    ]
    constraintWeights [1,1,1,1,1,1,1,1]
   }
   n43 = "Constraint - fx_skirt_06 (PPPPPPPP)"!(
    ComponentPose=@n42
    Alpha=1
    ConstraintWeights_0=0.6874
    ConstraintWeights_1:float=0.1756
    ConstraintWeights_2:float=0.0794
    ConstraintWeights_3:float=0.0326
    ConstraintWeights_4:float=0.0078
    ConstraintWeights_5:float=0.0072
    ConstraintWeights_6:float=0.0061
    ConstraintWeights_7:float=0.0039
   ) -> (Pose) {
    alphaBoolBlend {blendOption=Linear}
    boneToModify {boneName=fx_skirt_06}
    constraintSetup [
     {targetBone {boneName=thigh_r}}
     {targetBone {boneName=thigh_twist_01_r}}
     {targetBone {boneName=calf_r}}
     {targetBone {boneName=deform_knee_in_r}}
     {targetBone {boneName=thigh_l}}
     {targetBone {boneName=deform_glute_r}}
     {targetBone {boneName=calf_twist_02_r}}
     {targetBone {boneName=pelvis}}
    ]
    constraintWeights [1,1,1,1,1,1,1,1]
   }
   n44 = "Constraint - fx_skirt_07 (PPPPPP)"!(
    ComponentPose=@n43
    Alpha=1
    ConstraintWeights_0=0.6282
    ConstraintWeights_1:float=0.2271
    ConstraintWeights_2:float=0.1058
    ConstraintWeights_3:float=0.0302
    ConstraintWeights_4:float=0.0085
    ConstraintWeights_5:float=0.0002
   ) -> (Pose) {
    alphaBoolBlend {blendOption=Linear}
    boneToModify {boneName=fx_skirt_07}
    constraintSetup [
     {targetBone {boneName=thigh_twist_01_r}}
     {targetBone {boneName=thigh_r}}
     {targetBone {boneName=pelvis}}
     {targetBone {boneName=deform_glute_r}}
     {targetBone {boneName=deform_groin_r}}
     {targetBone {boneName=spine_02}}
    ]
    constraintWeights [1,1,1,1,1,1]
   }
   n45 = "Constraint - fx_skirt_08 (PPPPPPP)"!(
    ComponentPose=@n44
    Alpha=1
    ConstraintWeights_0=0.5511
    ConstraintWeights_1:float=0.256
    ConstraintWeights_2:float=0.0645
    ConstraintWeights_3:float=0.0537
    ConstraintWeights_4:float=0.0337
    ConstraintWeights_5:float=0.0308
    ConstraintWeights_6:float=0.0103
   ) -> (Pose) {
    alphaBoolBlend {blendOption=Linear}
    boneToModify {boneName=fx_skirt_08}
    constraintSetup [
     {targetBone {boneName=thigh_r}}
     {targetBone {boneName=thigh_twist_01_r}}
     {targetBone {boneName=pelvis}}
     {targetBone {boneName=thigh_l}}
     {targetBone {boneName=deform_groin_r}}
     {targetBone {boneName=thigh_twist_01_l}}
     {targetBone {boneName=spine_02}}
    ]
    constraintWeights [1,1,1,1,1,1,1]
   }
   n46 = "Constraint - fx_skirt_09 (PPPPPP)"!(
    ComponentPose=@n45
    Alpha=1
    ConstraintWeights_0=0.5585
    ConstraintWeights_1:float=0.2061
    ConstraintWeights_2:float=0.0712
    ConstraintWeights_3:float=0.0699
    ConstraintWeights_4:float=0.0545
    ConstraintWeights_5:float=0.0398
   ) -> (Pose) {
    alphaBoolBlend {blendOption=Linear}
    boneToModify {boneName=fx_skirt_09}
    constraintSetup [
     {targetBone {boneName=thigh_r}}
     {targetBone {boneName=thigh_l}}
     {targetBone {boneName=deform_knee_out_r}}
     {targetBone {boneName=deform_knee_out_l}}
     {targetBone {boneName=thigh_twist_01_r}}
     {targetBone {boneName=calf_l}}
    ]
    constraintWeights [1,1,1,1,1,1]
   }
   n47 = "Constraint - fx_skirt_10 (PPPPPPP)"!(
    ComponentPose=@n46
    Alpha=1
    ConstraintWeights_0=0.4109
    ConstraintWeights_1:float=0.2139
    ConstraintWeights_2:float=0.1665
    ConstraintWeights_3:float=0.1203
    ConstraintWeights_4:float=0.0454
    ConstraintWeights_5:float=0.0315
    ConstraintWeights_6:float=0.0115
   ) -> (Pose) {
    alphaBoolBlend {blendOption=Linear}
    boneToModify {boneName=fx_skirt_10}
    constraintSetup [
     {targetBone {boneName=thigh_l}}
     {targetBone {boneName=calf_l}}
     {targetBone {boneName=deform_knee_out_l}}
     {targetBone {boneName=thigh_r}}
     {targetBone {boneName=calf_twist_02_l}}
     {targetBone {boneName=thigh_twist_01_l}}
     {targetBone {boneName=thigh_twist_01_r}}
    ]
    constraintWeights [1,1,1,1,1,1,1]
   }
   n48 = "Constraint - fx_skirt_11 (PPPPPP)"!(
    ComponentPose=@n47
    Alpha=1
    ConstraintWeights_0=0.4995
    ConstraintWeights_1:float=0.302
    ConstraintWeights_2:float=0.1069
    ConstraintWeights_3:float=0.0896
    ConstraintWeights_4:float=0.0015
    ConstraintWeights_5:float=0.0005
   ) -> (Pose) {
    alphaBoolBlend {blendOption=Linear}
    boneToModify {boneName=fx_skirt_11}
    constraintSetup [
     {targetBone {boneName=calf_l}}
     {targetBone {boneName=calf_twist_02_l}}
     {targetBone {boneName=deform_knee_out_l}}
     {targetBone {boneName=thigh_l}}
     {targetBone {boneName=thigh_twist_01_l}}
     {targetBone {boneName=deform_knee_in_l}}
    ]
    constraintWeights [1,1,1,1,1,1]
   }
   n49 = "Constraint - fx_skirt_12 (PPPPPPPP)"!(
    ComponentPose=@n48
    Alpha=1
    ConstraintWeights_0=0.5023
    ConstraintWeights_1:float=0.1928
    ConstraintWeights_2:float=0.17
    ConstraintWeights_3:float=0.1272
    ConstraintWeights_4:float=0.0055
    ConstraintWeights_5:float=0.0019
    ConstraintWeights_6:float=0.0002
    ConstraintWeights_7:float=0.0001
   ) -> (Pose) {
    alphaBoolBlend {blendOption=Linear}
    boneToModify {boneName=fx_skirt_12}
    constraintSetup [
     {targetBone {boneName=calf_l}}
     {targetBone {boneName=thigh_l}}
     {targetBone {boneName=calf_twist_02_l}}
     {targetBone {boneName=deform_knee_out_l}}
     {targetBone {boneName=thigh_twist_01_l}}
     {targetBone {boneName=deform_knee_in_l}}
     {targetBone {boneName=thigh_r}}
     {targetBone {boneName=calf_r}}
    ]
    constraintWeights [1,1,1,1,1,1,1,1]
   }
   n50 = "Constraint - fx_skirt_13 (PPPPPPPP)"!(
    ComponentPose=@n49
    Alpha=1
    ConstraintWeights_0=0.4195
    ConstraintWeights_1:float=0.3575
    ConstraintWeights_2:float=0.1134
    ConstraintWeights_3:float=0.0776
    ConstraintWeights_4:float=0.0213
    ConstraintWeights_5:float=0.0091
    ConstraintWeights_6:float=0.0008
    ConstraintWeights_7:float=0.0007
   ) -> (Pose) {
    alphaBoolBlend {blendOption=Linear}
    boneToModify {boneName=fx_skirt_13}
    constraintSetup [
     {targetBone {boneName=calf_l}}
     {targetBone {boneName=thigh_l}}
     {targetBone {boneName=deform_knee_out_l}}
     {targetBone {boneName=calf_twist_02_l}}
     {targetBone {boneName=thigh_twist_01_l}}
     {targetBone {boneName=deform_knee_in_l}}
     {targetBone {boneName=thigh_r}}
     {targetBone {boneName=calf_r}}
    ]
    constraintWeights [1,1,1,1,1,1,1,1]
   }
   n51 = "Constraint - fx_skirt_14 (PPPPPPPP)"!(
    ComponentPose=@n50
    Alpha=1
    ConstraintWeights_0=0.4308
    ConstraintWeights_1:float=0.3454
    ConstraintWeights_2:float=0.0958
    ConstraintWeights_3:float=0.0597
    ConstraintWeights_4:float=0.0406
    ConstraintWeights_5:float=0.0219
    ConstraintWeights_6:float=0.0031
    ConstraintWeights_7:float=0.0029
   ) -> (Pose) {
    alphaBoolBlend {blendOption=Linear}
    boneToModify {boneName=fx_skirt_14}
    constraintSetup [
     {targetBone {boneName=thigh_l}}
     {targetBone {boneName=calf_l}}
     {targetBone {boneName=calf_twist_02_l}}
     {targetBone {boneName=deform_knee_out_l}}
     {targetBone {boneName=thigh_twist_01_l}}
     {targetBone {boneName=deform_knee_in_l}}
     {targetBone {boneName=thigh_r}}
     {targetBone {boneName=calf_r}}
    ]
    constraintWeights [1,1,1,1,1,1,1,1]
   }
   n52 = "Constraint - fx_skirt_15 (PPPPPPPP)"!(
    ComponentPose=@n51
    Alpha=1
    ConstraintWeights_0=0.4303
    ConstraintWeights_1:float=0.4143
    ConstraintWeights_2:float=0.1185
    ConstraintWeights_3:float=0.0137
    ConstraintWeights_4:float=0.0088
    ConstraintWeights_5:float=0.0074
    ConstraintWeights_6:float=0.0036
    ConstraintWeights_7:float=0.0034
   ) -> (Pose) {
    alphaBoolBlend {blendOption=Linear}
    boneToModify {boneName=fx_skirt_15}
    constraintSetup [
     {targetBone {boneName=calf_l}}
     {targetBone {boneName=calf_twist_02_l}}
     {targetBone {boneName=thigh_l}}
     {targetBone {boneName=deform_knee_in_l}}
     {targetBone {boneName=thigh_twist_01_l}}
     {targetBone {boneName=deform_knee_out_l}}
     {targetBone {boneName=calf_twist_02_r}}
     {targetBone {boneName=calf_r}}
    ]
    constraintWeights [1,1,1,1,1,1,1,1]
   }
   n53 = "Constraint - fx_skirt_01 (PPPPPPPP)"!(
    ComponentPose=Local To Component(LocalPose=Use cached pose 'FX_joints'())
    Alpha=1
    ConstraintWeights_0=0.4218
    ConstraintWeights_1:float=0.4212
    ConstraintWeights_2:float=0.054
    ConstraintWeights_3:float=0.0539
    ConstraintWeights_4:float=0.0312
    ConstraintWeights_5:float=0.0081
    ConstraintWeights_6:float=0.0079
    ConstraintWeights_7:float=0.0019
   ) -> (Pose) {
    alphaBoolBlend {blendOption=Linear}
    boneToModify {boneName=fx_skirt_01}
    constraintSetup [
     {targetBone {boneName=dyn_dress_03_l_bk}}
     {targetBone {boneName=dyn_dress_04_l_bk}}
     {targetBone {boneName=dyn_dress_03_l_side}}
     {targetBone {boneName=dyn_dress_04_l_side}}
     {targetBone {boneName=dyn_dress_02_l_bk}}
     {targetBone {boneName=dyn_dress_03_r_bk}}
     {targetBone {boneName=dyn_dress_04_r_bk}}
     {targetBone {boneName=dyn_dress_02_l_side}}
    ]
    constraintWeights [1,1,1,1,1,1,1,1]
   }
   n54 = "Constraint - fx_skirt_02 (PPPPPPPP)"(
    ComponentPose=@n53
    Alpha=1
    ConstraintWeights_0=0.2518
    ConstraintWeights_1=0.2457
    ConstraintWeights_2=0.2363
    ConstraintWeights_3=0.2324
    ConstraintWeights_4=0.0143
    ConstraintWeights_5=0.0125
    ConstraintWeights_6=0.0035
    ConstraintWeights_7=0.0035
   ) {
    alphaBoolBlend {blendOption=Linear}
    boneToModify {boneName=fx_skirt_02}
    constraintSetup [
     {targetBone {boneName=dyn_dress_03_r_bk}}
     {targetBone {boneName=dyn_dress_04_r_bk}}
     {targetBone {boneName=dyn_dress_03_l_bk}}
     {targetBone {boneName=dyn_dress_04_l_bk}}
     {targetBone {boneName=dyn_dress_02_r_bk}}
     {targetBone {boneName=dyn_dress_02_l_bk}}
     {targetBone {boneName=dyn_dress_03_l_side}}
     {targetBone {boneName=dyn_dress_04_l_side}}
    ]
    constraintWeights [1,1,1,1,1,1,1,1]
   }
   n55 = "Constraint - fx_skirt_03 (PPPPPPPP)"(
    ComponentPose=@n54
    Alpha=1
    ConstraintWeights_0=0.3524
    ConstraintWeights_1=0.3273
    ConstraintWeights_2=0.1488
    ConstraintWeights_3=0.1399
    ConstraintWeights_4=0.0215
    ConstraintWeights_5=0.0086
    ConstraintWeights_6=0.0008
    ConstraintWeights_7=0.0008
   ) {
    alphaBoolBlend {blendOption=Linear}
    boneToModify {boneName=fx_skirt_03}
    constraintSetup [
     {targetBone {boneName=dyn_dress_03_r_bk}}
     {targetBone {boneName=dyn_dress_04_r_bk}}
     {targetBone {boneName=dyn_dress_03_l_bk}}
     {targetBone {boneName=dyn_dress_04_l_bk}}
     {targetBone {boneName=dyn_dress_02_r_bk}}
     {targetBone {boneName=dyn_dress_02_l_bk}}
     {targetBone {boneName=dyn_dress_03_l_side}}
     {targetBone {boneName=dyn_dress_04_l_side}}
    ]
    constraintWeights [1,1,1,1,1,1,1,1]
   }
   n56 = "Constraint - fx_skirt_04 (PPPPPPPP)"(
    ComponentPose=@n55
    Alpha=1
    ConstraintWeights_0=0.5213
    ConstraintWeights_1=0.2524
    ConstraintWeights_2=0.08
    ConstraintWeights_3=0.052
    ConstraintWeights_4=0.038
    ConstraintWeights_5=0.0353
    ConstraintWeights_6=0.0175
    ConstraintWeights_7=0.0035
   ) {
    alphaBoolBlend {blendOption=Linear}
    boneToModify {boneName=fx_skirt_04}
    constraintSetup [
     {targetBone {boneName=dyn_dress_03_r_bk}}
     {targetBone {boneName=dyn_dress_02_r_bk}}
     {targetBone {boneName=dyn_dress_04_r_bk}}
     {targetBone {boneName=dyn_dress_03_l_bk}}
     {targetBone {boneName=dyn_dress_02_l_bk}}
     {targetBone {boneName=dyn_dress_01_r_bk}}
     {targetBone {boneName=dyn_dress_04_l_bk}}
     {targetBone {boneName=dyn_dress_01_l_bk}}
    ]
    constraintWeights [1,1,1,1,1,1,1,1]
   }
   n57 = "Constraint - fx_skirt_05 (PPPPPP)"!(
    ComponentPose=@n56
    Alpha=1
    ConstraintWeights_0=0.5274
    ConstraintWeights_1:float=0.4033
    ConstraintWeights_2:float=0.0682
    ConstraintWeights_3:float=0.0004
    ConstraintWeights_4:float=0.0004
    ConstraintWeights_5:float=0.0003
   ) -> (Pose) {
    alphaBoolBlend {blendOption=Linear}
    boneToModify {boneName=fx_skirt_05}
    constraintSetup [
     {targetBone {boneName=dyn_dress_03_r_bk}}
     {targetBone {boneName=dyn_dress_04_r_bk}}
     {targetBone {boneName=dyn_dress_02_r_bk}}
     {targetBone {boneName=dyn_dress_03_l_bk}}
     {targetBone {boneName=dyn_dress_04_l_bk}}
     {targetBone {boneName=dyn_dress_01_r_bk}}
    ]
    constraintWeights [1,1,1,1,1,1]
   }
   n58 = "Constraint - fx_skirt_06 (PPPPPPPP)"(
    ComponentPose=@n57
    Alpha=1
    ConstraintWeights_0=0.4581
    ConstraintWeights_1=0.3137
    ConstraintWeights_2=0.1642
    ConstraintWeights_3=0.0296
    ConstraintWeights_4=0.0186
    ConstraintWeights_5=0.0093
    ConstraintWeights_6=0.0038
    ConstraintWeights_7=0.0027
   ) {
    alphaBoolBlend {blendOption=Linear}
    boneToModify {boneName=fx_skirt_06}
    constraintSetup [
     {targetBone {boneName=dyn_dress_03_r_bk}}
     {targetBone {boneName=dyn_dress_02_r_bk}}
     {targetBone {boneName=dyn_dress_04_r_bk}}
     {targetBone {boneName=dyn_dress_01_r_bk}}
     {targetBone {boneName=thigh_twist_01_r}}
     {targetBone {boneName=pelvis}}
     {targetBone {boneName=deform_glute_r}}
     {targetBone {boneName=thigh_r}}
    ]
    constraintWeights [1,1,1,1,1,1,1,1]
   }
   n59 = "Constraint - fx_skirt_07 (PPPPPPPP)"!(
    ComponentPose=@n58
    Alpha=1
    ConstraintWeights_0=0.5758
    ConstraintWeights_1:float=0.1508
    ConstraintWeights_2:float=0.1492
    ConstraintWeights_3:float=0.0327
    ConstraintWeights_4:float=0.0301
    ConstraintWeights_5:float=0.0288
    ConstraintWeights_6:float=0.0167
    ConstraintWeights_7:float=0.016
   ) -> (Pose) {
    alphaBoolBlend {blendOption=Linear}
    boneToModify {boneName=fx_skirt_07}
    constraintSetup [
     {targetBone {boneName=thigh_twist_01_r}}
     {targetBone {boneName=pelvis}}
     {targetBone {boneName=thigh_r}}
     {targetBone {boneName=dyn_dress_01_r_bk}}
     {targetBone {boneName=deform_glute_r}}
     {targetBone {boneName=dyn_dress_02_r_bk}}
     {targetBone {boneName=dyn_dress_03_r_bk}}
     {targetBone {boneName=deform_groin_r}}
    ]
    constraintWeights [1,1,1,1,1,1,1,1]
   }
   n60 = "Constraint - fx_skirt_08 (PPPPPPPP)"!(
    ComponentPose=@n59
    Alpha=1
    ConstraintWeights_0=0.3251
    ConstraintWeights_1:float=0.2443
    ConstraintWeights_2:float=0.1346
    ConstraintWeights_3:float=0.1143
    ConstraintWeights_4:float=0.065
    ConstraintWeights_5:float=0.0636
    ConstraintWeights_6:float=0.038
    ConstraintWeights_7:float=0.0151
   ) -> (Pose) {
    alphaBoolBlend {blendOption=Linear}
    boneToModify {boneName=fx_skirt_08}
    constraintSetup [
     {targetBone {boneName=thigh_r}}
     {targetBone {boneName=thigh_twist_01_r}}
     {targetBone {boneName=thigh_twist_01_l}}
     {targetBone {boneName=thigh_l}}
     {targetBone {boneName=pelvis}}
     {targetBone {boneName=dyn_dress_03_l_fr}}
     {targetBone {boneName=dyn_dress_02_l_fr}}
     {targetBone {boneName=deform_groin_r}}
    ]
    constraintWeights [1,1,1,1,1,1,1,1]
   }
   n61 = "Constraint - fx_skirt_09 (PPPPPPPP)"!(
    ComponentPose=@n60
    Alpha=1
    ConstraintWeights_0=0.2943
    ConstraintWeights_1:float=0.1748
    ConstraintWeights_2:float=0.1314
    ConstraintWeights_3:float=0.115
    ConstraintWeights_4:float=0.0947
    ConstraintWeights_5:float=0.0924
    ConstraintWeights_6:float=0.0643
    ConstraintWeights_7:float=0.0331
   ) -> (Pose) {
    alphaBoolBlend {blendOption=Linear}
    boneToModify {boneName=fx_skirt_09}
    constraintSetup [
     {targetBone {boneName=dyn_dress_03_l_fr}}
     {targetBone {boneName=dyn_dress_02_l_fr}}
     {targetBone {boneName=thigh_r}}
     {targetBone {boneName=thigh_twist_01_l}}
     {targetBone {boneName=thigh_twist_01_r}}
     {targetBone {boneName=thigh_l}}
     {targetBone {boneName=dyn_dress_04_l_fr}}
     {targetBone {boneName=pelvis}}
    ]
    constraintWeights [1,1,1,1,1,1,1,1]
   }
   n62 = "Constraint - fx_skirt_10 (PPPPPPPP)"!(
    ComponentPose=@n61
    Alpha=1
    ConstraintWeights_0=0.5602
    ConstraintWeights_1:float=0.3272
    ConstraintWeights_2:float=0.0655
    ConstraintWeights_3:float=0.0152
    ConstraintWeights_4:float=0.0142
    ConstraintWeights_5:float=0.0086
    ConstraintWeights_6:float=0.005
    ConstraintWeights_7:float=0.0042
   ) -> (Pose) {
    alphaBoolBlend {blendOption=Linear}
    boneToModify {boneName=fx_skirt_10}
    constraintSetup [
     {targetBone {boneName=dyn_dress_03_l_fr}}
     {targetBone {boneName=dyn_dress_04_l_fr}}
     {targetBone {boneName=dyn_dress_02_l_fr}}
     {targetBone {boneName=thigh_l}}
     {targetBone {boneName=thigh_twist_01_l}}
     {targetBone {boneName=dyn_dress_03_l_side}}
     {targetBone {boneName=dyn_dress_02_l_side}}
     {targetBone {boneName=thigh_r}}
    ]
    constraintWeights [1,1,1,1,1,1,1,1]
   }
   n63 = "Constraint - fx_skirt_11 (PPPPPP)"(
    ComponentPose=@n62
    Alpha=1
    ConstraintWeights_0=0.4472
    ConstraintWeights_1=0.4468
    ConstraintWeights_2=0.0422
    ConstraintWeights_3=0.042
    ConstraintWeights_4=0.0212
    ConstraintWeights_5=0.0006
   ) {
    alphaBoolBlend {blendOption=Linear}
    boneToModify {boneName=fx_skirt_11}
    constraintSetup [
     {targetBone {boneName=dyn_dress_03_l_fr}}
     {targetBone {boneName=dyn_dress_04_l_fr}}
     {targetBone {boneName=dyn_dress_03_l_side}}
     {targetBone {boneName=dyn_dress_04_l_side}}
     {targetBone {boneName=dyn_dress_02_l_fr}}
     {targetBone {boneName=dyn_dress_02_l_side}}
    ]
    constraintWeights [1,1,1,1,1,1]
   }
   n64 = "Constraint - fx_skirt_12 (PPPPPPP)"!(
    ComponentPose=@n63
    Alpha=1
    ConstraintWeights_0=0.3191
    ConstraintWeights_1:float=0.3127
    ConstraintWeights_2:float=0.1712
    ConstraintWeights_3:float=0.1675
    ConstraintWeights_4:float=0.0218
    ConstraintWeights_5:float=0.0076
    ConstraintWeights_6:float=0.0001
   ) -> (Pose) {
    alphaBoolBlend {blendOption=Linear}
    boneToModify {boneName=fx_skirt_12}
    constraintSetup [
     {targetBone {boneName=dyn_dress_03_l_fr}}
     {targetBone {boneName=dyn_dress_04_l_fr}}
     {targetBone {boneName=dyn_dress_03_l_side}}
     {targetBone {boneName=dyn_dress_04_l_side}}
     {targetBone {boneName=dyn_dress_02_l_fr}}
     {targetBone {boneName=dyn_dress_02_l_side}}
     {targetBone {boneName=thigh_twist_01_l}}
    ]
    constraintWeights [1,1,1,1,1,1,1]
   }
   n65 = "Constraint - fx_skirt_13 (PPPPPPPP)"(
    ComponentPose=@n64
    Alpha=1
    ConstraintWeights_0=0.4262
    ConstraintWeights_1=0.3858
    ConstraintWeights_2=0.0801
    ConstraintWeights_3=0.0757
    ConstraintWeights_4=0.02
    ConstraintWeights_5=0.0085
    ConstraintWeights_6=0.0018
    ConstraintWeights_7=0.0018
   ) {
    alphaBoolBlend {blendOption=Linear}
    boneToModify {boneName=fx_skirt_13}
    constraintSetup [
     {targetBone {boneName=dyn_dress_03_l_side}}
     {targetBone {boneName=dyn_dress_04_l_side}}
     {targetBone {boneName=dyn_dress_03_l_fr}}
     {targetBone {boneName=dyn_dress_04_l_fr}}
     {targetBone {boneName=dyn_dress_02_l_side}}
     {targetBone {boneName=dyn_dress_02_l_fr}}
     {targetBone {boneName=dyn_dress_03_l_bk}}
     {targetBone {boneName=dyn_dress_04_l_bk}}
    ]
    constraintWeights [1,1,1,1,1,1,1,1]
   }
   n66 = "Constraint - fx_skirt_14 (PPPPPPPP)"(
    ComponentPose=@n65
    Alpha=1
    ConstraintWeights_0=0.4745
    ConstraintWeights_1=0.2877
    ConstraintWeights_2=0.0754
    ConstraintWeights_3=0.0648
    ConstraintWeights_4=0.0643
    ConstraintWeights_5=0.0178
    ConstraintWeights_6=0.008
    ConstraintWeights_7=0.0074
   ) {
    alphaBoolBlend {blendOption=Linear}
    boneToModify {boneName=fx_skirt_14}
    constraintSetup [
     {targetBone {boneName=dyn_dress_03_l_side}}
     {targetBone {boneName=dyn_dress_04_l_side}}
     {targetBone {boneName=dyn_dress_03_l_bk}}
     {targetBone {boneName=dyn_dress_04_l_bk}}
     {targetBone {boneName=dyn_dress_02_l_side}}
     {targetBone {boneName=dyn_dress_02_l_bk}}
     {targetBone {boneName=dyn_dress_03_l_fr}}
     {targetBone {boneName=dyn_dress_04_l_fr}}
    ]
    constraintWeights [1,1,1,1,1,1,1,1]
   }
   n67 = "Constraint - fx_skirt_15 (PPPPPPPP)"(
    ComponentPose=@n66
    Alpha=1
    ConstraintWeights_0=0.26
    ConstraintWeights_1=0.2583
    ConstraintWeights_2=0.2216
    ConstraintWeights_3=0.2167
    ConstraintWeights_4=0.0213
    ConstraintWeights_5=0.0208
    ConstraintWeights_6=0.0006
    ConstraintWeights_7=0.0006
   ) {
    alphaBoolBlend {blendOption=Linear}
    boneToModify {boneName=fx_skirt_15}
    constraintSetup [
     {targetBone {boneName=dyn_dress_03_l_bk}}
     {targetBone {boneName=dyn_dress_04_l_bk}}
     {targetBone {boneName=dyn_dress_03_l_side}}
     {targetBone {boneName=dyn_dress_04_l_side}}
     {targetBone {boneName=dyn_dress_02_l_side}}
     {targetBone {boneName=dyn_dress_02_l_bk}}
     {targetBone {boneName=dyn_dress_03_l_fr}}
     {targetBone {boneName=dyn_dress_04_l_fr}}
    ]
    constraintWeights [1,1,1,1,1,1,1,1]
   }
   Save cached pose 'FrontEnd'(Pose=Blend Poses by bool(BlendPose_0=Component To Local(ComponentPose=@n67), BlendPose_1=Component To Local(ComponentPose=@n52)) {blendType=Linear}) {cacheName=FrontEnd}
  ]}
 EventGraph {nodes [
   Event BlueprintUpdateAnimation()
   > n1 = Sequence()
   n2 = Gravity Override State Function DEPRECATED(
    execute=@n1.then_1
    CharacterPartAnimInstance=Self-Reference()
    GameTimeInSeconds=GetGameTimeInSeconds()
    GravityOverrideParams=Get Gravity Override Params()
    RBAN Params=Get RBAN Params()
   )
   > Set Gravity(Gravity:Vector=@n2) -> (Output_Get:Vector)
   > Set!(
    execute
    Sim Space Settings_WorldAlpha:float=SelectFloat(A=0.1, B=0.5, bPickA=Get bIsMoving2D())
    Sim Space Settings_VelocityScaleZ:float=1
    Sim Space Settings_DampingAlpha:float=1
    Sim Space Settings_MaxLinearVelocity:float=10000
    Sim Space Settings_MaxAngularVelocity:float=10000
    Sim Space Settings_MaxLinearAcceleration:float=10000
    Sim Space Settings_MaxAngularAcceleration:float=10000
    Sim Space Settings_ExternalLinearDragV:Vector="0,0,0"
    Sim Space Settings_ExternalLinearVelocity:Vector="0,0,0"
    Sim Space Settings_ExternalAngularVelocity:Vector="0,0,0"
   ) -> (then, Output_Get:Sim Space Settings Structure)
   Set crouch(execute=@n1.then_0, crouch:bool=Get bIsCrouching()) -> (Output_Get:bool)
   > Set skydive(skydive:bool=Get bIsSkydiveFloating()) -> (Output_Get:bool)
   > Set sliding(sliding:bool=Get bIsSliding()) -> (Output_Get:bool)
   > Set LOD_1-2(LOD_1-2:bool="integer > integer"(A=Get CurrentLOD())) -> (Output_Get:bool)
  ]}
]
variables [
 crouch {type=boolean}
 Gravity {type=Vector}
 Gravity Override Params {type=GravityOverrideParamsStruct}
 LOD_1-2 {type=boolean}
 move_jump {type=number, default=0.5}
 RBAN Params {
  type=RBANParamsStruct
  default {
   dBNO {simSpaceSettings {
     maxAngularAcceleration=1000
     maxAngularVelocity=3
     maxLinearAcceleration=300
     maxLinearVelocity=700
     velocityScaleZ=0
     externalLinearDragV=<0 0.5 0>
    }}
   dBNO_Carried {simSpaceSettings {
     maxAngularAcceleration=1000
     maxAngularVelocity=3
     maxLinearAcceleration=300
     maxLinearVelocity=700
     velocityScaleZ=0
     externalLinearDragV=<0 0.5 0>
    }}
   emoteOrMelee {simSpaceSettings {
     maxAngularAcceleration=1000
     maxAngularVelocity=3
     maxLinearAcceleration=300
     maxLinearVelocity=700
     velocityScaleZ=0
     externalLinearDragV=<0 0.5 0>
    }}
   front_End {simSpaceSettings {
     maxAngularAcceleration=1000
     maxAngularVelocity=3
     maxLinearAcceleration=300
     maxLinearVelocity=700
     velocityScaleZ=0
     externalLinearDragV=<0 0.5 0>
    }}
   onGround_Crouch_Moving {simSpaceSettings {
     maxAngularAcceleration=1000
     maxAngularVelocity=3
     maxLinearAcceleration=300
     maxLinearVelocity=700
     velocityScaleZ=0
     externalLinearDragV=<0 0.5 0>
    }}
   onGround_Falling {simSpaceSettings {
     maxAngularAcceleration=1000
     maxAngularVelocity=3
     maxLinearAcceleration=300
     maxLinearVelocity=700
     velocityScaleZ=0
     externalLinearDragV=<0 0.5 0>
    }}
   onGround_Moving {simSpaceSettings {
     maxAngularAcceleration=1000
     maxAngularVelocity=3
     maxLinearAcceleration=300
     maxLinearVelocity=700
     velocityScaleZ=0
     externalLinearDragV=<0 0.5 0>
    }}
   onGround_Standing {simSpaceSettings {
     maxAngularAcceleration=1000
     maxAngularVelocity=3
     maxLinearAcceleration=300
     maxLinearVelocity=700
     velocityScaleZ=0
     externalLinearDragV=<0 0.5 0>
    }}
   rail_Grinding {simSpaceSettings {
     maxAngularAcceleration=1000
     maxAngularVelocity=3
     maxLinearAcceleration=300
     maxLinearVelocity=700
     velocityScaleZ=0
     worldAlpha=0.1
     externalLinearDragV=<0 0.5 0>
    }}
   riding_Moving {simSpaceSettings {
     maxAngularAcceleration=1000
     maxAngularVelocity=3
     maxLinearAcceleration=300
     maxLinearVelocity=700
     velocityScaleZ=0
     externalLinearDragV=<0 0.5 0>
    }}
   riding_Moving_Targeting {simSpaceSettings {
     maxAngularAcceleration=1000
     maxAngularVelocity=3
     maxLinearAcceleration=300
     maxLinearVelocity=700
     velocityScaleZ=0
     externalLinearDragV=<0 0.5 0>
    }}
   riding_Standing {simSpaceSettings {
     maxAngularAcceleration=1000
     maxAngularVelocity=3
     maxLinearAcceleration=300
     maxLinearVelocity=700
     velocityScaleZ=0
     externalLinearDragV=<0 0.5 0>
    }}
   skydive_DivingDown {simSpaceSettings {
     maxAngularAcceleration=1000
     maxAngularVelocity=3
     maxLinearAcceleration=300
     maxLinearVelocity=700
     velocityScaleZ=0
     externalLinearDragV=<0 0.5 0>
    }}
   skydive_FloatingDown {
    componentLinearAccScale=<0 0.5 0>
    componentLinearVelScale=<0 0.5 0>
    simSpaceSettings {
     maxAngularAcceleration=1000
     maxAngularVelocity=3
     maxLinearAcceleration=300
     maxLinearVelocity=700
     velocityScaleZ=0
     externalLinearDragV=<0 0.5 0>
    }
   }
   skydive_GoingUp {simSpaceSettings {
     maxAngularAcceleration=1000
     maxAngularVelocity=3
     maxLinearAcceleration=300
     maxLinearVelocity=700
     velocityScaleZ=0
     externalLinearDragV=<0 0.5 0>
    }}
   skydive_Parachuting {
    componentLinearAccScale=<0 0.1 0.3>
    componentLinearVelScale=<0 0.1 0.3>
    simSpaceSettings {
     maxAngularAcceleration=1000
     maxAngularVelocity=3
     maxLinearAcceleration=300
     maxLinearVelocity=700
     velocityScaleZ=0
     externalLinearDragV=<0 0.5 0>
    }
   }
   skydive_Surfing {simSpaceSettings {
     maxAngularAcceleration=1000
     maxAngularVelocity=3
     maxLinearAcceleration=300
     maxLinearVelocity=700
     velocityScaleZ=0
     externalLinearDragV=<0 0.5 0>
    }}
   swimming {simSpaceSettings {
     maxAngularAcceleration=1000
     maxAngularVelocity=3
     maxLinearAcceleration=300
     maxLinearVelocity=700
     velocityScaleZ=0
     externalLinearDragV=<0 0.5 0>
    }}
   tactical_Sprint {simSpaceSettings {
     maxAngularAcceleration=1000
     maxAngularVelocity=3
     maxLinearAcceleration=300
     maxLinearVelocity=700
     velocityScaleZ=0
     externalLinearDragV=<0 0.5 0>
    }}
   ziplining {simSpaceSettings {
     maxAngularAcceleration=1000
     maxAngularVelocity=3
     maxLinearAcceleration=300
     maxLinearVelocity=700
     velocityScaleZ=0
     externalLinearDragV=<0 0.5 0>
    }}
  }
 }
 Sim Space Settings {
  type=SimSpaceSettings
  default {
   dampingAlpha=1
   maxAngularAcceleration=3000
   maxAngularVelocity=3000
   maxLinearAcceleration=3000
   maxLinearVelocity=3000
   velocityScaleZ=0.1
   worldAlpha=0.5
   externalLinearDragV=<0 0 0>
  }
 }
 skydive {type=boolean}
 sliding {type=boolean}
]
