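# Control Rig for the F_MED_FlameBride face accessory that repositions hair base bones, whose chains carry veil bones, as the character moves.
# Spherical pose readers on the head bone weight transform offsets and, in one direction, a spine_05 parent constraint; one reader drives nothing.
# The skydive flag rotates back hair and suppresses two head-driven offsets.
# An interpolated crouch_move flag scales another reader's corrections.
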ControlRigBlueprint /BRCosmetics/Characters/Player/Female/Medium/Bodies/F_MED_FlameBride/Meshes/Parts/F_MED_FlameBride_FaceAcc_CtrlRig
graphs [
 RigVMFunctionLibrary {links [], nodes []}
 RigVMFunction_Sequence_1_ContainedGraph {nodes [
   n1 = RigVMFunctionEntryNode()
   RigVMFunction_Sequence = &RigVMFunction_Sequence(ExecuteContext=@n1.ExecuteContext)
   n2 = &RigVMFunction_Sequence(ExecuteContext=@RigVMFunction_Sequence.B)
   n3 = &RigVMFunction_Sequence(ExecuteContext=@n2.B)
   RigVMFunctionReturnNode(A=@RigVMFunction_Sequence.A, B=@n2.A, C=@n3.A, D=@n3.B)
  ]}
 RigVMFunction_Sequence_1_ContainedGraph {nodes [
   n1 = RigVMFunctionEntryNode()
   RigVMFunction_Sequence = &RigVMFunction_Sequence(ExecuteContext=@n1.ExecuteContext)
   n2 = &RigVMFunction_Sequence(ExecuteContext=@RigVMFunction_Sequence.B)
   n3 = &RigVMFunction_Sequence(ExecuteContext=@n2.B)
   RigVMFunctionReturnNode(A=@RigVMFunction_Sequence.A, B=@n2.A, C=@n3.A, D=@n3.B)
  ]}
 RigVMModel {nodes [
   RigUnit_BeginExecution = &RigUnit_BeginExecution()
   RigVMFunction_Sequence_2 = RigVMAggregateNode(ExecuteContext=@RigUnit_BeginExecution.ExecutePin)
   n1 = &RigVMFunction_MathBoolToFloat(Value=RigVMVariableNode() {variable=skydive}.Value)
   &RigUnit_ModifyTransforms(
    ItemToModify=((Item=(Type=Bone,Name="dyn_hair_bk_l_base"),Transform=(Rotation=(X=0,Y=0.2588,Z=0,W=0.9659),Translation=(X=0,Y=0,Z=0),Scale3D=(X=1,Y=1,Z=1))))
    WeightMinimum=0
    ExecutePin=@RigVMFunction_Sequence_2.A
    Weight=@n1.Result
   )
   > &RigUnit_ModifyTransforms(
    ItemToModify=((Item=(Type=Bone,Name="dyn_hair_bk_center_base"),Transform=(Rotation=(X=0,Y=0.2588,Z=0,W=0.9659),Translation=(X=0,Y=0,Z=0),Scale3D=(X=1,Y=1,Z=1))))
    WeightMinimum=0
    Weight=@n1.Result
   )
   > &RigUnit_ModifyTransforms(
    ItemToModify=((Item=(Type=Bone,Name="dyn_hair_bk_r_base"),Transform=(Rotation=(X=0,Y=0.2588,Z=0,W=0.9659),Translation=(X=0,Y=0,Z=0),Scale3D=(X=1,Y=1,Z=1))))
    WeightMinimum=0
    Weight=@n1.Result
   )
   RigVMFunction_Sequence = RigVMAggregateNode(ExecuteContext=@RigVMFunction_Sequence_2.C)
   n2 = &RigUnit_SphericalPoseReader(
    ActiveRegionScaleFactors=(PositiveWidth=0.75,NegativeWidth=0.75,PositiveHeight=0.5,NegativeHeight=1)
    ActiveRegionSize=0.1
    Debug=(bDrawDebug=True,bDraw2D=False,bDrawLocalAxes=False,DebugScale=25,DebugSegments=20,DebugThickness=0.25)
    DriverAxis=(X=0,Y=1,Z=0)
    DriverItem=(Type=Bone,Name="head")
    FalloffRegionScaleFactors=(PositiveWidth=1,NegativeWidth=1,PositiveHeight=0.5,NegativeHeight=1)
    FalloffSize=0.2
    FlipHeightScaling=False
    FlipWidthScaling=False
    OptionalParentItem=(Type=Bone,Name="None")
    RotationOffset=(X=60,Y=0,Z=0)
    ExecutePin=@RigVMFunction_Sequence.D
   )
   > &RigUnit_ModifyTransforms(
    ItemToModify=((Item=(Type=Bone,Name="dyn_hair_side_l_base"),Transform=(Rotation=(X=0,Y=0,Z=-0.1305,W=0.9914),Translation=(X=0.3944,Y=-1.528,Z=5.0001),Scale3D=(X=1,Y=1,Z=1))))
    WeightMinimum=0
    Weight=@n2.OutputParam
   )
   > &RigUnit_ModifyTransforms(
    ItemToModify=((Item=(Type=Bone,Name="dyn_hair_bk_l_base"),Transform=(Rotation=(X=0,Y=0,Z=-0.0872,W=0.9962),Translation=(X=0,Y=-2,Z=3),Scale3D=(X=1,Y=1,Z=1))))
    WeightMinimum=0
    Weight=@n2.OutputParam
   )
   > &RigUnit_ModifyTransforms(
    ItemToModify=((Item=(Type=Bone,Name="dyn_hair_bk_center_base"),Transform=(Rotation=(X=0,Y=0,Z=-0.0872,W=0.9962),Translation=(X=0,Y=-2,Z=3),Scale3D=(X=1,Y=1,Z=1))))
    WeightMinimum=0
    Weight=@n2.OutputParam
   )
   n3 = &RigUnit_SphericalPoseReader(
    ActiveRegionScaleFactors=(PositiveWidth=1,NegativeWidth=0.5,PositiveHeight=0.75,NegativeHeight=0.75)
    ActiveRegionSize=0.1
    Debug=(bDrawDebug=True,bDraw2D=False,bDrawLocalAxes=False,DebugScale=25,DebugSegments=20,DebugThickness=0.25)
    DriverAxis=(X=0,Y=1,Z=0)
    DriverItem=(Type=Bone,Name="head")
    FalloffRegionScaleFactors=(PositiveWidth=1,NegativeWidth=0.5,PositiveHeight=1,NegativeHeight=1)
    FalloffSize=0.2
    FlipHeightScaling=False
    FlipWidthScaling=False
    OptionalParentItem=(Type=Bone,Name="None")
    RotationOffset=(X=90,Y=0,Z=-30)
    ExecutePin=@RigVMFunction_Sequence.A
   )
   AlphaInterp = &RigVMFunction_AlphaInterp(InterpSpeedDecreasing=8, InterpSpeedIncreasing=8, Scale=1, bClampResult=False, Value=@n3.OutputParam)
   RigVMFunction_MathBoolNot = &RigVMFunction_MathBoolNot(Value=RigVMVariableNode() {variable=skydive}.Value)
   RigVMFunction_MathBoolToFloat = &RigVMFunction_MathBoolToFloat(Value=@RigVMFunction_MathBoolNot.Result)
   Multiply = &RigVMFunction_MathDoubleMul(A=@AlphaInterp.Result, B=@RigVMFunction_MathBoolToFloat.Result)
   > &RigUnit_ModifyTransforms(
    ItemToModify=((Item=(Type=Bone,Name="dyn_hair_side_r_base"),Transform=(Rotation=(X=0.0915,Y=-0.4126,Z=-0.1962,W=0.8848),Translation=(X=-0.5964,Y=2.8413,Z=5.0801),Scale3D=(X=1,Y=1,Z=1))))
    WeightMinimum=0
    Weight=@Multiply.Result
   )
   > &RigUnit_ModifyTransforms(
    ItemToModify=((Item=(Type=Bone,Name="dyn_hair_bk_r_base"),Transform=(Rotation=(X=0,Y=-0.4617,Z=0,W=0.887),Translation=(X=0,Y=0,Z=5),Scale3D=(X=1,Y=1,Z=1))))
    WeightMinimum=0
    Weight=@AlphaInterp.Result
   )
   > &RigUnit_ModifyTransforms(
    ItemToModify=((Item=(Type=Bone,Name="dyn_hair_bk_center_base"),Transform=(Rotation=(X=0,Y=-0.5373,Z=0,W=0.8434),Translation=(X=0,Y=0,Z=5),Scale3D=(X=1,Y=1,Z=1))))
    WeightMinimum=0
    Weight=@AlphaInterp.Result
   )
   > &RigUnit_ModifyTransforms(
    ItemToModify=((Item=(Type=Bone,Name="dyn_hair_bk_l_base"),Transform=(Rotation=(X=0,Y=-0.4617,Z=0,W=0.887),Translation=(X=0,Y=0,Z=5),Scale3D=(X=1,Y=1,Z=1))))
    WeightMinimum=0
    Weight=@AlphaInterp.Result
   )
   > &RigUnit_ModifyTransforms(
    ItemToModify=((Item=(Type=Bone,Name="dyn_hair_fr_side_l_base"),Transform=(Rotation=(X=0.065,Y=0.0822,Z=0.2464,W=0.9635),Translation=(X=0.2423,Y=-2.6866,Z=-0.9808),Scale3D=(X=1,Y=1,Z=1))))
    WeightMinimum=0
    Weight=&RigVMFunction_MathDoubleMul(B=@RigVMFunction_MathBoolToFloat.Result, A=@AlphaInterp.Result).Result
   )
   &RigUnit_SphericalPoseReader(
    ActiveRegionScaleFactors=(PositiveWidth=0.5,NegativeWidth=1,PositiveHeight=0.75,NegativeHeight=0.75)
    ActiveRegionSize=0.1
    Debug=(bDrawDebug=True,bDraw2D=False,bDrawLocalAxes=False,DebugScale=25,DebugSegments=20,DebugThickness=0.25)
    DriverAxis=(X=0,Y=1,Z=0)
    DriverItem=(Type=Bone,Name="head")
    FalloffRegionScaleFactors=(PositiveWidth=0.5,NegativeWidth=1,PositiveHeight=1,NegativeHeight=1)
    FalloffSize=0.2
    FlipHeightScaling=False
    FlipWidthScaling=False
    OptionalParentItem=(Type=Bone,Name="None")
    RotationOffset=(X=90,Y=0,Z=30)
    ExecutePin=@RigVMFunction_Sequence.B
   )
   n4 = &RigUnit_SphericalPoseReader(
    ActiveRegionScaleFactors=(PositiveWidth=0.75,NegativeWidth=0.75,PositiveHeight=1,NegativeHeight=0.5)
    ActiveRegionSize=0.1
    Debug=(bDrawDebug=True,bDraw2D=False,bDrawLocalAxes=False,DebugScale=25,DebugSegments=20,DebugThickness=0.25)
    DriverAxis=(X=0,Y=1,Z=0)
    DriverItem=(Type=Bone,Name="head")
    FalloffRegionScaleFactors=(PositiveWidth=1,NegativeWidth=1,PositiveHeight=1,NegativeHeight=0.5)
    FalloffSize=0.2
    FlipHeightScaling=False
    FlipWidthScaling=False
    OptionalParentItem=(Type=Bone,Name="None")
    RotationOffset=(X=120,Y=0,Z=0)
    ExecutePin=@RigVMFunction_Sequence.C
   )
   > &RigUnit_ParentConstraint(
    AdvancedSettings=(InterpolationType=Average,RotationOrderForFilter=XZY)
    Child=(Type=Bone,Name="dyn_hair_side_03_l")
    Parents=((Item=(Type=Bone,Name="spine_05"),Weight=1))
    Weight=@n4.OutputParam
   )
   > &RigUnit_ModifyTransforms(
    ItemToModify=((Item=(Type=Bone,Name="dyn_hair_side_r_base"),Transform=(Rotation=(X=0,Y=0,Z=-0.1305,W=0.9914),Translation=(X=1.7556,Y=1.7221,Z=5.0001),Scale3D=(X=1,Y=1,Z=1))))
    WeightMinimum=0
    Weight=@n4.OutputParam
   )
   > &RigUnit_ModifyTransforms(
    ItemToModify=((Item=(Type=Bone,Name="dyn_hair_bk_r_base"),Transform=(Rotation=(X=0,Y=0,Z=-0.0872,W=0.9962),Translation=(X=0,Y=2,Z=3),Scale3D=(X=1,Y=1,Z=1))))
    WeightMinimum=0
    Weight=@n4.OutputParam
   )
   > &RigUnit_ModifyTransforms(
    ItemToModify=((Item=(Type=Bone,Name="dyn_hair_bk_center_base"),Transform=(Rotation=(X=0,Y=0,Z=-0.0872,W=0.9962),Translation=(X=0,Y=2,Z=3),Scale3D=(X=1,Y=1,Z=1))))
    WeightMinimum=0
    Weight=@n4.OutputParam
   )
   n5 = &RigUnit_SphericalPoseReader(
    ActiveRegionScaleFactors=(PositiveWidth=1,NegativeWidth=0.5,PositiveHeight=0.75,NegativeHeight=0.75)
    ActiveRegionSize=0.1
    Debug=(bDrawDebug=True,bDraw2D=False,bDrawLocalAxes=False,DebugScale=25,DebugSegments=20,DebugThickness=0.25)
    DriverAxis=(X=0,Y=1,Z=0)
    DriverItem=(Type=Bone,Name="head")
    FalloffRegionScaleFactors=(PositiveWidth=1,NegativeWidth=0.5,PositiveHeight=1,NegativeHeight=1)
    FalloffSize=0.2
    FlipHeightScaling=False
    FlipWidthScaling=False
    OptionalParentItem=(Type=Bone,Name="None")
    RotationOffset=(X=90,Y=0,Z=-61)
    ExecutePin=@RigVMFunction_Sequence_2.B
   )
   n6 = &RigVMFunction_AlphaInterp(
    InterpSpeedDecreasing=10
    InterpSpeedIncreasing=10
    Scale=1
    bClampResult=False
    Value=&RigVMFunction_MathBoolToFloat(Value=RigVMVariableNode() {variable=crouch_move}.Value).Result
   )
   n7 = &RigVMFunction_MathDoubleMul(B=@n6.Result, A=@n5.OutputParam)
   > &RigUnit_ModifyTransforms(
    ItemToModify=((Item=(Type=Bone,Name="dyn_hair_side_r_base"),Transform=(Rotation=(X=0,Y=-0.2588,Z=0,W=0.9659),Translation=(X=0,Y=0,Z=0),Scale3D=(X=1,Y=1,Z=1))))
    WeightMinimum=0
    Weight=@n7.Result
   )
   > &RigUnit_ModifyTransforms(
    ItemToModify=((Item=(Type=Bone,Name="dyn_hair_bk_r_base"),Transform=(Rotation=(X=0,Y=-0.2588,Z=0,W=0.9659),Translation=(X=-5,Y=0,Z=5),Scale3D=(X=1,Y=1,Z=1))))
    WeightMinimum=0
    Weight=@n7.Result
   )
   > &RigUnit_ModifyTransforms(
    ItemToModify=((Item=(Type=Bone,Name="dyn_hair_bk_center_base"),Transform=(Rotation=(X=0,Y=-0.2588,Z=0,W=0.9659),Translation=(X=-5,Y=0,Z=5),Scale3D=(X=1,Y=1,Z=1))))
    WeightMinimum=0
    Weight=@n7.Result
   )
   > &RigUnit_ModifyTransforms(
    ItemToModify=((Item=(Type=Bone,Name="dyn_hair_bk_l_base"),Transform=(Rotation=(X=0,Y=-0.2588,Z=0,W=0.9659),Translation=(X=-5,Y=0,Z=5),Scale3D=(X=1,Y=1,Z=1))))
    WeightMinimum=0
    Weight=@n7.Result
   )
   > &RigUnit_ModifyTransforms(
    ItemToModify=((Item=(Type=Bone,Name="dyn_hair_side_l_base"),Transform=(Rotation=(X=0,Y=-0.2588,Z=0,W=0.9659),Translation=(X=0,Y=0,Z=5),Scale3D=(X=1,Y=1,Z=1))))
    WeightMinimum=0
    Weight=@n7.Result
   )
   &RigUnit_ModifyTransforms(
    ExecutePin=()
    ItemToModify=((Item=(Type=Bone,Name="dyn_hair_side_l_base"),Transform=(Rotation=(X=0,Y=0,Z=-0.1305,W=0.9914),Translation=(X=0,Y=-3,Z=5),Scale3D=(X=1,Y=1,Z=1))))
    Weight=1
    WeightMinimum=0
   )
   &RigUnit_ModifyTransforms(
    ExecutePin=()
    ItemToModify=((Item=(Type=Bone,Name="dyn_hair_side_r_base"),Transform=(Rotation=(X=0,Y=0,Z=-0.1305,W=0.9914),Translation=(X=0,Y=3,Z=5),Scale3D=(X=1,Y=1,Z=1))))
    Weight=1
    WeightMinimum=0
   )
  ]}
]
hierarchy [
 root {loc=<0 0 0>, rot=<0 0 0 1>}
 pelvis {parent=root, loc=<0 -0.5301 98.0571>, rot=<0.0127 0.707 -0.0127 -0.707>}
 spine_01 {parent=pelvis, loc=<6.1208 -0.9328 0>, rot=<0 0 0.034 -0.9994>}
 spine_02 {parent=spine_01, loc=<6.5997 -0.2095 0>, rot=<0 0 0.0238 -0.9997>}
 spine_03 {parent=spine_02, loc=<6.4914 -0.1237 0>, rot=<0 0 0.0149 -0.9999>}
 spine_04 {parent=spine_03, loc=<6.9598 0.0794 0>, rot=<0 0 -0.0132 -0.9999>}
 spine_05 {parent=spine_04, loc=<6.9987 0 0>, rot=<0 0 -0.0015 -1>}
 neck_01 {parent=spine_05, loc=<11.3138 0.1529 0>, rot=<0 0 -0.2003 -0.9797>}
 neck_02 {parent=neck_01, loc=<4.5627 0 0>, rot=<0 0 0.0504 -0.9987>}
 head {parent=neck_02, loc=<5.3914 0 0>, rot=<0 0 0.115 -0.9934>}
 dyn_hair_center_fr_l_base {parent=head, loc=<10.7673 8.5168 -1.5252>, rot=<-0.0817 -0.7961 -0.5284 -0.2833>}
 dyn_hair_center_fr_01_l {parent=dyn_hair_center_fr_l_base, loc=<0.6815 0 0>, rot=<0 0 0 1>}
 dyn_hair_center_fr_02_l {parent=dyn_hair_center_fr_01_l, loc=<2.3003 0 -0.1085>, rot=<0 -0.2556 0 -0.9668>}
 dyn_hair_center_fr_r_base {parent=head, loc=<10.78 8.4295 1.2134>, rot=<0.2279 0.5714 0.7762 0.1382>}
 dyn_hair_center_fr_01_r {parent=dyn_hair_center_fr_r_base, loc=<0.6815 0 0>, rot=<0 0 0 1>}
 dyn_hair_center_fr_02_r {parent=dyn_hair_center_fr_01_r, loc=<2.362 0 0>, rot=<0 -0.2556 0 -0.9668>}
 dyn_hair_fr_l_base {parent=head, loc=<10.2694 6.9431 -4.0082>, rot=<0.0473 -0.9279 -0.3088 -0.2034>}
 dyn_hair_fr_01_l {parent=dyn_hair_fr_l_base, loc=<0.6815 0 0>, rot=<0 0 0 1>}
 dyn_hair_fr_02_l {parent=dyn_hair_fr_01_l, loc=<2.3003 0 -0.1085>, rot=<0 0 0 1>}
 dyn_hair_fr_r_base {parent=head, loc=<10.2735 6.4616 4.078>, rot=<0.2534 0.3206 0.9113 0.051>}
 dyn_hair_fr_01_r {parent=dyn_hair_fr_r_base, loc=<0.6809 0.0001 0.0002>, rot=<0 0 0 1>}
 dyn_hair_fr_02_r {parent=dyn_hair_fr_01_r, loc=<2.3002 0 -0.1085>, rot=<0 -0.2133 0 -0.977>}
 dyn_hair_fr_side_l_base {parent=head, loc=<9.3041 3.5277 -6.1862>, rot=<0.0081 -0.9795 -0.1567 -0.1265>}
 dyn_hair_fr_side_01_l {parent=dyn_hair_fr_side_l_base, loc=<0.6815 0 0>, rot=<0 0 0 1>}
 dyn_hair_fr_side_02_l {parent=dyn_hair_fr_side_01_l, loc=<3.5 0 0>, rot=<0 -0.1086 0 -0.9941>}
 dyn_hair_fr_side_03_l {parent=dyn_hair_fr_side_02_l, loc=<3.5 0 0>, rot=<0 -0.0581 0 -0.9983>}
 dyn_hair_fr_side_r_base {parent=head, loc=<9.3039 3.5277 6.1862>, rot=<0.0977 0.1247 0.987 -0.0267>}
 dyn_hair_fr_side_01_r {parent=dyn_hair_fr_side_r_base, loc=<0.2 0 -0.0001>, rot=<0 0 0 1>}
 dyn_hair_fr_side_02_r {parent=dyn_hair_fr_side_01_r, loc=<4.5 0 0>, rot=<0 -0.1265 0 -0.992>}
 dyn_hair_fr_side_03_r {parent=dyn_hair_fr_side_02_r, loc=<4.5 0 0>, rot=<0 -0.0134 0 -0.9999>}
 dyn_hair_side_l_base {parent=head, loc=<8.0385 -0.27 -6.564>, rot=<-0.0073 -0.7071 0.7071 0.0073>}
 dyn_hair_side_01_l {parent=dyn_hair_side_l_base, loc=<0.681 0 0.0763>, rot=<0.0245 0.0984 -0.036 -0.9942>}
 dyn_hair_side_02_l {parent=dyn_hair_side_01_l, loc=<7.1653 0 -0.0569>, rot=<-0.0015 0.0676 0.0338 -0.9971>}
 dyn_hair_side_03_l {parent=dyn_hair_side_02_l, loc=<7.3501 0 -0.0128>, rot=<0 0.2168 0 -0.9762>}
 dyn_veil_side_01_l {parent=dyn_hair_side_03_l, loc=<5.0014 2.5012 1.3697>, rot=<-0.0126 -0.0613 0.0124 -0.998>}
 dyn_hair_side_r_base {parent=head, loc=<8.0383 -0.27 6.564>, rot=<-0.0073 -0.7071 0.7071 0.0073>}
 dyn_hair_side_01_r {parent=dyn_hair_side_r_base, loc=<0.6811 0 0.0763>, rot=<-0.0245 0.0984 0.036 -0.9942>}
 dyn_hair_side_02_r {parent=dyn_hair_side_01_r, loc=<7.1654 0 -0.0569>, rot=<0.0015 0.0676 -0.0338 -0.9971>}
 dyn_hair_side_03_r {parent=dyn_hair_side_02_r, loc=<7.3499 0 -0.0128>, rot=<0 0.2168 0 -0.9762>}
 dyn_veil_side_01_r {parent=dyn_hair_side_03_r, loc=<5.0014 -2.5012 1.3698>, rot=<0.0126 -0.0613 -0.0124 -0.998>}
 dyn_hair_bk_l_base {parent=head, loc=<6.593 -5.0763 -4.2767>, rot=<0.0029 -0.7071 0.7071 -0.0029>}
 dyn_hair_bk_01_l {parent=dyn_hair_bk_l_base, loc=<0.681 0 0.0763>, rot=<0 0.1091 0 -0.994>}
 dyn_hair_bk_02_l {parent=dyn_hair_bk_01_l, loc=<7.1651 0 -0.0568>, rot=<0 0.0813 0 -0.9967>}
 dyn_hair_bk_03_l {parent=dyn_hair_bk_02_l, loc=<7.3504 0 -0.0129>, rot=<0 -0.0629 0 -0.998>}
 dyn_veil_bk_01_l {parent=dyn_hair_bk_03_l, loc=<4.9796 1.562 3.0419>, rot=<0 -0.0016 0 -1>}
 dyn_hair_bk_r_base {parent=head, loc=<6.593 -5.0763 4.2767>, rot=<0.0029 -0.7071 0.7071 -0.0029>}
 dyn_hair_bk_01_r {parent=dyn_hair_bk_r_base, loc=<0.6815 0 0.0763>, rot=<0 0.1091 0 -0.994>}
 dyn_hair_bk_02_r {parent=dyn_hair_bk_01_r, loc=<7.1648 0 -0.0568>, rot=<0 0.0813 0 -0.9967>}
 dyn_hair_bk_03_r {parent=dyn_hair_bk_02_r, loc=<7.35 0 -0.0128>, rot=<0 -0.0629 0 -0.998>}
 dyn_veil_bk_01_r {parent=dyn_hair_bk_03_r, loc=<4.9799 -1.5622 3.0419>, rot=<0 -0.0016 0 -1>}
 dyn_hair_bk_center_base {parent=head, loc=<6.6114 -7.2977 0>, rot=<0.0029 -0.7071 0.7071 -0.0029>}
 dyn_hair_bk_01_center {parent=dyn_hair_bk_center_base, loc=<0.6815 0 0.0763>, rot=<0 0.0557 0 -0.9984>}
 dyn_hair_bk_02_center {parent=dyn_hair_bk_01_center, loc=<7.1648 0 -0.0568>, rot=<0 0.0904 0 -0.9959>}
 dyn_hair_bk_03_center {parent=dyn_hair_bk_02_center, loc=<7.35 0 -0.0128>, rot=<0 -0.0438 0 -0.999>}
 dyn_veil_bk_01_center {parent=dyn_hair_bk_03_center, loc=<6.2339 0 3.6279>, rot=<0 0.0238 0 -0.9997>}
 clavicle_l {parent=spine_05, loc=<5.4831 -0.1053 -3.2946>, rot=<0.0465 -0.7289 -0.009 -0.6829>}
 upperarm_l {parent=clavicle_l, loc=<9.8658 -1.2136 -0.4105>, rot=<0.0164 -0.3245 -0.0269 -0.9454>}
 clavicle_r {parent=spine_05, loc=<5.4831 -0.1053 3.2946>, rot=<0.7289 0.0465 -0.6829 0.009>}
 upperarm_r {parent=clavicle_r, loc=<-9.8658 1.2136 0.4105>, rot=<0.0164 -0.3245 -0.0269 -0.9454>}
 dyn_backpack_collision {parent=spine_05, loc=<10 -22 0>, rot=<0.0616 -0.7044 -0.0616 -0.7044>}
]
variables [
 crouch_move {type=bool, default=False}
 skydive {type=bool, default=False}
]
