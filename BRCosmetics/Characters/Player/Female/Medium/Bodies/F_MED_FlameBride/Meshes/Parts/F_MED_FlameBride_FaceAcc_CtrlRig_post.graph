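# Post-process Control Rig for the FlameBride face accessory that corrects side hair and veil bones as the head and arm move.
# Three spherical pose readers on the head are combined by maximum to offset both side hair chains and blend hair and veil toward spine_05 through axis-filtered, scaled-weight parent constraints.
# A second sequence branch reads upperarm_r and offsets only the left side hair bones.
# All pose readers have debug drawing enabled.
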ControlRigBlueprint /BRCosmetics/Characters/Player/Female/Medium/Bodies/F_MED_FlameBride/Meshes/Parts/F_MED_FlameBride_FaceAcc_CtrlRig_post
graphs [
 Maximum_1_ContainedGraph {nodes [
   n1 = RigVMFunctionEntryNode()
   Maximum = &RigVMFunction_MathDoubleMax(A=@n1.A, B=@n1.B)
   RigVMFunctionReturnNode(Result=&RigVMFunction_MathDoubleMax(B=@n1.C, A=@Maximum.Result).Result)
  ]}
 RigVMFunctionLibrary {links [], nodes []}
 RigVMModel {nodes [
   RigUnit_BeginExecution = &RigUnit_BeginExecution()
   RigVMFunction_Sequence = &RigVMFunction_Sequence(ExecuteContext=@RigUnit_BeginExecution.ExecutePin)
   n1 = &RigUnit_SphericalPoseReader(
    ActiveRegionScaleFactors=(PositiveWidth=0.75,NegativeWidth=0.75,PositiveHeight=1,NegativeHeight=0.5)
    ActiveRegionSize=0.1
    Debug=(bDrawDebug=True,bDraw2D=False,bDrawLocalAxes=False,DebugScale=25,DebugSegments=20,DebugThickness=0.25)
    DriverAxis=(X=0,Y=1,Z=0)
    DriverItem=(Type=Bone,Name="head")
    FalloffRegionScaleFactors=(PositiveWidth=1,NegativeWidth=1,PositiveHeight=1,NegativeHeight=0.5)
    FalloffSize=0.2
    FlipHeightScaling=False
    FlipWidthScaling=False
    OptionalParentItem=(Type=Bone,Name="None")
    RotationOffset=(X=120,Y=0,Z=0)
    ExecutePin=@RigVMFunction_Sequence.A
   )
   > n2 = &RigUnit_SphericalPoseReader(
    ActiveRegionScaleFactors=(PositiveWidth=0.75,NegativeWidth=0.75,PositiveHeight=0.5,NegativeHeight=1)
    ActiveRegionSize=0.1
    Debug=(bDrawDebug=True,bDraw2D=False,bDrawLocalAxes=False,DebugScale=25,DebugSegments=20,DebugThickness=0.25)
    DriverAxis=(X=0,Y=1,Z=0)
    DriverItem=(Type=Bone,Name="head")
    FalloffRegionScaleFactors=(PositiveWidth=1,NegativeWidth=1,PositiveHeight=0.5,NegativeHeight=1)
    FalloffSize=0.2
    FlipHeightScaling=False
    FlipWidthScaling=False
    OptionalParentItem=(Type=Bone,Name="None")
    RotationOffset=(X=60,Y=0,Z=0)
   )
   > n3 = &RigUnit_SphericalPoseReader(
    ActiveRegionScaleFactors=(PositiveWidth=0.5,NegativeWidth=1,PositiveHeight=0.75,NegativeHeight=0.75)
    ActiveRegionSize=0.1
    Debug=(bDrawDebug=True,bDraw2D=False,bDrawLocalAxes=False,DebugScale=25,DebugSegments=20,DebugThickness=0.25)
    DriverAxis=(X=0,Y=1,Z=0)
    DriverItem=(Type=Bone,Name="head")
    FalloffRegionScaleFactors=(PositiveWidth=0.5,NegativeWidth=1,PositiveHeight=1,NegativeHeight=1)
    FalloffSize=0.2
    FlipHeightScaling=False
    FlipWidthScaling=False
    OptionalParentItem=(Type=Bone,Name="None")
    RotationOffset=(X=90,Y=0,Z=30)
   )
   Maximum = RigVMAggregateNode(A=@n3.OutputParam, C=@n1.OutputParam, B=@n2.OutputParam)
   > &RigUnit_ModifyTransforms(
    ItemToModify=((Item=(Type=Bone,Name="dyn_hair_side_02_l"),Transform=(Rotation=(X=0,Y=0,Z=0,W=1),Translation=(X=0,Y=0,Z=-3),Scale3D=(X=1,Y=1,Z=1))))
    WeightMinimum=0
    Weight=@Maximum.Result
   )
   > &RigUnit_ModifyTransforms(
    ItemToModify=((Item=(Type=Bone,Name="dyn_hair_side_02_r"),Transform=(Rotation=(X=0,Y=0,Z=0,W=1),Translation=(X=0,Y=0,Z=-3),Scale3D=(X=1,Y=1,Z=1))))
    WeightMinimum=0
    Weight=@Maximum.Result
   )
   Multiply = &RigVMFunction_MathDoubleMul(B=0.25, A=@Maximum.Result)
   > &RigUnit_ParentConstraint(
    AdvancedSettings=(InterpolationType=Average,RotationOrderForFilter=XZY)
    Child=(Type=Bone,Name="dyn_hair_side_02_l")
    Filter=(TranslationFilter=(bX=false,bY=true,bZ=false),RotationFilter=(bX=true,bY=false,bZ=false),ScaleFilter=(bX=True,bY=True,bZ=True))
    Parents=((Item=(Type=Bone,Name="spine_05"),Weight=1))
    Weight=@Multiply.Result
   )
   > &RigUnit_ParentConstraint(
    AdvancedSettings=(InterpolationType=Average,RotationOrderForFilter=XZY)
    Child=(Type=Bone,Name="dyn_hair_side_03_l")
    Filter=(TranslationFilter=(bX=false,bY=True,bZ=false),RotationFilter=(bX=true,bY=false,bZ=false),ScaleFilter=(bX=True,bY=True,bZ=True))
    Parents=((Item=(Type=Bone,Name="spine_05"),Weight=1))
    Weight=&RigVMFunction_MathDoubleMul(B=0.5, A=@Maximum.Result).Result
   )
   > &RigUnit_ParentConstraint(
    AdvancedSettings=(InterpolationType=Average,RotationOrderForFilter=XZY)
    Child=(Type=Bone,Name="dyn_veil_side_01_l")
    Filter=(TranslationFilter=(bX=false,bY=true,bZ=false),RotationFilter=(bX=false,bY=false,bZ=false),ScaleFilter=(bX=True,bY=True,bZ=True))
    Parents=((Item=(Type=Bone,Name="spine_05"),Weight=1))
    Weight=&RigVMFunction_MathDoubleMul(B=1, A=@Maximum.Result).Result
   )
   > &RigUnit_ParentConstraint(
    AdvancedSettings=(InterpolationType=Average,RotationOrderForFilter=XZY)
    Child=(Type=Bone,Name="dyn_hair_side_02_r")
    Filter=(TranslationFilter=(bX=false,bY=true,bZ=false),RotationFilter=(bX=true,bY=false,bZ=false),ScaleFilter=(bX=True,bY=True,bZ=True))
    Parents=((Item=(Type=Bone,Name="spine_05"),Weight=1))
    Weight=&RigVMFunction_MathDoubleMul(B=0.25, A=@Maximum.Result).Result
   )
   > &RigUnit_ParentConstraint(
    AdvancedSettings=(InterpolationType=Average,RotationOrderForFilter=XZY)
    Child=(Type=Bone,Name="dyn_hair_side_03_r")
    Filter=(TranslationFilter=(bX=false,bY=True,bZ=false),RotationFilter=(bX=true,bY=false,bZ=false),ScaleFilter=(bX=True,bY=True,bZ=True))
    Parents=((Item=(Type=Bone,Name="spine_05"),Weight=1))
    Weight=&RigVMFunction_MathDoubleMul(B=0.5, A=@Maximum.Result).Result
   )
   > &RigUnit_ParentConstraint(
    AdvancedSettings=(InterpolationType=Average,RotationOrderForFilter=XZY)
    Child=(Type=Bone,Name="dyn_veil_side_01_r")
    Filter=(TranslationFilter=(bX=false,bY=True,bZ=false),RotationFilter=(bX=false,bY=false,bZ=false),ScaleFilter=(bX=True,bY=True,bZ=True))
    Parents=((Item=(Type=Bone,Name="spine_05"),Weight=1))
    Weight=&RigVMFunction_MathDoubleMul(B=1, A=@Maximum.Result).Result
   )
   n4 = &RigUnit_SphericalPoseReader(
    ActiveRegionScaleFactors=(PositiveWidth=1,NegativeWidth=1,PositiveHeight=1,NegativeHeight=1)
    ActiveRegionSize=0.1
    Debug=(bDrawDebug=True,bDraw2D=False,bDrawLocalAxes=False,DebugScale=25,DebugSegments=20,DebugThickness=0.25)
    DriverAxis=(X=-1,Y=0,Z=0)
    DriverItem=(Type=Bone,Name="upperarm_r")
    FalloffRegionScaleFactors=(PositiveWidth=1,NegativeWidth=1,PositiveHeight=1,NegativeHeight=1)
    FalloffSize=0.2
    FlipHeightScaling=False
    FlipWidthScaling=False
    OptionalParentItem=(Type=Bone,Name="None")
    RotationOffset=(X=-23,Y=35,Z=0)
    ExecutePin=@RigVMFunction_Sequence.B
   )
   > &RigUnit_ModifyTransforms(
    ItemToModify=((Item=(Type=Bone,Name="dyn_hair_side_02_l"),Transform=(Rotation=(X=0,Y=0,Z=0,W=1),Translation=(X=-1.1707,Y=0,Z=-1.8568),Scale3D=(X=1,Y=1,Z=1))))
    WeightMinimum=0
    Weight=@n4.OutputParam
   )
   > &RigUnit_ModifyTransforms(
    ItemToModify=((Item=(Type=Bone,Name="dyn_hair_side_03_l"),Transform=(Rotation=(X=0,Y=0,Z=0,W=1),Translation=(X=-4.1151,Y=-2,Z=0.7345),Scale3D=(X=1,Y=1,Z=1))))
    WeightMinimum=0
    Weight=@n4.OutputParam
   )
  ]}
]
hierarchy [
 root {loc=<0 0 0>, rot=<0 0 0 1>}
 pelvis {parent=root, loc=<0 -0.5301 98.0571>, rot=<0.0127 0.707 -0.0127 -0.707>}
 spine_01 {parent=pelvis, loc=<6.1208 -0.9328 0>, rot=<0 0 0.034 -0.9994>}
 spine_02 {parent=spine_01, loc=<6.5997 -0.2095 0>, rot=<0 0 0.0238 -0.9997>}
 spine_03 {parent=spine_02, loc=<6.4914 -0.1237 0>, rot=<0 0 0.0149 -0.9999>}
 spine_04 {parent=spine_03, loc=<6.9598 0.0794 0>, rot=<0 0 -0.0132 -0.9999>}
 spine_05 {parent=spine_04, loc=<6.9987 0 0>, rot=<0 0 -0.0015 -1>}
 neck_01 {parent=spine_05, loc=<11.3138 0.1529 0>, rot=<0 0 -0.2003 -0.9797>}
 neck_02 {parent=neck_01, loc=<4.5627 0 0>, rot=<0 0 0.0504 -0.9987>}
 head {parent=neck_02, loc=<5.3914 0 0>, rot=<0 0 0.115 -0.9934>}
 dyn_hair_center_fr_l_base {parent=head, loc=<10.7673 8.5168 -1.5252>, rot=<-0.0817 -0.7961 -0.5284 -0.2833>}
 dyn_hair_center_fr_01_l {parent=dyn_hair_center_fr_l_base, loc=<0.6815 0 0>, rot=<0 0 0 1>}
 dyn_hair_center_fr_02_l {parent=dyn_hair_center_fr_01_l, loc=<2.3003 0 -0.1085>, rot=<0 -0.2556 0 -0.9668>}
 dyn_hair_center_fr_r_base {parent=head, loc=<10.78 8.4295 1.2134>, rot=<0.2279 0.5714 0.7762 0.1382>}
 dyn_hair_center_fr_01_r {parent=dyn_hair_center_fr_r_base, loc=<0.6815 0 0>, rot=<0 0 0 1>}
 dyn_hair_center_fr_02_r {parent=dyn_hair_center_fr_01_r, loc=<2.362 0 0>, rot=<0 -0.2556 0 -0.9668>}
 dyn_hair_fr_l_base {parent=head, loc=<10.2694 6.9431 -4.0082>, rot=<0.0473 -0.9279 -0.3088 -0.2034>}
 dyn_hair_fr_01_l {parent=dyn_hair_fr_l_base, loc=<0.6815 0 0>, rot=<0 0 0 1>}
 dyn_hair_fr_02_l {parent=dyn_hair_fr_01_l, loc=<2.3003 0 -0.1085>, rot=<0 0 0 1>}
 dyn_hair_fr_r_base {parent=head, loc=<10.2735 6.4616 4.078>, rot=<0.2534 0.3206 0.9113 0.051>}
 dyn_hair_fr_01_r {parent=dyn_hair_fr_r_base, loc=<0.6809 0.0001 0.0002>, rot=<0 0 0 1>}
 dyn_hair_fr_02_r {parent=dyn_hair_fr_01_r, loc=<2.3002 0 -0.1085>, rot=<0 -0.2133 0 -0.977>}
 dyn_hair_fr_side_l_base {parent=head, loc=<9.3041 3.5277 -6.1862>, rot=<0.0081 -0.9795 -0.1567 -0.1265>}
 dyn_hair_fr_side_01_l {parent=dyn_hair_fr_side_l_base, loc=<0.6815 0 0>, rot=<0 0 0 1>}
 dyn_hair_fr_side_02_l {parent=dyn_hair_fr_side_01_l, loc=<3.5 0 0>, rot=<0 -0.1086 0 -0.9941>}
 dyn_hair_fr_side_03_l {parent=dyn_hair_fr_side_02_l, loc=<3.5 0 0>, rot=<0 -0.0581 0 -0.9983>}
 dyn_hair_fr_side_r_base {parent=head, loc=<9.3039 3.5277 6.1862>, rot=<0.0977 0.1247 0.987 -0.0267>}
 dyn_hair_fr_side_01_r {parent=dyn_hair_fr_side_r_base, loc=<0.2 0 -0.0001>, rot=<0 0 0 1>}
 dyn_hair_fr_side_02_r {parent=dyn_hair_fr_side_01_r, loc=<4.5 0 0>, rot=<0 -0.1265 0 -0.992>}
 dyn_hair_fr_side_03_r {parent=dyn_hair_fr_side_02_r, loc=<4.5 0 0>, rot=<0 -0.0134 0 -0.9999>}
 dyn_hair_fr_side_04_r {parent=dyn_hair_fr_side_03_r, loc=<3 0 0>, rot=<0 0 0 1>}
 dyn_hair_side_l_base {parent=head, loc=<8.0385 -0.27 -6.564>, rot=<-0.0073 -0.7071 0.7071 0.0073>}
 dyn_hair_side_01_l {parent=dyn_hair_side_l_base, loc=<0.681 0 0.0763>, rot=<0.0245 0.0984 -0.036 -0.9942>}
 dyn_hair_side_02_l {parent=dyn_hair_side_01_l, loc=<7.1653 0 -0.0569>, rot=<-0.0015 0.0676 0.0338 -0.9971>}
 dyn_hair_side_03_l {parent=dyn_hair_side_02_l, loc=<7.3501 0 -0.0128>, rot=<0 0.2168 0 -0.9762>}
 dyn_veil_side_01_l {parent=dyn_hair_side_03_l, loc=<5.0014 2.5012 1.3697>, rot=<-0.0126 -0.0613 0.0124 -0.998>}
 dyn_veil_side_01_r1 {parent=dyn_hair_side_03_l, loc=<4.867 2.742 2.742>, rot=<0 -0.1299 0 -0.9915>}
 dyn_hair_side_r_base {parent=head, loc=<8.0383 -0.27 6.564>, rot=<-0.0073 -0.7071 0.7071 0.0073>}
 dyn_hair_side_01_r {parent=dyn_hair_side_r_base, loc=<0.6811 0 0.0763>, rot=<-0.0245 0.0984 0.036 -0.9942>}
 dyn_hair_side_02_r {parent=dyn_hair_side_01_r, loc=<7.1654 0 -0.0569>, rot=<0.0015 0.0676 -0.0338 -0.9971>}
 dyn_hair_side_03_r {parent=dyn_hair_side_02_r, loc=<7.3499 0 -0.0128>, rot=<0 0.2168 0 -0.9762>}
 dyn_veil_side_01_r {parent=dyn_hair_side_03_r, loc=<5.0014 -2.5012 1.3698>, rot=<0.0126 -0.0613 -0.0124 -0.998>}
 dyn_hair_bk_l_base {parent=head, loc=<6.593 -5.0763 -4.2767>, rot=<0.0029 -0.7071 0.7071 -0.0029>}
 dyn_hair_bk_01_l {parent=dyn_hair_bk_l_base, loc=<0.681 0 0.0763>, rot=<0 0.1091 0 -0.994>}
 dyn_hair_bk_02_l {parent=dyn_hair_bk_01_l, loc=<7.1651 0 -0.0568>, rot=<0 0.0813 0 -0.9967>}
 dyn_hair_bk_03_l {parent=dyn_hair_bk_02_l, loc=<7.3504 0 -0.0129>, rot=<0 -0.0629 0 -0.998>}
 dyn_veil_bk_01_l {parent=dyn_hair_bk_03_l, loc=<4.9796 1.562 3.0419>, rot=<0 -0.0016 0 -1>}
 dyn_hair_bk_r_base {parent=head, loc=<6.593 -5.0763 4.2767>, rot=<0.0029 -0.7071 0.7071 -0.0029>}
 dyn_hair_bk_01_r {parent=dyn_hair_bk_r_base, loc=<0.6815 0 0.0763>, rot=<0 0.1091 0 -0.994>}
 dyn_hair_bk_02_r {parent=dyn_hair_bk_01_r, loc=<7.1648 0 -0.0568>, rot=<0 0.0813 0 -0.9967>}
 dyn_hair_bk_03_r {parent=dyn_hair_bk_02_r, loc=<7.35 0 -0.0128>, rot=<0 -0.0629 0 -0.998>}
 dyn_veil_bk_01_r {parent=dyn_hair_bk_03_r, loc=<4.9799 -1.5622 3.0419>, rot=<0 -0.0016 0 -1>}
 dyn_hair_bk_center_base {parent=head, loc=<6.6114 -7.2977 0>, rot=<0.0029 -0.7071 0.7071 -0.0029>}
 dyn_hair_bk_01_center {parent=dyn_hair_bk_center_base, loc=<0.6815 0 0.0763>, rot=<0 0.0557 0 -0.9984>}
 dyn_hair_bk_02_center {parent=dyn_hair_bk_01_center, loc=<7.1648 0 -0.0568>, rot=<0 0.0904 0 -0.9959>}
 dyn_hair_bk_03_center {parent=dyn_hair_bk_02_center, loc=<7.35 0 -0.0128>, rot=<0 -0.0438 0 -0.999>}
 dyn_veil_bk_01_center {parent=dyn_hair_bk_03_center, loc=<6.2339 0 3.6279>, rot=<0 0.0238 0 -0.9997>}
 FACIAL_C_FacialRoot {parent=head, loc=<2.1524 -0.746 -0.0002>, rot=<0.4879 -0.5118 -0.4879 -0.5118>}
 FACIAL_C_Forehead {parent=FACIAL_C_FacialRoot, loc=<-0.0043 -2.7016 8.9374>, rot=<-0.0492 0 0 -0.9988>}
 FACIAL_L_ForeheadIn {parent=FACIAL_C_FacialRoot, loc=<1.5981 -2.5578 8.8444>, rot=<-0.0665 -0.0696 0.0039 -0.9953>}
 FACIAL_R_ForeheadIn {parent=FACIAL_C_FacialRoot, loc=<-1.6088 -2.5542 8.852>, rot=<-0.0665 0.0697 -0.0038 -0.9953>}
 FACIAL_L_ForeheadMid {parent=FACIAL_C_FacialRoot, loc=<3.0213 -2.7882 8.4136>, rot=<-0.0767 -0.1755 -0.0216 -0.9812>}
 FACIAL_R_ForeheadMid {parent=FACIAL_C_FacialRoot, loc=<-3.0329 -2.7868 8.425>, rot=<-0.0767 0.1756 0.0217 -0.9812>}
 FACIAL_L_ForeheadOut {parent=FACIAL_C_FacialRoot, loc=<4.3451 -2.7674 7.6292>, rot=<-0.1269 -0.2551 -0.0806 -0.9552>}
 FACIAL_R_ForeheadOut {parent=FACIAL_C_FacialRoot, loc=<-4.361 -2.7639 7.6425>, rot=<-0.1269 0.2552 0.0807 -0.9551>}
 FACIAL_L_EyesackUpper {parent=FACIAL_C_FacialRoot, loc=<2.8864 -1.5067 8.9212>, rot=<0.043 -0.0838 -0.065 -0.9934>}
 FACIAL_R_EyesackUpper {parent=FACIAL_C_FacialRoot, loc=<-2.8943 -1.501 8.9272>, rot=<0.043 0.0839 0.065 -0.9934>}
 FACIAL_L_EyelidUpperFurrow {parent=FACIAL_C_FacialRoot, loc=<2.7523 -1.2649 8.8034>, rot=<-0.0358 -0.1503 -0.0805 -0.9847>}
 FACIAL_L_EyelidUpperA {parent=FACIAL_C_FacialRoot, loc=<2.8289 -0.7513 7.0646>, rot=<-0.1536 0 0 -0.9881>}
 FACIAL_L_EyeCornerOuter {parent=FACIAL_C_FacialRoot, loc=<4.3349 -0.7158 7.4854>, rot=<-0.0054 -0.3826 -0.0022 -0.9239>}
 FACIAL_L_EyesackLower {parent=FACIAL_C_FacialRoot, loc=<3.0808 0.8084 7.1406>, rot=<-0.028 -0.1395 -0.023 -0.9896>}
 FACIAL_L_CheekInner {parent=FACIAL_C_FacialRoot, loc=<2.0212 1.3582 7.6643>, rot=<-0.1288 -0.0868 -0.1976 -0.9679>}
 FACIAL_R_CheekInner {parent=FACIAL_C_FacialRoot, loc=<-2.0365 1.3672 7.6643>, rot=<-0.1289 0.0869 0.1976 -0.9679>}
 FACIAL_L_CheekOuter {parent=FACIAL_C_FacialRoot, loc=<3.7846 1.3795 6.2967>, rot=<0.0498 -0.3775 0.0982 -0.9194>}
 FACIAL_R_CheekOuter {parent=FACIAL_C_FacialRoot, loc=<-3.7886 1.3871 6.3001>, rot=<0.0498 0.3775 -0.0981 -0.9194>}
 FACIAL_R_NasolabialBulge {parent=FACIAL_C_FacialRoot, loc=<-2.9749 4.7358 6.2834>, rot=<0.0526 0.3775 -0.0216 -0.9243>}
 FACIAL_L_CheekLower {parent=FACIAL_C_FacialRoot, loc=<3.3108 4.5523 5.628>, rot=<0.0293 -0.4272 0.0553 -0.902>}
 FACIAL_R_CheekLower {parent=FACIAL_C_FacialRoot, loc=<-3.3123 4.5538 5.6264>, rot=<0.0294 0.4272 -0.0552 -0.902>}
 FACIAL_C_Nose {parent=FACIAL_C_FacialRoot, loc=<-0.0065 0.7812 8.8342>, rot=<-0.1361 0 0 -0.9907>}
 FACIAL_L_Nostril {parent=FACIAL_C_Nose, loc=<0.7873 2.1961 -0.2929>, rot=<0.0173 0.1305 0.0023 -0.9913>}
 FACIAL_R_Nostril {parent=FACIAL_C_Nose, loc=<-0.7768 2.1934 -0.2946>, rot=<0.0173 -0.1305 -0.0023 -0.9913>}
 FACIAL_C_NoseTip {parent=FACIAL_C_Nose, loc=<0.0054 2.1918 0.4232>, rot=<0.0872 0 0 -0.9962>}
 FACIAL_L_Masseter {parent=FACIAL_C_FacialRoot, loc=<5.9875 1.948 4.1386>, rot=<0.0636 0.1129 0.0086 -0.9915>}
 FACIAL_R_Masseter {parent=FACIAL_C_FacialRoot, loc=<-5.9926 1.9505 4.1379>, rot=<0.0636 -0.1128 -0.0085 -0.9915>}
 FACIAL_L_Ear {parent=FACIAL_C_FacialRoot, loc=<6.3771 0.2383 0.489>, rot=<-0.0321 -0.0005 -0.0261 -0.9991>}
 FACIAL_R_Ear {parent=FACIAL_C_FacialRoot, loc=<-6.3792 0.2412 0.4836>, rot=<-0.0321 0.0006 0.0262 -0.9991>}
 FACIAL_C_NoseBridge {parent=FACIAL_C_FacialRoot, loc=<-0.0011 -1.3441 9.6281>, rot=<0.038 0 0 -0.9993>}
 extra_bandage {parent=head, loc=<3.0918 7.8599 -3.0015>, rot=<-0.3769 -0.5671 -0.4021 -0.6121>}
 clavicle_l {parent=spine_05, loc=<5.4831 -0.1053 -3.2946>, rot=<0.0465 -0.7289 -0.009 -0.6829>}
 upperarm_l {parent=clavicle_l, loc=<9.8658 -1.2136 -0.4105>, rot=<0.0164 -0.3245 -0.0269 -0.9454>}
 clavicle_r {parent=spine_05, loc=<5.4831 -0.1053 3.2946>, rot=<0.7289 0.0465 -0.6829 0.009>}
 upperarm_r {parent=clavicle_r, loc=<-9.8658 1.2136 0.4105>, rot=<0.0164 -0.3245 -0.0269 -0.9454>}
 dyn_backpack_collision {parent=spine_05, loc=<10 -22 0>, rot=<0.0616 -0.7044 -0.0616 -0.7044>}
]
