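# Post-process-style animation blueprint for the Flame Bride face accessory part, driving secondary motion on hair and veil.
# Copied parent pose, with body-mesh shoulders, feeds a Control Rig and head-space RigidBody simulation.
# Switchable bone constraints and offsets lead into a post Control Rig, spring-limited AnimDynamics on front hair, and final bandage and hair-scale tweaks.
# The event graph tracks backbling, cape, LOD, crouch movement and gravity override.
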AnimBlueprint /BRCosmetics/Characters/Player/Female/Medium/Bodies/F_MED_FlameBride/Meshes/Parts/F_MED_FlameBride_FaceAcc_AnimBP
target_skeleton=&/F_MED_FlameBride_FaceAcc_Skeleton
graphs [
 AnimGraph {nodes [
   n1 = "Constraint - dyn_hair_bk_03_l (PPPPPP)"!(
    ComponentPose=Local To Component(LocalPose=Use cached pose 'post_dyn'())
    Alpha=0.5
    ConstraintWeights_0=0.6327
    ConstraintWeights_1:float=0.1427
    ConstraintWeights_2:float=0.1319
    ConstraintWeights_3:float=0.0388
    ConstraintWeights_4:float=0.0305
    ConstraintWeights_5:float=0.0234
   ) -> (Pose) {
    alphaBoolBlend {blendOption=Linear}
    boneToModify {boneName=dyn_hair_bk_03_l}
    constraintSetup [
     {targetBone {boneName=spine_05}}
     {targetBone {boneName=neck_02}}
     {targetBone {boneName=neck_01}}
     {targetBone {boneName=spine_03}}
     {targetBone {boneName=clavicle_l}}
     {targetBone {boneName=head}}
    ]
    constraintWeights [1,1,1,1,1,1]
   }
   n2 = "Constraint - dyn_hair_bk_03_center (PPPPPPP)"!(
    ComponentPose=@n1
    Alpha=0.5
    ConstraintWeights_0=0.6281
    ConstraintWeights_1:float=0.1577
    ConstraintWeights_2:float=0.1388
    ConstraintWeights_3:float=0.0438
    ConstraintWeights_4:float=0.0272
    ConstraintWeights_5:float=0.0037
    ConstraintWeights_6:float=0.0007
   ) -> (Pose) {
    alphaBoolBlend {blendOption=Linear}
    boneToModify {boneName=dyn_hair_bk_03_center}
    constraintSetup [
     {targetBone {boneName=spine_05}}
     {targetBone {boneName=neck_02}}
     {targetBone {boneName=neck_01}}
     {targetBone {boneName=spine_03}}
     {targetBone {boneName=head}}
     {targetBone {boneName=clavicle_r}}
     {targetBone {boneName=clavicle_l}}
    ]
    constraintWeights [1,1,1,1,1,1,1]
   }
   n3 = "Constraint - dyn_hair_bk_03_r (PPPPPP)"!(
    ComponentPose=@n2
    Alpha=0.5
    ConstraintWeights_0=0.6332
    ConstraintWeights_1:float=0.1423
    ConstraintWeights_2:float=0.1318
    ConstraintWeights_3:float=0.0389
    ConstraintWeights_4:float=0.0306
    ConstraintWeights_5:float=0.0232
   ) -> (Pose) {
    alphaBoolBlend {blendOption=Linear}
    boneToModify {boneName=dyn_hair_bk_03_r}
    constraintSetup [
     {targetBone {boneName=spine_05}}
     {targetBone {boneName=neck_02}}
     {targetBone {boneName=neck_01}}
     {targetBone {boneName=spine_03}}
     {targetBone {boneName=clavicle_r}}
     {targetBone {boneName=head}}
    ]
    constraintWeights [1,1,1,1,1,1]
   }
   n4 = "Constraint - dyn_hair_side_03_r (PPPPPP)"!(
    ComponentPose=@n3
    Alpha=0.5
    ConstraintWeights_0=0.6475
    ConstraintWeights_1:float=0.1431
    ConstraintWeights_2:float=0.0988
    ConstraintWeights_3:float=0.0758
    ConstraintWeights_4:float=0.0255
    ConstraintWeights_5:float=0.0093
   ) -> (Pose) {
    alphaBoolBlend {blendOption=Linear}
    boneToModify {boneName=dyn_hair_side_03_r}
    constraintSetup [
     {targetBone {boneName=spine_05}}
     {targetBone {boneName=clavicle_r}}
     {targetBone {boneName=neck_01}}
     {targetBone {boneName=neck_02}}
     {targetBone {boneName=spine_03}}
     {targetBone {boneName=head}}
    ]
    constraintWeights [1,1,1,1,1,1]
   }
   n5 = "Constraint - dyn_hair_side_03_l (PPPPPP)"!(
    ComponentPose=@n4
    Alpha=0.5
    ConstraintWeights_0=0.6479
    ConstraintWeights_1:float=0.1428
    ConstraintWeights_2:float=0.0988
    ConstraintWeights_3:float=0.0757
    ConstraintWeights_4:float=0.0255
    ConstraintWeights_5:float=0.0093
   ) -> (Pose) {
    alphaBoolBlend {blendOption=Linear}
    boneToModify {boneName=dyn_hair_side_03_l}
    constraintSetup [
     {targetBone {boneName=spine_05}}
     {targetBone {boneName=clavicle_l}}
     {targetBone {boneName=neck_01}}
     {targetBone {boneName=neck_02}}
     {targetBone {boneName=spine_03}}
     {targetBone {boneName=head}}
    ]
    constraintWeights [1,1,1,1,1,1]
   }
   n6 = "Constraint - dyn_veil_side_01_r (PPP)"!(ComponentPose=@n5, Alpha=0.5, ConstraintWeights_0=0.8027, ConstraintWeights_1:float=0.1964, ConstraintWeights_2:float=0.0008) -> (Pose) {
    alphaBoolBlend {blendOption=Linear}
    boneToModify {boneName=dyn_veil_side_01_r}
    constraintSetup [{targetBone {boneName=clavicle_l}}, {targetBone {boneName=spine_05}}, {targetBone {boneName=spine_04}}]
    constraintWeights [1,1,1]
   }
   n7 = "Constraint - dyn_veil_bk_01_center (PPP)"!(ComponentPose=@n6, Alpha=0.5, ConstraintWeights_0=0.7361, ConstraintWeights_1:float=0.133, ConstraintWeights_2:float=0.1309) -> (Pose) {
    alphaBoolBlend {blendOption=Linear}
    boneToModify {boneName=dyn_veil_bk_01_center}
    constraintSetup [{targetBone {boneName=spine_05}}, {targetBone {boneName=clavicle_r}}, {targetBone {boneName=clavicle_l}}]
    constraintWeights [1,1,1]
   }
   n8 = "Constraint - dyn_veil_bk_01_r (PPP)"!(ComponentPose=@n7, Alpha=0.5, ConstraintWeights_0=0.5165, ConstraintWeights_1:float=0.4597, ConstraintWeights_2:float=0.0238) -> (Pose) {
    alphaBoolBlend {blendOption=Linear}
    boneToModify {boneName=dyn_veil_bk_01_r}
    constraintSetup [{targetBone {boneName=spine_05}}, {targetBone {boneName=clavicle_r}}, {targetBone {boneName=clavicle_l}}]
    constraintWeights [1,1,1]
   }
   n9 = "Constraint - dyn_veil_side_01_r (PPP)"(ComponentPose=@n8, Alpha=0.5, ConstraintWeights_0=0.8027, ConstraintWeights_1=0.1964, ConstraintWeights_2=0.0008) {
    alphaBoolBlend {blendOption=Linear}
    boneToModify {boneName=dyn_veil_side_01_r}
    constraintSetup [{targetBone {boneName=clavicle_r}}, {targetBone {boneName=spine_05}}, {targetBone {boneName=spine_04}}]
    constraintWeights [1,1,1]
   }
   n10 = "Constraint - dyn_veil_bk_01_l (PPP)"!(ComponentPose=@n9, Alpha=0.5, ConstraintWeights_0=0.5165, ConstraintWeights_1:float=0.4597, ConstraintWeights_2:float=0.0238) -> (Pose) {
    alphaBoolBlend {blendOption=Linear}
    boneToModify {boneName=dyn_veil_bk_01_l}
    constraintSetup [{targetBone {boneName=spine_05}}, {targetBone {boneName=clavicle_l}}, {targetBone {boneName=clavicle_r}}]
    constraintWeights [1,1,1]
   }
   n11 = "Transform (Modify) Bone - Bone: dyn_hair_bk_03_r"(ComponentPose=@n10, Alpha=1, Translation="0,0,0", Rotation="-5,0,0", Scale="1,1,1") {
    rotationMode=BMM_Additive
    rotationSpace=BCS_BoneSpace
    translationMode=BMM_Additive
    translationSpace=BCS_BoneSpace
    alphaBoolBlend {blendOption=Linear}
    boneToModify {boneName=dyn_hair_bk_03_r}
    rotation {pitch=-5}
   }
   n12 = "Transform (Modify) Bone - Bone: dyn_hair_bk_03_center"(ComponentPose=@n11, Alpha=1, Translation="0,0,-3", Rotation="-20,0,0", Scale="1,1,1") {
    rotationMode=BMM_Additive
    rotationSpace=BCS_BoneSpace
    translationMode=BMM_Additive
    translationSpace=BCS_BoneSpace
    alphaBoolBlend {blendOption=Linear}
    boneToModify {boneName=dyn_hair_bk_03_center}
    rotation {pitch=-20}
    translation {z=-3}
   }
   n13 = "Transform (Modify) Bone - Bone: dyn_hair_side_03_l"(ComponentPose=@n12, Alpha=1, Translation="0,0,0", Rotation="-5,0,0", Scale="1,1,1") {
    rotationMode=BMM_Additive
    rotationSpace=BCS_BoneSpace
    translationMode=BMM_Additive
    translationSpace=BCS_BoneSpace
    alphaBoolBlend {blendOption=Linear}
    boneToModify {boneName=dyn_hair_side_03_l}
    rotation {pitch=-5}
   }
   n14 = Control Rig(Source=Blend Poses by bool(BlendPose_0=Component To Local(ComponentPose=@n13), BlendPose_1=Use cached pose 'post_dyn'(), BlendTime_0=0.2, BlendTime_1=0.2) {blendType=Linear}) {
    alphaBoolBlend {blendOption=Linear}
    controlRigAssetReference {blueprintRigClass=&/F_MED_FlameBride_FaceAcc_CtrlRig_post.F_MED_FlameBride_FaceAcc_CtrlRig_post_C}
   }
   n15 = Anim Dynamics - Bone: dyn_hair_fr_01_l(ComponentPose=Local To Component(LocalPose=@n14), Alpha=1, ExternalForce="0,0,0") {
    angularDampingOverride=1
    angularSpringConstant=20000
    bAngularSpring=true
    bOverrideAngularDamping=true
    bUseGravityOverride=true
    bUseSphericalLimits=true
    gravityScale=0.35
    simulationSpace=RootRelative
    alphaBoolBlend {blendOption=Linear}
    boundBone {boneName=dyn_hair_fr_01_l}
    componentLinearAccScale=<1 1 1>
    componentLinearVelScale=<1 1 1>
    gravityOverride {z=-980}
    physicsBodyDefinitions [{
      boundBone {boneName=dyn_hair_fr_01_l}
      boxExtents=<20 30 30>
      constraintSetup {angularLimitsMax=<15 15 15>, angularLimitsMin=<-15 -15 -15>, angularTarget=<1 0 0>}
      localJointOffset=<4 0 0>
     }]
    sphericalLimits [{limitRadius=5.4298, drivingBone {boneName=head}, sphereLocalOffset=<6.3962 3.6195 -4.4279>}]
   }
   n16 = Anim Dynamics - Bone: dyn_hair_fr_02_l(ComponentPose=@n15, Alpha=1, ExternalForce="0,0,0") {
    angularDampingOverride=1
    angularSpringConstant=20000
    bAngularSpring=true
    bOverrideAngularDamping=true
    bUseGravityOverride=true
    gravityScale=0.35
    simulationSpace=RootRelative
    alphaBoolBlend {blendOption=Linear}
    boundBone {boneName=dyn_hair_fr_02_l}
    componentLinearAccScale=<1 1 1>
    componentLinearVelScale=<1 1 1>
    gravityOverride {z=-980}
    physicsBodyDefinitions [{
      boundBone {boneName=dyn_hair_fr_02_l}
      boxExtents=<20 30 30>
      constraintSetup {angularLimitsMax=<15 15 15>, angularLimitsMin=<-15 -15 -15>, angularTarget=<1 0 0>}
      localJointOffset=<4 0 0>
     }]
    sphericalLimits [{limitRadius=5.4298, drivingBone {boneName=head}, sphereLocalOffset=<6.3962 3.6195 -4.4279>}]
   }
   n17 = Anim Dynamics - Bone: dyn_hair_center_fr_01_l(ComponentPose=@n16, Alpha=1, ExternalForce="0,0,0") {
    angularDampingOverride=1
    angularSpringConstant=20000
    bAngularSpring=true
    bOverrideAngularDamping=true
    bUseGravityOverride=true
    bUseSphericalLimits=true
    gravityScale=0.35
    simulationSpace=RootRelative
    alphaBoolBlend {blendOption=Linear}
    boundBone {boneName=dyn_hair_center_fr_01_l}
    componentLinearAccScale=<1 1 1>
    componentLinearVelScale=<1 1 1>
    gravityOverride {z=-980}
    physicsBodyDefinitions [{
      boundBone {boneName=dyn_hair_center_fr_01_l}
      boxExtents=<20 30 30>
      constraintSetup {angularLimitsMax=<15 15 15>, angularLimitsMin=<-15 -15 -15>, angularTarget=<1 0 0>}
      localJointOffset=<4 0 0>
     }]
    sphericalLimits [{limitRadius=5.4298, drivingBone {boneName=head}, sphereLocalOffset=<6.3636 5.8088 -2.2771>}]
   }
   n18 = Anim Dynamics - Bone: dyn_hair_center_fr_02_l(ComponentPose=@n17, Alpha=1, ExternalForce="0,0,0") {
    angularDampingOverride=1
    angularSpringConstant=20000
    bAngularSpring=true
    bOverrideAngularDamping=true
    bUseGravityOverride=true
    bUseSphericalLimits=true
    gravityScale=0.35
    simulationSpace=RootRelative
    alphaBoolBlend {blendOption=Linear}
    boundBone {boneName=dyn_hair_center_fr_02_l}
    componentLinearAccScale=<1 1 1>
    componentLinearVelScale=<1 1 1>
    gravityOverride {z=-980}
    physicsBodyDefinitions [{
      boundBone {boneName=dyn_hair_center_fr_02_l}
      boxExtents=<20 30 30>
      constraintSetup {angularLimitsMax=<15 15 15>, angularLimitsMin=<-15 -15 -15>, angularTarget=<1 0 0>}
      localJointOffset=<4 0 0>
     }]
    sphericalLimits [{limitRadius=5.4298, drivingBone {boneName=head}, sphereLocalOffset=<6.3636 5.8088 -2.2771>}]
   }
   n19 = Anim Dynamics - Bone: dyn_hair_center_fr_01_r(ComponentPose=@n18, Alpha=1, ExternalForce="0,0,0") {
    angularDampingOverride=1
    angularSpringConstant=20000
    bAngularSpring=true
    bOverrideAngularDamping=true
    bUseGravityOverride=true
    bUseSphericalLimits=true
    gravityScale=0.35
    simulationSpace=RootRelative
    alphaBoolBlend {blendOption=Linear}
    boundBone {boneName=dyn_hair_center_fr_01_r}
    componentLinearAccScale=<1 1 1>
    componentLinearVelScale=<1 1 1>
    gravityOverride {z=-980}
    physicsBodyDefinitions [{
      boundBone {boneName=dyn_hair_center_fr_01_r}
      boxExtents=<20 30 30>
      constraintSetup {angularLimitsMax=<15 15 15>, angularLimitsMin=<-15 -15 -15>, angularTarget=<1 0 0>}
      localJointOffset=<4 0 0>
     }]
    sphericalLimits [{limitRadius=5.4298, drivingBone {boneName=head}, sphereLocalOffset=<6.2531 5.3573 0.1294>}]
   }
   n20 = Anim Dynamics - Bone: dyn_hair_center_fr_02_r(ComponentPose=@n19, Alpha=1, ExternalForce="0,0,0") {
    angularDampingOverride=1
    angularSpringConstant=20000
    bAngularSpring=true
    bOverrideAngularDamping=true
    bUseGravityOverride=true
    bUseSphericalLimits=true
    gravityScale=0.35
    simulationSpace=RootRelative
    alphaBoolBlend {blendOption=Linear}
    boundBone {boneName=dyn_hair_center_fr_02_r}
    componentLinearAccScale=<1 1 1>
    componentLinearVelScale=<1 1 1>
    gravityOverride {z=-980}
    physicsBodyDefinitions [{
      boundBone {boneName=dyn_hair_center_fr_02_r}
      boxExtents=<20 30 30>
      constraintSetup {angularLimitsMax=<15 15 15>, angularLimitsMin=<-15 -15 -15>, angularTarget=<1 0 0>}
      localJointOffset=<4 0 0>
     }]
    sphericalLimits [{limitRadius=5.4298, drivingBone {boneName=head}, sphereLocalOffset=<6.2531 5.3573 0.1294>}]
   }
   n21 = Anim Dynamics - Bone: dyn_hair_fr_01_r(ComponentPose=@n20, Alpha=1, ExternalForce="0,0,0") {
    angularDampingOverride=1
    angularSpringConstant=20000
    bAngularSpring=true
    bOverrideAngularDamping=true
    bUseGravityOverride=true
    bUseSphericalLimits=true
    gravityScale=0.35
    simulationSpace=RootRelative
    alphaBoolBlend {blendOption=Linear}
    boundBone {boneName=dyn_hair_fr_01_r}
    componentLinearAccScale=<1 1 1>
    componentLinearVelScale=<1 1 1>
    gravityOverride {z=-980}
    physicsBodyDefinitions [{
      boundBone {boneName=dyn_hair_fr_01_r}
      boxExtents=<20 30 30>
      constraintSetup {angularLimitsMax=<15 15 15>, angularLimitsMin=<-15 -15 -15>, angularTarget=<1 0 0>}
      localJointOffset=<4 0 0>
     }]
    sphericalLimits [{limitRadius=5.4298, drivingBone {boneName=head}, sphereLocalOffset=<6.2524 4.9134 1.024>}]
   }
   n22 = Anim Dynamics - Bone: dyn_hair_fr_02_r(ComponentPose=@n21, Alpha=1, ExternalForce="0,0,0") {
    angularDampingOverride=1
    angularSpringConstant=20000
    bAngularSpring=true
    bOverrideAngularDamping=true
    bUseGravityOverride=true
    bUseSphericalLimits=true
    gravityScale=0.35
    simulationSpace=RootRelative
    alphaBoolBlend {blendOption=Linear}
    boundBone {boneName=dyn_hair_fr_02_r}
    componentLinearAccScale=<1 1 1>
    componentLinearVelScale=<1 1 1>
    gravityOverride {z=-980}
    physicsBodyDefinitions [{
      boundBone {boneName=dyn_hair_fr_02_r}
      boxExtents=<20 30 30>
      constraintSetup {angularLimitsMax=<15 15 15>, angularLimitsMin=<-15 -15 -15>, angularTarget=<1 0 0>}
      localJointOffset=<4 0 0>
     }]
    sphericalLimits [{limitRadius=5.4298, drivingBone {boneName=head}, sphereLocalOffset=<6.2524 4.9134 1.024>}]
   }
   n23 = "Constraint - extra_bandage (PP)"!(ComponentPose=@n22, Alpha=1, ConstraintWeights_0=0.5, ConstraintWeights_1:float=0.5) -> (Pose) {
    alphaBoolBlend {blendOption=Linear}
    boneToModify {boneName=extra_bandage}
    constraintSetup [{targetBone {boneName=FACIAL_L_CheekInner}}, {targetBone {boneName=FACIAL_L_ForeheadMid}}]
    constraintWeights [1,1]
   }
   n24 = "Transform (Modify) Bone - Bone: extra_head_hair_scale"!(ComponentPose=@n23, bAlphaBoolEnabled:bool=True, Translation="-10,0,0", Rotation="0,0,0", Scale="1,0.01,0.001") -> (Pose) {
    alphaInputType=Bool
    scaleMode=BMM_Additive
    scaleSpace=BCS_BoneSpace
    translationMode=BMM_Additive
    translationSpace=BCS_BoneSpace
    alphaBoolBlend {blendOption=Linear}
    boneToModify {boneName=extra_head_hair_scale}
    scale {y=0.01, z=0.001}
    translation {x=-10}
   }
   n25 = "Transform (Modify) Bone - Bone: extra_head_hair_scale_r"!(ComponentPose=@n24, bAlphaBoolEnabled:bool=True, Translation="0,0,-20", Rotation="0,0,0", Scale="1,1,1") -> (Pose) {
    alphaInputType=Bool
    scaleMode=BMM_Additive
    scaleSpace=BCS_BoneSpace
    translationMode=BMM_Additive
    translationSpace=BCS_BoneSpace
    alphaBoolBlend {blendOption=Linear}
    boneToModify {boneName=extra_head_hair_scale_r}
    translation {z=-20}
   }
   n26 = "Transform (Modify) Bone - Bone: extra_head_hair_scale_l"!(ComponentPose=@n25, bAlphaBoolEnabled:bool=True, Translation="0,0,20", Rotation="0,0,0", Scale="1,1,1") -> (Pose) {
    alphaInputType=Bool
    scaleMode=BMM_Additive
    scaleSpace=BCS_BoneSpace
    translationMode=BMM_Additive
    translationSpace=BCS_BoneSpace
    alphaBoolBlend {blendOption=Linear}
    boneToModify {boneName=extra_head_hair_scale_l}
    translation {z=20}
   }
   Output Pose(Result=Component To Local(ComponentPose=@n26)) {layerGroup=None}
   n27 = Layered blend per bone(
    BasePose=Copy Pose From Mesh() {bUseAttachedParent=true}
    BlendPoses_0=Copy Pose From Mesh(SourceMeshComponent=Get Body_Mesh()) {bUseAttachedParent=true}
   ) {layerSetup [{branchFilters [{boneName=clavicle_l}, {boneName=clavicle_r}, {boneName=upperarm_r}, {boneName=upperarm_l}]}]}
   Save cached pose 'pre_dyn'(Pose=@n27) {cacheName=pre_dyn}
   n28 = "Transform (Modify) Bone - Bone: dyn_backpack_collision"(ComponentPose=Local To Component(LocalPose=Use cached pose 'pre_dyn'()), Alpha=1, Translation="0,30,0", Rotation="0,0,-5", Scale="1,1,1") {
    rotationMode=BMM_Additive
    rotationSpace=BCS_BoneSpace
    translationMode=BMM_Additive
    translationSpace=BCS_BoneSpace
    alphaBoolBlend {blendOption=Linear}
    alphaScaleBiasClamp {bInterpResult=true}
    boneToModify {boneName=dyn_backpack_collision}
    rotation {roll=-5}
    translation {y=30}
   }
   n29 = Control Rig!(
    Source=Blend Poses by bool(BlendPose_0=Component To Local(ComponentPose=@n28), BlendPose_1=Use cached pose 'pre_dyn'(), BlendTime_0=0.2, BlendTime_1=0.2) {blendType=Linear}
    Alpha
    skydive:bool=false
    crouch_move:bool=false
   ) -> (Pose) {
    alphaBoolBlend {blendOption=Linear}
    controlRigAssetReference {blueprintRigClass=&/F_MED_FlameBride_FaceAcc_CtrlRig.F_MED_FlameBride_FaceAcc_CtrlRig_C}
   }
   n30 = "Transform (Modify) Bone - Bone: dyn_hair_collision_r"!(ComponentPose=Local To Component(LocalPose=@n29), bAlphaBoolEnabled:bool=True, Translation="0,0,-20", Rotation="0,0,0", Scale="1,1,1") -> (Pose) {
    alphaInputType=Bool
    translationMode=BMM_Additive
    translationSpace=BCS_BoneSpace
    alphaBoolBlend {blendInTime=0.6, blendOption=Linear, blendOutTime=0.6}
    boneToModify {boneName=dyn_hair_collision_r}
    translation {z=-20}
   }
   n31 = "Transform (Modify) Bone - Bone: dyn_hair_collision_l"!(ComponentPose=@n30, bAlphaBoolEnabled:bool=True, Translation="0,0,-20", Rotation="0,0,0", Scale="1,1,1") -> (Pose) {
    alphaInputType=Bool
    translationMode=BMM_Additive
    translationSpace=BCS_BoneSpace
    alphaBoolBlend {blendInTime=0.6, blendOption=Linear, blendOutTime=0.6}
    boneToModify {boneName=dyn_hair_collision_l}
    translation {z=-20}
   }
   n32 = RigidBody!(
    ComponentPose=@n31
    bAlphaBoolEnabled:bool
    OverrideWorldGravity:Vector="0,0,0"
    ComponentLinearAccScale:Vector="0,0,0"
    ComponentLinearVelScale:Vector="0,0,0"
    bTransferBoneVelocities:bool=true
   ) -> (Pose) {
    alphaInputType=Bool
    bClampLinearTranslationLimitToRefPose=true
    bFreezeIncomingPoseOnStart=true
    bOverrideWorldGravity=true
    bTransferBoneVelocities=true
    overridePhysicsAsset=&/F_MED_FlameBride_FaceAcc_Physics
    simulationSpace=BaseBoneSpace
    alphaBoolBlend {blendOption=Linear}
    baseBoneRef {boneName=head}
    componentAppliedLinearAccClamp=<100 100 100>
    componentLinearAccScale=<0.6 0.2 0.2>
    componentLinearVelScale=<0.6 0.2 0.2>
    simSpaceSettings {maxAngularAcceleration=40, maxAngularVelocity=8, maxLinearAcceleration=100, maxLinearVelocity=700, worldAlpha=0.2}
   }
   Save cached pose 'post_dyn'(Pose=Component To Local(ComponentPose=@n32)) {cacheName=post_dyn}
  ]}
 EventGraph {nodes [
   Event OnCharacterCustomizationLoaded()
   > n1 = Cast To PlayerPawn_Athena(Object=TryGetPawnOwner())
   > Set Body_Mesh(Body_Mesh=GetSkeletalMeshForPartType(self=@n1))
   Event BlueprintInitializeAnimation()
   > Set Skel Mesh(Skel Mesh=GetOwnerSkeletalMeshComponent())
   Event BlueprintUpdateAnimation()
   > n2 = Sequence!(execute) -> (then_0, then_1, then_2:Exec)
   n3 = Sequence!(execute=@n2.then_1) -> (then_0, then_1)
   n4 = CheckForBackbling()
   > n5 = Branch(Condition=@n4)
   > Set Is Wearing Backbling(Is Wearing Backbling=true)
   Set Is Wearing Backbling(execute=@n5.else, Is Wearing Backbling=false)
   n6 = Branch(execute=@n3.then_1, Condition=@n4.bIsBackblingACape)
   > Set Is Backbling Cape(Is Backbling Cape:bool=true) -> (Output_Get:bool)
   Set Is Backbling Cape(execute=@n6.else, Is Backbling Cape=false)
   n7 = Sequence!(execute=@n2.then_2) -> (then_0, then_1, then_2:Exec)
   > Set LOD_1-2(LOD_1-2:bool="integer > integer"(A=Get CurrentLOD())) -> (Output_Get:bool)
   Set skydive(execute=@n7.then_1, skydive:bool=false) -> (Output_Get:bool)
   n8 = "Not Equal (Float)"(A=Get!() -> (PawnDeltaRotation_Roll:float, PawnDeltaRotation_Pitch:float, PawnDeltaRotation_Yaw:float).PawnDeltaRotation_Yaw)
   n9 = Get bIsMoving2D()
   Set crouch_move(execute=@n7.then_2, crouch_move:bool=AND Boolean(A=Get bIsCrouching(), B=OR Boolean(A=@n9, B=@n8))) -> (Output_Get:bool)
   n10 = Gravity Override State Function DEPRECATED(
    execute=@n2.then_0
    CharacterPartAnimInstance=Self-Reference()
    GameTimeInSeconds=GetGameTimeInSeconds()
    GravityOverrideParams=Get Gravity Override Params()
    DebugJointName=spine_05
   )
   > Set GravityOverride(GravityOverride=@n10)
   Get bIsFalling()
   Get bIsSkydiveFloating()
  ]}
]
variables [
 Body_Mesh {type=SkeletalMeshComponent}
 crouch_move {type=boolean}
 Gravity Override Params {type=GravityOverrideParamsStruct}
 GravityOverride {type=Vector}
 Is Backbling Cape {type=boolean}
 Is Wearing Backbling {type=boolean}
 LOD_1-2 {type=boolean}
 Pawn Delta Rotation {type=Rotator}
 Skel Mesh {type=SkeletalMeshComponent}
 skydive {type=boolean}
]
